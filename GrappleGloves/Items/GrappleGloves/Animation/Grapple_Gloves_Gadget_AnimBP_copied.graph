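# Player animation Blueprint for the Grapple Gloves gadget, overriding the full-body layer with grapple swing, release and free-fall poses.
# The event graph derives flags from the weapon, velocity, angles and SwingPhase curve, while state notifies toggle swing mesh rotation.
# Nested state machines cover hook cast, skid, air and near-ground swings, hanging, wall hit, release and fall.
# Inertialization smooths most blends, and world-space Root bone replacement orients the body.
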AnimBlueprint /GrappleGloves/Items/GrappleGloves/Animation/Grapple_Gloves_Gadget_AnimBP_copied
target_skeleton=&Fortnite_M_Avg_Player_Skeleton
graphs [
 AnimGraph {nodes [Output Pose(Result=Input Pose()) {layerGroup=None}]}
 CharacterSkeletalControlPostLegIK {nodes [Output Pose(Result=Input Pose() {name=InPosePostLegIK}) {layerGroup=PlayerOverride}]}
 EventGraph {nodes [
   n1 = JustLanded() {is=title K2Node_CustomEvent}
   > Set CalculateRotation(CalculateRotation:bool=false) -> (Output_Get:bool)
   Event BlueprintUpdateAnimation()
   > n2 = Sequence!(execute) -> (then_0, then_1, then_2:Exec, then_3:Exec, then_4:Exec)
   > n3 = Cast To B_Weapon_GrappleParent!(
    execute
    Object=Get CurrentWeapon!(self:Fort Pawn Object Reference=Get Player() -> (Player:Fort Player Pawn Athena Object Reference)) -> (CurrentWeapon:Fort Weapon Object Reference)
   ) -> (then, CastFailed, AsB Weapon Grapple Parent:B Weapon Grapple Parent Object Reference) {is=title K2Node_DynamicCast}
   > Set Weapon(Weapon:B Weapon Grapple Parent Object Reference=@n3) -> (Output_Get:B Weapon Grapple Parent Object Reference)
   > Set isFreeFalling(isFreeFalling:bool=Get FreeFalling!(self:B Weapon Grapple Parent Object Reference=Get Weapon() -> (Weapon:B Weapon Grapple Parent Object Reference)) -> (FreeFalling:bool)) -> (Output_Get:bool)
   Set bSwinging_UseProgrammaticRotation!(execute=@n2.then_1, bSwinging_UseProgrammaticRotation:bool=false, self:Fort Player Pawn Athena Object Reference=Get Player()) -> (then, Output_Get:bool)
   > n4 = Branch(Condition=Get Bound() -> (Bound:bool))
   Bind Event to On Pawn Exit Swinging(execute=@n4.else, self:Fort Player Pawn Object Reference=Get Player(), Delegate=@n1) {is=title K2Node_AddDelegate}
   > Set Bound(Bound:bool=true) -> (Output_Get:bool)
   n5 = BreakVector(InVec=Get XYVelocity() -> (XYVelocity:Vector))
   n6 = "float < float"(A="Absolute (Float)"(A=@n5), B=350)
   n7 = "float < float"(A="Absolute (Float)"(A=@n5.Y), B=350)
   n8 = GetCurveValue(CurveName=SwingPhase)
   n9 = "float >= float"(A=@n8, B=0.3)
   n10 = "float < float"(A=@n8, B=0.7)
   n11 = Branch(execute=@n2.then_4, Condition=AND Boolean(A=AND Boolean(A=@n6, B=@n7), B=AND Boolean(A=@n10, B=@n9)))
   > Set IsHanging(IsHanging:bool=true) -> (Output_Get:bool)
   Set IsHanging(execute=@n11.else, IsHanging=false)
   n12 = Get RollAngle() -> (RollAngle:float)
   n13 = "float < float"(A=@n12, B=-35)
   n14 = "float >= float"(A=@n12, B=35)
   n15 = Get PitchAngleFromAttachPoint() -> (PitchAngleFromAttachPoint:float)
   n16 = "float >= float"(A=@n15, B=-55)
   n17 = "float < float"(A=@n15, B=55)
   n18 = Get PlayerDistanceToGround() -> (PlayerDistanceToGround:float)
   n19 = Branch(
    execute=@n2.then_2
    Condition=AND Boolean!(A=AND Boolean!(A=@n17, B=@n16) -> (ReturnValue), B=OR Boolean(A=@n13, B=@n14), C:bool="float >= float"(A=@n18, B=200)) -> (ReturnValue)
   )
   > Set AngledRelease(AngledRelease:bool=true) -> (Output_Get:bool)
   Set AngledRelease(execute=@n19.else, AngledRelease=false)
   n20 = GetCurveValue(CurveName=SwingPhase)
   n21 = Branch(execute=@n2.then_3, Condition=OR Boolean(A=Get IsFallRelease() -> (IsFallRelease:bool), B="float >= float"(A=@n20, B=1.5)))
   > Set AllowNonRelease(AllowNonRelease:bool=true) -> (Output_Get:bool)
   Set AllowNonRelease(execute=@n21.else, AllowNonRelease=false)
   AnimNotify_EnterWebCast() {is=title K2Node_Event}
   n22 = Get Player()
   > Set CurrentRootRotation(CurrentRootRotation:Rotator=Combine Rotators!(
     A:Rotator=Get Actor Rotation!(self:Actor Object Reference=@n22) -> (ReturnValue:Rotator) {is=title K2Node_CallFunction}
     B:Rotator="0,-90, 0"
    ) -> (ReturnValue:Rotator)) -> (Output_Get:Rotator)
   > Set CalculateRotation(CalculateRotation=true)
   > n23 = Branch(Condition=Get IsGroundSwing() -> (IsGroundSwing:bool))
   > Set bSwingingShouldRotateMesh!(execute, bSwingingShouldRotateMesh:bool=false, self:Fort Player Pawn Athena Object Reference=Get Player()) -> (then, Output_Get:bool)
   Set bSwingingShouldRotateMesh(execute=@n23.else, bSwingingShouldRotateMesh=true, self=Get Player())
   AnimNotify_EnterNearGroundToPull() {is=title K2Node_Event}
   > Set bSwingingShouldRotateMesh(bSwingingShouldRotateMesh=false, self=Get Player())
   AnimNotify_EnterSwingNearGround() {is=title K2Node_Event}
   > Set bSwingingShouldRotateMesh(bSwingingShouldRotateMesh=true, self=Get Player())
   > Set IsGroundSwing(IsGroundSwing:bool=false) -> (Output_Get:bool)
   AnimNotify_EnterSwingInAir() {is=title K2Node_Event}
   > Set bSwingingShouldRotateMesh(bSwingingShouldRotateMesh=true, self=Get Player())
   > Set IsGroundSwing(IsGroundSwing=false)
   n24 = Event BlueprintProcessAnimInputEvent()
   > Branch(Condition=IsNamed(self=@n24.AnimInputEvent, InEventName=aie.Player.Swinging.HitWallAndDetached))
   > Set IsWallHit(IsWallHit:bool=true) -> (Output_Get:bool)
   Event BlueprintPostEvaluateAnimation() {is=title K2Node_Event}
   > Set IsWallHit(IsWallHit=false)
   SplitVector!(Vec:Vector=Get VecFromPlayerToAttachPoint() -> (VecFromPlayerToAttachPoint:Vector)) -> (Magnitude:double, Direction:Vector)
   SplitVector(Vec=Get Velocity() -> (Velocity:Vector))
  ]}
 FullBodyOverride {nodes [
   Input Pose() {name=PassThroughFullBodyPose}
   n1 = Sequence Player 'GrappleGloves_FreeFalling'() {sequence=&/GrappleGloves_FreeFalling}
   AnimGraphNode_StateMachine_0 = SwingStateMachine() {is=title AnimGraphNode_StateMachine, bSkipFirstUpdateTransition=false, maxTransitionsPerFrame=0}
   n2 = Get isFreeFalling() -> (isFreeFalling:bool)
   Output Pose(Result=Modify Curve(
     SourcePose=Blend Poses by bool(BlendPose_0=@n1, BlendPose_1=@AnimGraphNode_StateMachine_0, BlendTime_0=0.8, BlendTime_1=0.6, bActiveValue=@n2) {blendType=Linear, transitionType=Inertialization}
     CurveValues_0=1
     Alpha=1
    ) {curveNames [BothLegFK], curveValues [0]}) {layerGroup=PlayerOverride}
  ]}
 SwingStateMachine {
  entry=Swinging
  path=FullBodyOverride.AnimGraphNode_StateMachine_0.SwingStateMachine
  ids=id
  nodes [
   AnimStateConduitNode_0 = Conduit()
   AnimStateNode_0 = WallHit() {is=title AnimStateNode, stateEntered {guid=#1, notifyName=EnterWallHit}}
   AnimStateNode_4 = Swinging() {
    is=title AnimStateNode
    stateEntered {guid=#2, notifyName=EnterSwinging}
    stateFullyBlended {guid=#3}
    stateLeft {guid=#4, notifyName=ExitSwinging}
   }
   AnimStateNode_5 = SwingRelease() {
    is=title AnimStateNode
    stateEntered {guid=#5}
    stateFullyBlended {guid=#6}
    stateLeft {guid=#7, notifyName=ExitSwingRelease}
   }
   AnimStateNode_6 = Fall() {
    is=title AnimStateNode
    stateEntered {guid=#8}
    stateFullyBlended {guid=#9}
    stateLeft {guid=#10}
   }
   AnimStateTransitionNode_0 = AnimStateTransitionNode() {crossfadeDuration=0.5, logicType=TLT_Inertialization}
   AnimStateTransitionNode_1 = AnimStateTransitionNode() {crossfadeDuration=0.25, logicType=TLT_Inertialization}
   AnimStateTransitionNode_15 = AnimStateTransitionNode()
   AnimStateTransitionNode_16 = AnimStateTransitionNode() {crossfadeDuration=0.5, logicType=TLT_Inertialization, priorityOrder=100}
   AnimStateTransitionNode_17 = AnimStateTransitionNode() {crossfadeDuration=0.4, logicType=TLT_Inertialization}
   AnimStateTransitionNode_18 = AnimStateTransitionNode() {crossfadeDuration=0.15, logicType=TLT_Inertialization}
   AnimStateTransitionNode_19 = AnimStateTransitionNode()
   AnimStateTransitionNode_2 = AnimStateTransitionNode() {crossfadeDuration=0.25, logicType=TLT_Inertialization, priorityOrder=100}
   AnimStateTransitionNode_20 = AnimStateTransitionNode()
   AnimStateTransitionNode_4 = AnimStateTransitionNode() {bAutomaticRuleBasedOnSequencePlayerInState=true, logicType=TLT_Inertialization}
  ]
  links [
   {
    dst=Fall
    rule=FullBodyOverride.AnimGraphNode_StateMachine_0.SwingStateMachine.AnimStateTransitionNode_17.Transition
    src=Conduit
    via=AnimStateTransitionNode_17
   }
   {
    dst=SwingRelease
    rule=FullBodyOverride.AnimGraphNode_StateMachine_0.SwingStateMachine.AnimStateTransitionNode_16.Transition
    src=Conduit
    via=AnimStateTransitionNode_16
   }
   {
    dst=Swinging
    rule=FullBodyOverride.AnimGraphNode_StateMachine_0.SwingStateMachine.AnimStateTransitionNode_18.Transition
    src=Conduit
    via=AnimStateTransitionNode_18
   }
   {
    dst=Conduit
    rule=FullBodyOverride.AnimGraphNode_StateMachine_0.SwingStateMachine.AnimStateTransitionNode_20.Transition
    src=Fall
    via=AnimStateTransitionNode_20
   }
   {
    dst=Conduit
    rule=FullBodyOverride.AnimGraphNode_StateMachine_0.SwingStateMachine.AnimStateTransitionNode_19.Transition
    src=SwingRelease
    via=AnimStateTransitionNode_19
   }
   {
    dst=Fall
    rule=FullBodyOverride.AnimGraphNode_StateMachine_0.SwingStateMachine.AnimStateTransitionNode_0.Transition
    src=SwingRelease
    via=AnimStateTransitionNode_0
   }
   {
    dst=Conduit
    rule=FullBodyOverride.AnimGraphNode_StateMachine_0.SwingStateMachine.AnimStateTransitionNode_15.Transition
    src=Swinging
    via=AnimStateTransitionNode_15
   }
   {
    dst=WallHit
    rule=FullBodyOverride.AnimGraphNode_StateMachine_0.SwingStateMachine.AnimStateTransitionNode_1.Transition
    src=Swinging
    via=AnimStateTransitionNode_1
   }
   {
    dst=Fall
    rule=FullBodyOverride.AnimGraphNode_StateMachine_0.SwingStateMachine.AnimStateTransitionNode_4.Transition
    src=WallHit
    via=AnimStateTransitionNode_4
   }
   {
    dst=Swinging
    rule=FullBodyOverride.AnimGraphNode_StateMachine_0.SwingStateMachine.AnimStateTransitionNode_2.Transition
    src=WallHit
    via=AnimStateTransitionNode_2
   }
  ]
 }
 Conduit {
  path=FullBodyOverride.AnimGraphNode_StateMachine_0.SwingStateMachine.AnimStateConduitNode_0.Conduit
  nodes [Result(bCanEnterTransition=true)]
 }
 WallHit {
  path=FullBodyOverride.AnimGraphNode_StateMachine_0.SwingStateMachine.AnimStateNode_0.WallHit
  nodes [
   n1 = "Transform (Modify) Bone - Bone: Root"!(
    ComponentPose=Local To Component(LocalPose=Sequence Player 'GrappleGloves_Release_Wall_Contact'() {sequence=&/GrappleGloves_Release_Wall_Contact})
    bAlphaBoolEnabled:bool=True
    Translation=Get CurrentRootTranslation() -> (CurrentRootTranslation:Vector)
    Rotation=Get CurrentRootRotation() -> (CurrentRootRotation:Rotator)
    Scale="1,1,1"
   ) -> (Pose) {
    alphaInputType=Bool
    rotationMode=BMM_Replace
    rotationSpace=BCS_WorldSpace
    translationMode=BMM_Replace
    translationSpace=BCS_WorldSpace
    alphaBoolBlend {blendOption=Linear, blendOutTime=0.5}
    boneToModify {boneName=Root}
   }
   Output Animation Pose(Result=Component To Local(ComponentPose=@n1))
  ]
 }
 Swinging {
  path=FullBodyOverride.AnimGraphNode_StateMachine_0.SwingStateMachine.AnimStateNode_4.Swinging
  nodes [
   AnimGraphNode_StateMachine_1 = Swing() {is=title AnimGraphNode_StateMachine, bSkipFirstUpdateTransition=false, maxTransitionsPerFrame=0}
   n1 = "Transform (Modify) Bone - Bone: Root"(
    ComponentPose=Local To Component(LocalPose=@AnimGraphNode_StateMachine_1)
    bAlphaBoolEnabled=True
    Translation=Get CurrentRootTranslation()
    Rotation=Get CurrentRootRotation()
    Scale="1,1,1"
   ) {
    alphaInputType=Bool
    rotationMode=BMM_Replace
    rotationSpace=BCS_WorldSpace
    translationMode=BMM_Replace
    translationSpace=BCS_WorldSpace
    alphaBoolBlend {blendInTime=1, blendOption=CubicInOut}
    boneToModify {boneName=Root}
   }
   Output Animation Pose(Result=Component To Local(ComponentPose=@n1))
  ]
 }
 Swing {
  entry=CastWeb
  path=FullBodyOverride.AnimGraphNode_StateMachine_0.SwingStateMachine.AnimStateNode_4.Swinging.AnimGraphNode_StateMachine_1.Swing
  ids=id
  nodes [
   AnimStateNode_0 = Skid() {
    is=title AnimStateNode
    stateEntered {guid=#11, notifyName=EnterIsBeingPulled}
    stateFullyBlended {guid=#12}
    stateLeft {guid=#13, notifyName=ExitIsBeingPulled}
   }
   AnimStateNode_1 = SwingInAir() {is=title AnimStateNode, stateEntered {guid=#14, notifyName=EnterSwingInAir}}
   AnimStateNode_2 = NearGroundToSkid() {is=title AnimStateNode, stateEntered {guid=#15, notifyName=EnterNearGroundToPull}}
   AnimStateNode_6 = Hanging() {is=title AnimStateNode, stateEntered {guid=#16, notifyName=EnterHanging}}
   AnimStateNode_8 = CastWeb() {
    is=title AnimStateNode
    stateEntered {guid=#17, notifyName=EnterWebCast}
    stateFullyBlended {guid=#18}
    stateLeft {guid=#19, notifyName=ExitWebCast}
   }
   AnimStateNode_9 = SwingNearGround() {is=title AnimStateNode, stateEntered {guid=#20, notifyName=EnterSwingNearGround}}
   AnimStateTransitionNode_0 = AnimStateTransitionNode() {crossfadeDuration=0.4, logicType=TLT_Inertialization, priorityOrder=2}
   AnimStateTransitionNode_1 = AnimStateTransitionNode() {crossfadeDuration=0.35, logicType=TLT_Inertialization}
   AnimStateTransitionNode_12 = AnimStateTransitionNode() {crossfadeDuration=0.4, logicType=TLT_Inertialization}
   AnimStateTransitionNode_17 = AnimStateTransitionNode() {crossfadeDuration=0.5, logicType=TLT_Inertialization, priorityOrder=100}
   AnimStateTransitionNode_19 = AnimStateTransitionNode() {crossfadeDuration=0.5, logicType=TLT_Inertialization, priorityOrder=100}
   AnimStateTransitionNode_2 = AnimStateTransitionNode() {crossfadeDuration=0.35, logicType=TLT_Inertialization}
   AnimStateTransitionNode_20 = AnimStateTransitionNode() {crossfadeDuration=0.6, logicType=TLT_Inertialization, priorityOrder=100}
   AnimStateTransitionNode_3 = AnimStateTransitionNode() {bAutomaticRuleBasedOnSequencePlayerInState=true, crossfadeDuration=0.35, logicType=TLT_Inertialization}
   AnimStateTransitionNode_4 = AnimStateTransitionNode() {crossfadeDuration=0.4, logicType=TLT_Inertialization, priorityOrder=2}
   AnimStateTransitionNode_5 = AnimStateTransitionNode() {crossfadeDuration=0.3, logicType=TLT_Inertialization, priorityOrder=100}
   AnimStateTransitionNode_6 = AnimStateTransitionNode() {crossfadeDuration=0.3, logicType=TLT_Inertialization}
   AnimStateTransitionNode_8 = AnimStateTransitionNode() {crossfadeDuration=0.6, logicType=TLT_Inertialization}
  ]
  links [
   {
    dst=Skid
    rule=FullBodyOverride.AnimGraphNode_StateMachine_0.SwingStateMachine.AnimStateNode_4.Swinging.AnimGraphNode_StateMachine_1.Swing.AnimStateTransitionNode_0.Transition
    src=CastWeb
    via=AnimStateTransitionNode_0
   }
   {
    dst=SwingInAir
    rule=FullBodyOverride.AnimGraphNode_StateMachine_0.SwingStateMachine.AnimStateNode_4.Swinging.AnimGraphNode_StateMachine_1.Swing.AnimStateTransitionNode_5.Transition
    src=CastWeb
    via=AnimStateTransitionNode_5
   }
   {
    dst=SwingNearGround
    rule=FullBodyOverride.AnimGraphNode_StateMachine_0.SwingStateMachine.AnimStateNode_4.Swinging.AnimGraphNode_StateMachine_1.Swing.AnimStateTransitionNode_6.Transition
    src=CastWeb
    via=AnimStateTransitionNode_6
   }
   {
    dst=SwingInAir
    rule=FullBodyOverride.AnimGraphNode_StateMachine_0.SwingStateMachine.AnimStateNode_4.Swinging.AnimGraphNode_StateMachine_1.Swing.AnimStateTransitionNode_20.Transition
    src=Hanging
    via=AnimStateTransitionNode_20
   }
   {
    dst=SwingNearGround
    rule=FullBodyOverride.AnimGraphNode_StateMachine_0.SwingStateMachine.AnimStateNode_4.Swinging.AnimGraphNode_StateMachine_1.Swing.AnimStateTransitionNode_8.Transition
    src=Hanging
    via=AnimStateTransitionNode_8
   }
   {
    dst=Skid
    rule=FullBodyOverride.AnimGraphNode_StateMachine_0.SwingStateMachine.AnimStateNode_4.Swinging.AnimGraphNode_StateMachine_1.Swing.AnimStateTransitionNode_3.Transition
    src=NearGroundToSkid
    via=AnimStateTransitionNode_3
   }
   {
    dst=SwingNearGround
    rule=FullBodyOverride.AnimGraphNode_StateMachine_0.SwingStateMachine.AnimStateNode_4.Swinging.AnimGraphNode_StateMachine_1.Swing.AnimStateTransitionNode_1.Transition
    src=Skid
    via=AnimStateTransitionNode_1
   }
   {
    dst=Hanging
    rule=FullBodyOverride.AnimGraphNode_StateMachine_0.SwingStateMachine.AnimStateNode_4.Swinging.AnimGraphNode_StateMachine_1.Swing.AnimStateTransitionNode_19.Transition
    src=SwingInAir
    via=AnimStateTransitionNode_19
   }
   {
    dst=SwingNearGround
    rule=FullBodyOverride.AnimGraphNode_StateMachine_0.SwingStateMachine.AnimStateNode_4.Swinging.AnimGraphNode_StateMachine_1.Swing.AnimStateTransitionNode_12.Transition
    src=SwingInAir
    via=AnimStateTransitionNode_12
   }
   {
    dst=Hanging
    rule=FullBodyOverride.AnimGraphNode_StateMachine_0.SwingStateMachine.AnimStateNode_4.Swinging.AnimGraphNode_StateMachine_1.Swing.AnimStateTransitionNode_17.Transition
    src=SwingNearGround
    via=AnimStateTransitionNode_17
   }
   {
    dst=NearGroundToSkid
    rule=FullBodyOverride.AnimGraphNode_StateMachine_0.SwingStateMachine.AnimStateNode_4.Swinging.AnimGraphNode_StateMachine_1.Swing.AnimStateTransitionNode_2.Transition
    src=SwingNearGround
    via=AnimStateTransitionNode_2
   }
   {
    dst=SwingInAir
    rule=FullBodyOverride.AnimGraphNode_StateMachine_0.SwingStateMachine.AnimStateNode_4.Swinging.AnimGraphNode_StateMachine_1.Swing.AnimStateTransitionNode_4.Transition
    src=SwingNearGround
    via=AnimStateTransitionNode_4
   }
  ]
 }
 Skid {
  path=FullBodyOverride.AnimGraphNode_StateMachine_0.SwingStateMachine.AnimStateNode_4.Swinging.AnimGraphNode_StateMachine_1.Swing.AnimStateNode_0.Skid
  nodes [
   n1 = Get YawAngleToAttachTarget() -> (YawAngleToAttachTarget:float)
   n2 = Blend Poses by bool(
    BlendPose_0=Blendspace Player 'GrappleGloves_Skid_Arc_Female_BS'!(X=@n1) -> (Pose) {blendSpace=/GrappleGloves/Items/GrappleGloves/Animation/Female/GrappleGloves_Skid_Arc_Female_BS}
    BlendPose_1=Blendspace Player 'GrappleGloves_Skid_Arc_BS'!(X=@n1) -> (Pose) {blendSpace=&/GrappleGloves_Skid_Arc_BS}
    BlendTime_0=0.5
    BlendTime_1=0.5
    bActiveValue=Get IsFemale() -> (IsFemale:bool)
   ) {blendType=Linear, transitionType=Inertialization}
   n3 = Blendspace Player 'GrappleGloves_Skid_ArmAdditive_BS'!(X=Get PitchAngleFromAttachPoint()) -> (Pose) {blendSpace=&/GrappleGloves_Skid_ArmAdditive_BS}
   n4 = Get YawAngleToAttachTarget()
   n5 = Blendspace Player 'GrappleGloves_Skid_Arc_Moving_Female_BS'!(X=@n4) -> (Pose) {blendSpace=/GrappleGloves/Items/GrappleGloves/Animation/Female/GrappleGloves_Skid_Arc_Moving_Female_BS}
   n6 = Blendspace Player 'GrappleGloves_Skid_Arc_Moving_BS'!(X=@n4) -> (Pose) {blendSpace=&/GrappleGloves_Skid_Arc_Moving_BS}
   n7 = Get IsFemale()
   Output Animation Pose(Result=Blend Poses by bool(
     BlendPose_0=Blend Poses by bool(BlendPose_0=@n5, BlendPose_1=@n6, BlendTime_0=0.5, BlendTime_1=0.5, bActiveValue=@n7) {blendType=Linear, transitionType=Inertialization}
     BlendPose_1=Apply Additive(Base=@n2, Additive=@n3) {alphaBoolBlend {blendOption=Linear}}
     BlendTime_0=0.5
     BlendTime_1=0.5
     bActiveValue=Get IsMovingAttachPoint() -> (IsMovingAttachPoint:bool)
    ) {blendType=Linear, transitionType=Inertialization})
  ]
 }
 SwingInAir {
  path=FullBodyOverride.AnimGraphNode_StateMachine_0.SwingStateMachine.AnimStateNode_4.Swinging.AnimGraphNode_StateMachine_1.Swing.AnimStateNode_1.SwingInAir
  nodes [
   n1 = Blendspace Player 'GrappleGloves_Swing_Arc_1Hand_BS'!(X=Get PitchAngleFromAttachPoint()) -> (Pose) {blendSpace=&/GrappleGloves_Swing_Arc_1Hand_BS}
   n2 = Blendspace Player 'GrappleGloves_Swing_Roll_1Hand_Add_BS'!(X=Get RollAngle()) -> (Pose) {blendSpace=&/GrappleGloves_Swing_Roll_1Hand_Add_BS}
   Output Animation Pose(Result=Apply Additive(
     Base=Apply Additive(Base=@n1, Additive=@n2) {alphaBoolBlend {blendOption=Linear}}
     Additive="Sequence Player (additive) 'GrappleGloves_Swing_1hand_ADD'"() {sequence=&/GrappleGloves_Swing_1hand_ADD}
    ) {alphaBoolBlend {blendInTime=0.4, blendOption=Linear, blendOutTime=0.4}})
  ]
 }
 NearGroundToSkid {
  path=FullBodyOverride.AnimGraphNode_StateMachine_0.SwingStateMachine.AnimStateNode_4.Swinging.AnimGraphNode_StateMachine_1.Swing.AnimStateNode_2.NearGroundToSkid
  nodes [Output Animation Pose(Result=Sequence Player 'GrappleGloves_CastHook_To_Skid'() {sequence=&/GrappleGloves_CastHook_To_Skid})]
 }
 Hanging {
  path=FullBodyOverride.AnimGraphNode_StateMachine_0.SwingStateMachine.AnimStateNode_4.Swinging.AnimGraphNode_StateMachine_1.Swing.AnimStateNode_6.Hanging
  nodes [Output Animation Pose(Result=Sequence Player 'GrappleGloves_Swing_Hanging_Idle'() {sequence=&/GrappleGloves_Swing_Hanging_Idle})]
 }
 CastWeb {
  path=FullBodyOverride.AnimGraphNode_StateMachine_0.SwingStateMachine.AnimStateNode_4.Swinging.AnimGraphNode_StateMachine_1.Swing.AnimStateNode_8.CastWeb
  nodes [
   AnimGraphNode_StateMachine_0 = WebCastType() {is=title AnimGraphNode_StateMachine, bAllowConduitEntryStates=true}
   Output Animation Pose(Result=@AnimGraphNode_StateMachine_0)
  ]
 }
 WebCastType {
  entry=Conduit
  path=FullBodyOverride.AnimGraphNode_StateMachine_0.SwingStateMachine.AnimStateNode_4.Swinging.AnimGraphNode_StateMachine_1.Swing.AnimStateNode_8.CastWeb.AnimGraphNode_StateMachine_0.WebCastType
  ids=id
  nodes [
   AnimStateConduitNode_0 = Conduit()
   AnimStateNode_1 = WebCastSkid() {is=title AnimStateNode}
   AnimStateNode_2 = WebCastSwing() {is=title AnimStateNode}
   AnimStateTransitionNode_1 = AnimStateTransitionNode() {crossfadeDuration=0.25, logicType=TLT_Inertialization}
   AnimStateTransitionNode_2 = AnimStateTransitionNode() {logicType=TLT_Inertialization, priorityOrder=100}
  ]
  links [
   {
    dst=WebCastSkid
    rule=FullBodyOverride.AnimGraphNode_StateMachine_0.SwingStateMachine.AnimStateNode_4.Swinging.AnimGraphNode_StateMachine_1.Swing.AnimStateNode_8.CastWeb.AnimGraphNode_StateMachine_0.WebCastType.AnimStateTransitionNode_1.Transition
    src=Conduit
    via=AnimStateTransitionNode_1
   }
   {
    dst=WebCastSwing
    rule=FullBodyOverride.AnimGraphNode_StateMachine_0.SwingStateMachine.AnimStateNode_4.Swinging.AnimGraphNode_StateMachine_1.Swing.AnimStateNode_8.CastWeb.AnimGraphNode_StateMachine_0.WebCastType.AnimStateTransitionNode_2.Transition
    src=Conduit
    via=AnimStateTransitionNode_2
   }
  ]
 }
 Conduit {
  path=FullBodyOverride.AnimGraphNode_StateMachine_0.SwingStateMachine.AnimStateNode_4.Swinging.AnimGraphNode_StateMachine_1.Swing.AnimStateNode_8.CastWeb.AnimGraphNode_StateMachine_0.WebCastType.AnimStateConduitNode_0.Conduit
  nodes [Result(bCanEnterTransition=true)]
 }
 WebCastSkid {
  path=FullBodyOverride.AnimGraphNode_StateMachine_0.SwingStateMachine.AnimStateNode_4.Swinging.AnimGraphNode_StateMachine_1.Swing.AnimStateNode_8.CastWeb.AnimGraphNode_StateMachine_0.WebCastType.AnimStateNode_1.WebCastSkid
  nodes [Output Animation Pose(Result=Sequence Player 'GrappleGloves_CastHook_To_Skid'() {sequence=&GrappleGloves_CastHook_To_Skid})]
 }
 WebCastSwing {
  path=FullBodyOverride.AnimGraphNode_StateMachine_0.SwingStateMachine.AnimStateNode_4.Swinging.AnimGraphNode_StateMachine_1.Swing.AnimStateNode_8.CastWeb.AnimGraphNode_StateMachine_0.WebCastType.AnimStateNode_2.WebCastSwing
  nodes [
   AnimGraphNode_StateMachine_2 = WebCastAir_1Hand() {is=title AnimGraphNode_StateMachine, bAllowConduitEntryStates=true, maxTransitionsPerFrame=0}
   Output Animation Pose(Result=@AnimGraphNode_StateMachine_2)
  ]
 }
 WebCastAir_1Hand {
  entry=Conduit
  path=FullBodyOverride.AnimGraphNode_StateMachine_0.SwingStateMachine.AnimStateNode_4.Swinging.AnimGraphNode_StateMachine_1.Swing.AnimStateNode_8.CastWeb.AnimGraphNode_StateMachine_0.WebCastType.AnimStateNode_2.WebCastSwing.AnimGraphNode_StateMachine_2.WebCastAir_1Hand
  ids=id
  nodes [
   AnimStateConduitNode_0 = Conduit()
   AnimStateNode_1 = WebCastDown() {is=title AnimStateNode}
   AnimStateNode_2 = WebCastUp() {is=title AnimStateNode}
   AnimStateTransitionNode_1 = AnimStateTransitionNode() {logicType=TLT_Inertialization}
   AnimStateTransitionNode_2 = AnimStateTransitionNode() {logicType=TLT_Inertialization, priorityOrder=100}
  ]
  links [
   {
    dst=WebCastDown
    rule=FullBodyOverride.AnimGraphNode_StateMachine_0.SwingStateMachine.AnimStateNode_4.Swinging.AnimGraphNode_StateMachine_1.Swing.AnimStateNode_8.CastWeb.AnimGraphNode_StateMachine_0.WebCastType.AnimStateNode_2.WebCastSwing.AnimGraphNode_StateMachine_2.WebCastAir_1Hand.AnimStateTransitionNode_1.AnimationTransitionGraph_0
    src=Conduit
    via=AnimStateTransitionNode_1
   }
   {
    dst=WebCastUp
    rule=FullBodyOverride.AnimGraphNode_StateMachine_0.SwingStateMachine.AnimStateNode_4.Swinging.AnimGraphNode_StateMachine_1.Swing.AnimStateNode_8.CastWeb.AnimGraphNode_StateMachine_0.WebCastType.AnimStateNode_2.WebCastSwing.AnimGraphNode_StateMachine_2.WebCastAir_1Hand.AnimStateTransitionNode_2.AnimationTransitionGraph_0
    src=Conduit
    via=AnimStateTransitionNode_2
   }
  ]
 }
 Conduit {
  path=FullBodyOverride.AnimGraphNode_StateMachine_0.SwingStateMachine.AnimStateNode_4.Swinging.AnimGraphNode_StateMachine_1.Swing.AnimStateNode_8.CastWeb.AnimGraphNode_StateMachine_0.WebCastType.AnimStateNode_2.WebCastSwing.AnimGraphNode_StateMachine_2.WebCastAir_1Hand.AnimStateConduitNode_0.Conduit
  nodes [Result(bCanEnterTransition=true)]
 }
 WebCastDown {
  path=FullBodyOverride.AnimGraphNode_StateMachine_0.SwingStateMachine.AnimStateNode_4.Swinging.AnimGraphNode_StateMachine_1.Swing.AnimStateNode_8.CastWeb.AnimGraphNode_StateMachine_0.WebCastType.AnimStateNode_2.WebCastSwing.AnimGraphNode_StateMachine_2.WebCastAir_1Hand.AnimStateNode_1.WebCastDown
  nodes [
   "Sequence Player (additive) 'GrappleGloves_Swing_1hand_ADD'"() {bLoopAnimation=false, sequence=&GrappleGloves_Swing_1hand_ADD}
   Output Animation Pose(Result=Sequence Player 'GrappleGloves_CastHook_Down_1hand'() {sequence=&/GrappleGloves_CastHook_Down_1hand})
  ]
 }
 WebCastUp {
  path=FullBodyOverride.AnimGraphNode_StateMachine_0.SwingStateMachine.AnimStateNode_4.Swinging.AnimGraphNode_StateMachine_1.Swing.AnimStateNode_8.CastWeb.AnimGraphNode_StateMachine_0.WebCastType.AnimStateNode_2.WebCastSwing.AnimGraphNode_StateMachine_2.WebCastAir_1Hand.AnimStateNode_2.WebCastUp
  nodes [Output Animation Pose(Result=Sequence Player 'GrappleGloves_CastHook_Up'() {sequence=&/GrappleGloves_CastHook_Up})]
 }
 AnimationTransitionGraph_0 {
  path=FullBodyOverride.AnimGraphNode_StateMachine_0.SwingStateMachine.AnimStateNode_4.Swinging.AnimGraphNode_StateMachine_1.Swing.AnimStateNode_8.CastWeb.AnimGraphNode_StateMachine_0.WebCastType.AnimStateNode_2.WebCastSwing.AnimGraphNode_StateMachine_2.WebCastAir_1Hand.AnimStateTransitionNode_1.AnimationTransitionGraph_0
  nodes [Result(bCanEnterTransition=Get IsWebCastDown() -> (IsWebCastDown:bool))]
 }
 AnimationTransitionGraph_0 {
  path=FullBodyOverride.AnimGraphNode_StateMachine_0.SwingStateMachine.AnimStateNode_4.Swinging.AnimGraphNode_StateMachine_1.Swing.AnimStateNode_8.CastWeb.AnimGraphNode_StateMachine_0.WebCastType.AnimStateNode_2.WebCastSwing.AnimGraphNode_StateMachine_2.WebCastAir_1Hand.AnimStateTransitionNode_2.AnimationTransitionGraph_0
  nodes [Result(bCanEnterTransition=true)]
 }
 Transition {
  path=FullBodyOverride.AnimGraphNode_StateMachine_0.SwingStateMachine.AnimStateNode_4.Swinging.AnimGraphNode_StateMachine_1.Swing.AnimStateNode_8.CastWeb.AnimGraphNode_StateMachine_0.WebCastType.AnimStateTransitionNode_1.Transition
  nodes [Result(bCanEnterTransition=Get IsOnGroundWebCast() -> (IsOnGroundWebCast:bool))]
 }
 Transition {
  path=FullBodyOverride.AnimGraphNode_StateMachine_0.SwingStateMachine.AnimStateNode_4.Swinging.AnimGraphNode_StateMachine_1.Swing.AnimStateNode_8.CastWeb.AnimGraphNode_StateMachine_0.WebCastType.AnimStateTransitionNode_2.Transition
  nodes [Result(bCanEnterTransition=true)]
 }
 SwingNearGround {
  path=FullBodyOverride.AnimGraphNode_StateMachine_0.SwingStateMachine.AnimStateNode_4.Swinging.AnimGraphNode_StateMachine_1.Swing.AnimStateNode_9.SwingNearGround
  nodes [
   n1 = Sequence Player 'GrappleGloves_Swing_Ground_Right_Pose'() {sequence=&/GrappleGloves_Swing_Ground_Right_Pose}
   n2 = Sequence Player 'GrappleGloves_Swing_Ground_Left_Pose'() {sequence=&/GrappleGloves_Swing_Ground_Left_Pose}
   n3 = "float >= float"(A=Get RollAngle(), B=5)
   Output Animation Pose(Result=Apply Additive(
     Base=Blend Poses by bool(BlendPose_0=@n1, BlendPose_1=@n2, BlendTime_0=0.7, BlendTime_1=0.7, bActiveValue=@n3) {blendType=Linear, transitionType=Inertialization}
     Additive="Sequence Player (additive) 'GrappleGloves_Swing_1hand_ADD'"() {sequence=&GrappleGloves_Swing_1hand_ADD}
    ) {alphaBoolBlend {blendOption=Linear}})
  ]
 }
 Transition {
  path=FullBodyOverride.AnimGraphNode_StateMachine_0.SwingStateMachine.AnimStateNode_4.Swinging.AnimGraphNode_StateMachine_1.Swing.AnimStateTransitionNode_0.Transition
  nodes [Result(bCanEnterTransition=AND Boolean(A=Get IsGroundSwing(), B=Get IsBlendOut() -> (IsBlendOut:bool)))]
 }
 Transition {
  path=FullBodyOverride.AnimGraphNode_StateMachine_0.SwingStateMachine.AnimStateNode_4.Swinging.AnimGraphNode_StateMachine_1.Swing.AnimStateTransitionNode_1.Transition
  nodes [Result(bCanEnterTransition="float >= float"(A=Get PlayerDistanceToGround(), B=90))]
 }
 Transition {
  path=FullBodyOverride.AnimGraphNode_StateMachine_0.SwingStateMachine.AnimStateNode_4.Swinging.AnimGraphNode_StateMachine_1.Swing.AnimStateTransitionNode_12.Transition
  nodes [Result(bCanEnterTransition="float < float"(A=Get PlayerDistanceToGround(), B=150))]
 }
 Transition {
  path=FullBodyOverride.AnimGraphNode_StateMachine_0.SwingStateMachine.AnimStateNode_4.Swinging.AnimGraphNode_StateMachine_1.Swing.AnimStateTransitionNode_17.Transition
  nodes [Result(bCanEnterTransition=AND Boolean(A=Get IsHanging() -> (IsHanging:bool), B=Get IsNotAttachedToCliff() -> (IsNotAttachedToCliff:bool)))]
 }
 Transition {
  path=FullBodyOverride.AnimGraphNode_StateMachine_0.SwingStateMachine.AnimStateNode_4.Swinging.AnimGraphNode_StateMachine_1.Swing.AnimStateTransitionNode_19.Transition
  nodes [Result(bCanEnterTransition=Get IsHanging())]
 }
 Transition {
  path=FullBodyOverride.AnimGraphNode_StateMachine_0.SwingStateMachine.AnimStateNode_4.Swinging.AnimGraphNode_StateMachine_1.Swing.AnimStateTransitionNode_2.Transition
  nodes [Result(bCanEnterTransition=Get IsSwingNearGroundToSkid() -> (IsSwingNearGroundToSkid:bool))]
 }
 Transition {
  path=FullBodyOverride.AnimGraphNode_StateMachine_0.SwingStateMachine.AnimStateNode_4.Swinging.AnimGraphNode_StateMachine_1.Swing.AnimStateTransitionNode_20.Transition
  nodes [Result(bCanEnterTransition=Get IsHangToSwingInAir() -> (IsHangToSwingInAir:bool))]
 }
 Transition {
  path=FullBodyOverride.AnimGraphNode_StateMachine_0.SwingStateMachine.AnimStateNode_4.Swinging.AnimGraphNode_StateMachine_1.Swing.AnimStateTransitionNode_3.Transition
  nodes [Result(bCanEnterTransition=true)]
 }
 Transition {
  path=FullBodyOverride.AnimGraphNode_StateMachine_0.SwingStateMachine.AnimStateNode_4.Swinging.AnimGraphNode_StateMachine_1.Swing.AnimStateTransitionNode_4.Transition
  nodes [Result(bCanEnterTransition="float >= float"(A=Get PlayerDistanceToGround(), B=300))]
 }
 Transition {
  path=FullBodyOverride.AnimGraphNode_StateMachine_0.SwingStateMachine.AnimStateNode_4.Swinging.AnimGraphNode_StateMachine_1.Swing.AnimStateTransitionNode_5.Transition
  nodes [
   Result(bCanEnterTransition=Get IsBlendOut())
   AND Boolean(A=false, B=NOT Boolean(A=Get IsGroundSwing()))
  ]
 }
 Transition {
  path=FullBodyOverride.AnimGraphNode_StateMachine_0.SwingStateMachine.AnimStateNode_4.Swinging.AnimGraphNode_StateMachine_1.Swing.AnimStateTransitionNode_6.Transition
  nodes [
   n1 = Get IsGroundSwing()
   Result(bCanEnterTransition=AND Boolean!(A=Get IsBlendOut(), B=Get IsWithin100UnitsOfGround() -> (IsWithin100UnitsOfGround:bool), C:bool=NOT Boolean(A=@n1)) -> (ReturnValue))
  ]
 }
 Transition {
  path=FullBodyOverride.AnimGraphNode_StateMachine_0.SwingStateMachine.AnimStateNode_4.Swinging.AnimGraphNode_StateMachine_1.Swing.AnimStateTransitionNode_8.Transition
  nodes [Result(bCanEnterTransition=Get IsHangToSwingNearGround() -> (IsHangToSwingNearGround:bool))]
 }
 SwingRelease {
  path=FullBodyOverride.AnimGraphNode_StateMachine_0.SwingStateMachine.AnimStateNode_5.SwingRelease
  nodes [
   AnimGraphNode_StateMachine_0 = SwingRelease() {
    is=title AnimGraphNode_StateMachine
    bAllowConduitEntryStates=true
    bReinitializeOnBecomingRelevant=false
    bSkipFirstUpdateTransition=false
    maxTransitionsPerFrame=0
   }
   n1 = "Transform (Modify) Bone - Bone: Root"(
    ComponentPose=Local To Component(LocalPose=@AnimGraphNode_StateMachine_0)
    bAlphaBoolEnabled=True
    Translation=Get CurrentRootTranslation()
    Rotation=Get CurrentRootRotation()
    Scale="1,1,1"
   ) {
    alphaInputType=Bool
    rotationMode=BMM_Replace
    rotationSpace=BCS_WorldSpace
    translationMode=BMM_Replace
    translationSpace=BCS_WorldSpace
    alphaBoolBlend {blendOption=Linear, blendOutTime=0.5}
    boneToModify {boneName=Root}
   }
   Output Animation Pose(Result=Component To Local(ComponentPose=@n1))
  ]
 }
 SwingRelease {
  entry=Conduit
  path=FullBodyOverride.AnimGraphNode_StateMachine_0.SwingStateMachine.AnimStateNode_5.SwingRelease.AnimGraphNode_StateMachine_0.SwingRelease
  ids=id
  nodes [
   AnimStateConduitNode_0 = Conduit()
   AnimStateNode_0 = Release_Air_Side() {
    is=title AnimStateNode
    stateEntered {guid=#21}
    stateFullyBlended {guid=#22}
    stateLeft {guid=#23}
   }
   AnimStateNode_1 = Release_Air_3() {is=title AnimStateNode}
   AnimStateNode_2 = Release_Air_4() {is=title AnimStateNode}
   AnimStateNode_3 = Release_Air_5() {is=title AnimStateNode}
   AnimStateNode_4 = Release_Air_2() {is=title AnimStateNode}
   AnimStateNode_5 = Release_Air_1() {is=title AnimStateNode}
   AnimStateNode_7 = Release_Ground() {is=title AnimStateNode}
   AnimStateTransitionNode_0 = AnimStateTransitionNode() {
    crossfadeDuration=0.35
    logicType=TLT_Inertialization
    priorityOrder=2
    transitionEnd {guid=#24}
    transitionInterrupt {guid=#25}
    transitionStart {guid=#26}
   }
   AnimStateTransitionNode_1 = AnimStateTransitionNode() {crossfadeDuration=0.35, logicType=TLT_Inertialization, priorityOrder=100}
   AnimStateTransitionNode_10 = AnimStateTransitionNode() {crossfadeDuration=0.3, logicType=TLT_Inertialization}
   AnimStateTransitionNode_2 = AnimStateTransitionNode() {crossfadeDuration=0.35, logicType=TLT_Inertialization, priorityOrder=5}
   AnimStateTransitionNode_3 = AnimStateTransitionNode() {crossfadeDuration=0.35, logicType=TLT_Inertialization, priorityOrder=6}
   AnimStateTransitionNode_4 = AnimStateTransitionNode() {crossfadeDuration=0.35, logicType=TLT_Inertialization, priorityOrder=4}
   AnimStateTransitionNode_5 = AnimStateTransitionNode() {
    crossfadeDuration=0.35
    logicType=TLT_Inertialization
    priorityOrder=3
    transitionEnd {guid=#27}
    transitionInterrupt {guid=#28}
    transitionStart {guid=#29}
   }
  ]
  links [
   {
    dst=Release_Air_1
    rule=FullBodyOverride.AnimGraphNode_StateMachine_0.SwingStateMachine.AnimStateNode_5.SwingRelease.AnimGraphNode_StateMachine_0.SwingRelease.AnimStateTransitionNode_5.Transition
    src=Conduit
    via=AnimStateTransitionNode_5
   }
   {
    dst=Release_Air_2
    rule=FullBodyOverride.AnimGraphNode_StateMachine_0.SwingStateMachine.AnimStateNode_5.SwingRelease.AnimGraphNode_StateMachine_0.SwingRelease.AnimStateTransitionNode_4.Transition
    src=Conduit
    via=AnimStateTransitionNode_4
   }
   {
    dst=Release_Air_3
    rule=FullBodyOverride.AnimGraphNode_StateMachine_0.SwingStateMachine.AnimStateNode_5.SwingRelease.AnimGraphNode_StateMachine_0.SwingRelease.AnimStateTransitionNode_1.Transition
    src=Conduit
    via=AnimStateTransitionNode_1
   }
   {
    dst=Release_Air_4
    rule=FullBodyOverride.AnimGraphNode_StateMachine_0.SwingStateMachine.AnimStateNode_5.SwingRelease.AnimGraphNode_StateMachine_0.SwingRelease.AnimStateTransitionNode_2.Transition
    src=Conduit
    via=AnimStateTransitionNode_2
   }
   {
    dst=Release_Air_5
    rule=FullBodyOverride.AnimGraphNode_StateMachine_0.SwingStateMachine.AnimStateNode_5.SwingRelease.AnimGraphNode_StateMachine_0.SwingRelease.AnimStateTransitionNode_3.Transition
    src=Conduit
    via=AnimStateTransitionNode_3
   }
   {
    dst=Release_Air_Side
    rule=FullBodyOverride.AnimGraphNode_StateMachine_0.SwingStateMachine.AnimStateNode_5.SwingRelease.AnimGraphNode_StateMachine_0.SwingRelease.AnimStateTransitionNode_0.Transition
    src=Conduit
    via=AnimStateTransitionNode_0
   }
   {
    dst=Release_Ground
    rule=FullBodyOverride.AnimGraphNode_StateMachine_0.SwingStateMachine.AnimStateNode_5.SwingRelease.AnimGraphNode_StateMachine_0.SwingRelease.AnimStateTransitionNode_10.Transition
    src=Conduit
    via=AnimStateTransitionNode_10
   }
  ]
 }
 Conduit {
  path=FullBodyOverride.AnimGraphNode_StateMachine_0.SwingStateMachine.AnimStateNode_5.SwingRelease.AnimGraphNode_StateMachine_0.SwingRelease.AnimStateConduitNode_0.Conduit
  nodes [Result(bCanEnterTransition=true)]
 }
 Release_Air_Side {
  path=FullBodyOverride.AnimGraphNode_StateMachine_0.SwingStateMachine.AnimStateNode_5.SwingRelease.AnimGraphNode_StateMachine_0.SwingRelease.AnimStateNode_0.Release_Air_Side
  nodes [
   AnimGraphNode_StateMachine_0 = ReleaseSide() {is=title AnimGraphNode_StateMachine, bAllowConduitEntryStates=true}
   Output Animation Pose(Result=@AnimGraphNode_StateMachine_0)
  ]
 }
 ReleaseSide {
  entry=Conduit
  path=FullBodyOverride.AnimGraphNode_StateMachine_0.SwingStateMachine.AnimStateNode_5.SwingRelease.AnimGraphNode_StateMachine_0.SwingRelease.AnimStateNode_0.Release_Air_Side.AnimGraphNode_StateMachine_0.ReleaseSide
  ids=id
  nodes [
   AnimStateConduitNode_0 = Conduit()
   AnimStateNode_0 = ReleaseLeft() {is=title AnimStateNode}
   AnimStateNode_1 = ReleaseRight() {is=title AnimStateNode}
   AnimStateTransitionNode_0 = AnimStateTransitionNode() {logicType=TLT_Inertialization}
   AnimStateTransitionNode_1 = AnimStateTransitionNode() {logicType=TLT_Inertialization, priorityOrder=100}
  ]
  links [
   {
    dst=ReleaseLeft
    rule=FullBodyOverride.AnimGraphNode_StateMachine_0.SwingStateMachine.AnimStateNode_5.SwingRelease.AnimGraphNode_StateMachine_0.SwingRelease.AnimStateNode_0.Release_Air_Side.AnimGraphNode_StateMachine_0.ReleaseSide.AnimStateTransitionNode_0.Transition
    src=Conduit
    via=AnimStateTransitionNode_0
   }
   {
    dst=ReleaseRight
    rule=FullBodyOverride.AnimGraphNode_StateMachine_0.SwingStateMachine.AnimStateNode_5.SwingRelease.AnimGraphNode_StateMachine_0.SwingRelease.AnimStateNode_0.Release_Air_Side.AnimGraphNode_StateMachine_0.ReleaseSide.AnimStateTransitionNode_1.Transition
    src=Conduit
    via=AnimStateTransitionNode_1
   }
  ]
 }
 Conduit {
  path=FullBodyOverride.AnimGraphNode_StateMachine_0.SwingStateMachine.AnimStateNode_5.SwingRelease.AnimGraphNode_StateMachine_0.SwingRelease.AnimStateNode_0.Release_Air_Side.AnimGraphNode_StateMachine_0.ReleaseSide.AnimStateConduitNode_0.Conduit
  nodes [Result(bCanEnterTransition=true)]
 }
 ReleaseLeft {
  path=FullBodyOverride.AnimGraphNode_StateMachine_0.SwingStateMachine.AnimStateNode_5.SwingRelease.AnimGraphNode_StateMachine_0.SwingRelease.AnimStateNode_0.Release_Air_Side.AnimGraphNode_StateMachine_0.ReleaseSide.AnimStateNode_0.ReleaseLeft
  nodes [Output Animation Pose(Result=Sequence Player 'GrappleGloves_Release_45_Left'() {bLoopAnimation=false, sequence=&/GrappleGloves_Release_45_Left})]
 }
 ReleaseRight {
  path=FullBodyOverride.AnimGraphNode_StateMachine_0.SwingStateMachine.AnimStateNode_5.SwingRelease.AnimGraphNode_StateMachine_0.SwingRelease.AnimStateNode_0.Release_Air_Side.AnimGraphNode_StateMachine_0.ReleaseSide.AnimStateNode_1.ReleaseRight
  nodes [Output Animation Pose(Result=Sequence Player 'GrappleGloves_Release_45_Right'() {bLoopAnimation=false, sequence=&/GrappleGloves_Release_45_Right})]
 }
 Transition {
  path=FullBodyOverride.AnimGraphNode_StateMachine_0.SwingStateMachine.AnimStateNode_5.SwingRelease.AnimGraphNode_StateMachine_0.SwingRelease.AnimStateNode_0.Release_Air_Side.AnimGraphNode_StateMachine_0.ReleaseSide.AnimStateTransitionNode_0.Transition
  nodes [Result(bCanEnterTransition="float < float"(A=Get RollAngle()))]
 }
 Transition {
  path=FullBodyOverride.AnimGraphNode_StateMachine_0.SwingStateMachine.AnimStateNode_5.SwingRelease.AnimGraphNode_StateMachine_0.SwingRelease.AnimStateNode_0.Release_Air_Side.AnimGraphNode_StateMachine_0.ReleaseSide.AnimStateTransitionNode_1.Transition
  nodes [Result(bCanEnterTransition=true)]
 }
 Release_Air_3 {
  path=FullBodyOverride.AnimGraphNode_StateMachine_0.SwingStateMachine.AnimStateNode_5.SwingRelease.AnimGraphNode_StateMachine_0.SwingRelease.AnimStateNode_1.Release_Air_3
  nodes [
   n1 = Sequence Player() {
    becomeRelevantFunction {bSelfContext=true, memberGuid=#30, memberName=SwingReleaseSelection_0}
    node {becomeRelevantFunction {functionName=SwingReleaseSelection_0}}
   }
   Output Animation Pose(Result=@n1)
  ]
 }
 Release_Air_4 {
  path=FullBodyOverride.AnimGraphNode_StateMachine_0.SwingStateMachine.AnimStateNode_5.SwingRelease.AnimGraphNode_StateMachine_0.SwingRelease.AnimStateNode_2.Release_Air_4
  nodes [
   AnimGraphNode_StateMachine_0 = SwingRelease_45() {is=title AnimGraphNode_StateMachine, bAllowConduitEntryStates=true}
   Output Animation Pose(Result=@AnimGraphNode_StateMachine_0)
  ]
 }
 SwingRelease_45 {
  entry=Conduit
  path=FullBodyOverride.AnimGraphNode_StateMachine_0.SwingStateMachine.AnimStateNode_5.SwingRelease.AnimGraphNode_StateMachine_0.SwingRelease.AnimStateNode_2.Release_Air_4.AnimGraphNode_StateMachine_0.SwingRelease_45
  ids=id
  nodes [
   AnimStateConduitNode_0 = Conduit()
   AnimStateNode_0 = ReleaseLowToGround() {is=title AnimStateNode}
   AnimStateNode_1 = ReleaseHigh() {is=title AnimStateNode}
   AnimStateTransitionNode_0 = AnimStateTransitionNode() {crossfadeDuration=0.35, logicType=TLT_Inertialization, priorityOrder=100}
   AnimStateTransitionNode_1 = AnimStateTransitionNode() {crossfadeDuration=0.35, logicType=TLT_Inertialization}
  ]
  links [
   {
    dst=ReleaseHigh
    rule=FullBodyOverride.AnimGraphNode_StateMachine_0.SwingStateMachine.AnimStateNode_5.SwingRelease.AnimGraphNode_StateMachine_0.SwingRelease.AnimStateNode_2.Release_Air_4.AnimGraphNode_StateMachine_0.SwingRelease_45.AnimStateTransitionNode_1.Transition
    src=Conduit
    via=AnimStateTransitionNode_1
   }
   {
    dst=ReleaseLowToGround
    rule=FullBodyOverride.AnimGraphNode_StateMachine_0.SwingStateMachine.AnimStateNode_5.SwingRelease.AnimGraphNode_StateMachine_0.SwingRelease.AnimStateNode_2.Release_Air_4.AnimGraphNode_StateMachine_0.SwingRelease_45.AnimStateTransitionNode_0.Transition
    src=Conduit
    via=AnimStateTransitionNode_0
   }
  ]
 }
 Conduit {
  path=FullBodyOverride.AnimGraphNode_StateMachine_0.SwingStateMachine.AnimStateNode_5.SwingRelease.AnimGraphNode_StateMachine_0.SwingRelease.AnimStateNode_2.Release_Air_4.AnimGraphNode_StateMachine_0.SwingRelease_45.AnimStateConduitNode_0.Conduit
  nodes [Result(bCanEnterTransition=true)]
 }
 ReleaseLowToGround {
  path=FullBodyOverride.AnimGraphNode_StateMachine_0.SwingStateMachine.AnimStateNode_5.SwingRelease.AnimGraphNode_StateMachine_0.SwingRelease.AnimStateNode_2.Release_Air_4.AnimGraphNode_StateMachine_0.SwingRelease_45.AnimStateNode_0.ReleaseLowToGround
  nodes [Output Animation Pose(Result=Blend Poses by bool(
     BlendPose_0=Sequence Player 'GrappleGloves_Release_0'() {bLoopAnimation=false, sequence=&/GrappleGloves_Release_0}
     BlendPose_1=Sequence Player 'GrappleGloves_Release_0_B'() {bLoopAnimation=false, sequence=&/GrappleGloves_Release_0_B}
     bActiveValue=Get IsRandomSwingReleaseBool() -> (IsRandomSwingReleaseBool:bool)
    ) {blendType=Linear, transitionType=Inertialization})]
 }
 ReleaseHigh {
  path=FullBodyOverride.AnimGraphNode_StateMachine_0.SwingStateMachine.AnimStateNode_5.SwingRelease.AnimGraphNode_StateMachine_0.SwingRelease.AnimStateNode_2.Release_Air_4.AnimGraphNode_StateMachine_0.SwingRelease_45.AnimStateNode_1.ReleaseHigh
  nodes [
   n1 = Sequence Player() {
    becomeRelevantFunction {bSelfContext=true, memberGuid=#31, memberName=SwingReleaseSelection_45}
    node {becomeRelevantFunction {functionName=OnNewFunction}}
   }
   Output Animation Pose(Result=@n1)
  ]
 }
 Transition {
  path=FullBodyOverride.AnimGraphNode_StateMachine_0.SwingStateMachine.AnimStateNode_5.SwingRelease.AnimGraphNode_StateMachine_0.SwingRelease.AnimStateNode_2.Release_Air_4.AnimGraphNode_StateMachine_0.SwingRelease_45.AnimStateTransitionNode_0.Transition
  nodes [Result(bCanEnterTransition=true)]
 }
 Transition {
  path=FullBodyOverride.AnimGraphNode_StateMachine_0.SwingStateMachine.AnimStateNode_5.SwingRelease.AnimGraphNode_StateMachine_0.SwingRelease.AnimStateNode_2.Release_Air_4.AnimGraphNode_StateMachine_0.SwingRelease_45.AnimStateTransitionNode_1.Transition
  nodes [Result(bCanEnterTransition="float >= float"(A=Get PlayerDistanceToGround(), B=500))]
 }
 Release_Air_5 {
  path=FullBodyOverride.AnimGraphNode_StateMachine_0.SwingStateMachine.AnimStateNode_5.SwingRelease.AnimGraphNode_StateMachine_0.SwingRelease.AnimStateNode_3.Release_Air_5
  nodes [
   AnimGraphNode_StateMachine_0 = SwingRelease_90() {is=title AnimGraphNode_StateMachine, bAllowConduitEntryStates=true}
   Output Animation Pose(Result=@AnimGraphNode_StateMachine_0)
  ]
 }
 SwingRelease_90 {
  entry=Conduit
  path=FullBodyOverride.AnimGraphNode_StateMachine_0.SwingStateMachine.AnimStateNode_5.SwingRelease.AnimGraphNode_StateMachine_0.SwingRelease.AnimStateNode_3.Release_Air_5.AnimGraphNode_StateMachine_0.SwingRelease_90
  ids=id
  nodes [
   AnimStateConduitNode_0 = Conduit()
   AnimStateNode_0 = ReleaseLowToGround()
   AnimStateNode_1 = ReleaseHigh()
   AnimStateTransitionNode_0 = AnimStateTransitionNode() {crossfadeDuration=0.35, logicType=TLT_Inertialization, priorityOrder=100}
   AnimStateTransitionNode_1 = AnimStateTransitionNode() {crossfadeDuration=0.35, logicType=TLT_Inertialization}
  ]
  links [
   {
    dst=ReleaseHigh
    rule=FullBodyOverride.AnimGraphNode_StateMachine_0.SwingStateMachine.AnimStateNode_5.SwingRelease.AnimGraphNode_StateMachine_0.SwingRelease.AnimStateNode_3.Release_Air_5.AnimGraphNode_StateMachine_0.SwingRelease_90.AnimStateTransitionNode_1.Transition
    src=Conduit
    via=AnimStateTransitionNode_1
   }
   {
    dst=ReleaseLowToGround
    rule=FullBodyOverride.AnimGraphNode_StateMachine_0.SwingStateMachine.AnimStateNode_5.SwingRelease.AnimGraphNode_StateMachine_0.SwingRelease.AnimStateNode_3.Release_Air_5.AnimGraphNode_StateMachine_0.SwingRelease_90.AnimStateTransitionNode_0.Transition
    src=Conduit
    via=AnimStateTransitionNode_0
   }
  ]
 }
 Conduit {
  path=FullBodyOverride.AnimGraphNode_StateMachine_0.SwingStateMachine.AnimStateNode_5.SwingRelease.AnimGraphNode_StateMachine_0.SwingRelease.AnimStateNode_3.Release_Air_5.AnimGraphNode_StateMachine_0.SwingRelease_90.AnimStateConduitNode_0.Conduit
  nodes [Result(bCanEnterTransition=true)]
 }
 ReleaseLowToGround {
  path=FullBodyOverride.AnimGraphNode_StateMachine_0.SwingStateMachine.AnimStateNode_5.SwingRelease.AnimGraphNode_StateMachine_0.SwingRelease.AnimStateNode_3.Release_Air_5.AnimGraphNode_StateMachine_0.SwingRelease_90.AnimStateNode_0.ReleaseLowToGround
  nodes [Output Animation Pose(Result=Blend Poses by bool(
     BlendPose_0=Sequence Player 'GrappleGloves_Release_0_B'() {sequence=&GrappleGloves_Release_0_B}
     BlendPose_1=Sequence Player 'GrappleGloves_Release_-45A'() {bLoopAnimation=false, sequence=&/GrappleGloves_Release_-45A}
     bActiveValue=Get IsRandomSwingReleaseBool()
    ) {blendType=Linear, transitionType=Inertialization})]
 }
 ReleaseHigh {
  path=FullBodyOverride.AnimGraphNode_StateMachine_0.SwingStateMachine.AnimStateNode_5.SwingRelease.AnimGraphNode_StateMachine_0.SwingRelease.AnimStateNode_3.Release_Air_5.AnimGraphNode_StateMachine_0.SwingRelease_90.AnimStateNode_1.ReleaseHigh
  nodes [
   n1 = Sequence Player() {
    becomeRelevantFunction {bSelfContext=true, memberGuid=#32, memberName=SwingReleaseSelection_90}
    node {becomeRelevantFunction {functionName=SwingReleaseSelection_90}}
   }
   Output Animation Pose(Result=@n1)
  ]
 }
 Transition {
  path=FullBodyOverride.AnimGraphNode_StateMachine_0.SwingStateMachine.AnimStateNode_5.SwingRelease.AnimGraphNode_StateMachine_0.SwingRelease.AnimStateNode_3.Release_Air_5.AnimGraphNode_StateMachine_0.SwingRelease_90.AnimStateTransitionNode_0.Transition
  nodes [Result(bCanEnterTransition=true)]
 }
 Transition {
  path=FullBodyOverride.AnimGraphNode_StateMachine_0.SwingStateMachine.AnimStateNode_5.SwingRelease.AnimGraphNode_StateMachine_0.SwingRelease.AnimStateNode_3.Release_Air_5.AnimGraphNode_StateMachine_0.SwingRelease_90.AnimStateTransitionNode_1.Transition
  nodes [Result(bCanEnterTransition="float >= float"(A=Get PlayerDistanceToGround(), B=500))]
 }
 Release_Air_2 {
  path=FullBodyOverride.AnimGraphNode_StateMachine_0.SwingStateMachine.AnimStateNode_5.SwingRelease.AnimGraphNode_StateMachine_0.SwingRelease.AnimStateNode_4.Release_Air_2
  nodes [Output Animation Pose(Result=Sequence Player 'GrappleGloves_Release_-45A'() {sequence=&GrappleGloves_Release_-45A})]
 }
 Release_Air_1 {
  path=FullBodyOverride.AnimGraphNode_StateMachine_0.SwingStateMachine.AnimStateNode_5.SwingRelease.AnimGraphNode_StateMachine_0.SwingRelease.AnimStateNode_5.Release_Air_1
  nodes [
   Blend Poses by bool(bActiveValue=Get IsRandomSwingReleaseBool()) {blendType=Linear, transitionType=Inertialization}
   n1 = Sequence Player() {
    becomeRelevantFunction {bSelfContext=true, memberGuid=#33, memberName=SwingReleaseSelection_-90}
    node {becomeRelevantFunction {functionName=SwingReleaseSelection_-90}}
   }
   Output Animation Pose(Result=@n1)
  ]
 }
 Release_Ground {
  path=FullBodyOverride.AnimGraphNode_StateMachine_0.SwingStateMachine.AnimStateNode_5.SwingRelease.AnimGraphNode_StateMachine_0.SwingRelease.AnimStateNode_7.Release_Ground
  nodes [Output Animation Pose(Result=Sequence Player 'GrappleGloves_Release_-90_Ground'() {sequence=&/GrappleGloves_Release_-90_Ground})]
 }
 Transition {
  path=FullBodyOverride.AnimGraphNode_StateMachine_0.SwingStateMachine.AnimStateNode_5.SwingRelease.AnimGraphNode_StateMachine_0.SwingRelease.AnimStateTransitionNode_0.Transition
  nodes [Result(bCanEnterTransition=Get AngledRelease() -> (AngledRelease:bool))]
 }
 Transition {
  path=FullBodyOverride.AnimGraphNode_StateMachine_0.SwingStateMachine.AnimStateNode_5.SwingRelease.AnimGraphNode_StateMachine_0.SwingRelease.AnimStateTransitionNode_1.Transition
  nodes [Result(bCanEnterTransition=true)]
 }
 Transition {
  path=FullBodyOverride.AnimGraphNode_StateMachine_0.SwingStateMachine.AnimStateNode_5.SwingRelease.AnimGraphNode_StateMachine_0.SwingRelease.AnimStateTransitionNode_10.Transition
  nodes [
   Result(bCanEnterTransition=Get IsSwingReleaseSkid() -> (IsSwingReleaseSkid:bool))
   OR Boolean(A=false, B=Get IsWithin100UnitsOfGround())
  ]
 }
 Transition {
  path=FullBodyOverride.AnimGraphNode_StateMachine_0.SwingStateMachine.AnimStateNode_5.SwingRelease.AnimGraphNode_StateMachine_0.SwingRelease.AnimStateTransitionNode_2.Transition
  nodes [Result(bCanEnterTransition=Get IsSwingRelease4() -> (IsSwingRelease4:bool))]
 }
 Transition {
  path=FullBodyOverride.AnimGraphNode_StateMachine_0.SwingStateMachine.AnimStateNode_5.SwingRelease.AnimGraphNode_StateMachine_0.SwingRelease.AnimStateTransitionNode_3.Transition
  nodes [Result(bCanEnterTransition=Get IsSwingRelease5() -> (IsSwingRelease5:bool))]
 }
 Transition {
  path=FullBodyOverride.AnimGraphNode_StateMachine_0.SwingStateMachine.AnimStateNode_5.SwingRelease.AnimGraphNode_StateMachine_0.SwingRelease.AnimStateTransitionNode_4.Transition
  nodes [Result(bCanEnterTransition=Get IsSwingRelease2() -> (IsSwingRelease2:bool))]
 }
 Transition {
  path=FullBodyOverride.AnimGraphNode_StateMachine_0.SwingStateMachine.AnimStateNode_5.SwingRelease.AnimGraphNode_StateMachine_0.SwingRelease.AnimStateTransitionNode_5.Transition
  nodes [Result(bCanEnterTransition=Get IsSwingRelease1() -> (IsSwingRelease1:bool))]
 }
 Fall {
  path=FullBodyOverride.AnimGraphNode_StateMachine_0.SwingStateMachine.AnimStateNode_6.Fall
  nodes [
   AnimGraphNode_StateMachine_0 = FallState() {is=title AnimGraphNode_StateMachine}
   Output Animation Pose(Result=@AnimGraphNode_StateMachine_0)
  ]
 }
 FallState {
  entry=FallStart
  path=FullBodyOverride.AnimGraphNode_StateMachine_0.SwingStateMachine.AnimStateNode_6.Fall.AnimGraphNode_StateMachine_0.FallState
  ids=id
  nodes [
   AnimStateNode_0 = FallStart() {is=title AnimStateNode}
   AnimStateNode_1 = FallLoop() {is=title AnimStateNode}
   AnimStateTransitionNode_0 = AnimStateTransitionNode() {bAutomaticRuleBasedOnSequencePlayerInState=true, crossfadeDuration=0.6, logicType=TLT_Inertialization}
  ]
  links [{
    dst=FallLoop
    rule=FullBodyOverride.AnimGraphNode_StateMachine_0.SwingStateMachine.AnimStateNode_6.Fall.AnimGraphNode_StateMachine_0.FallState.AnimStateTransitionNode_0.Transition
    src=FallStart
    via=AnimStateTransitionNode_0
   }]
 }
 FallStart {
  path=FullBodyOverride.AnimGraphNode_StateMachine_0.SwingStateMachine.AnimStateNode_6.Fall.AnimGraphNode_StateMachine_0.FallState.AnimStateNode_0.FallStart
  nodes [Output Animation Pose(Result=Sequence Player 'GrappleGloves_Release_Fall_Into'() {sequence=&/GrappleGloves_Release_Fall_Into})]
 }
 FallLoop {
  path=FullBodyOverride.AnimGraphNode_StateMachine_0.SwingStateMachine.AnimStateNode_6.Fall.AnimGraphNode_StateMachine_0.FallState.AnimStateNode_1.FallLoop
  nodes [Output Animation Pose(Result=Apply Additive(Base=Sequence Player 'GrappleGloves_Release_Fall_Loop'() {sequence=&/GrappleGloves_Release_Fall_Loop}) {alphaBoolBlend {blendOption=Linear}})]
 }
 Transition {
  path=FullBodyOverride.AnimGraphNode_StateMachine_0.SwingStateMachine.AnimStateNode_6.Fall.AnimGraphNode_StateMachine_0.FallState.AnimStateTransitionNode_0.Transition
  nodes [Result(bCanEnterTransition=true)]
 }
 Transition {
  path=FullBodyOverride.AnimGraphNode_StateMachine_0.SwingStateMachine.AnimStateTransitionNode_0.Transition
  nodes [Result(bCanEnterTransition="float < float"(A="Get Relevant Anim Time Remaining (SwingRelease)"!() -> (ReturnValue:float) {is=title K2Node_AnimGetter}, B=0.05))]
 }
 Transition {
  path=FullBodyOverride.AnimGraphNode_StateMachine_0.SwingStateMachine.AnimStateTransitionNode_1.Transition
  nodes [Result(bCanEnterTransition=Get IsWallHit() -> (IsWallHit:bool))]
 }
 Transition {
  path=FullBodyOverride.AnimGraphNode_StateMachine_0.SwingStateMachine.AnimStateTransitionNode_15.Transition
  nodes [Result(bCanEnterTransition=NOT Boolean(A=Get IsAttached() -> (IsAttached:bool)))]
 }
 Transition {
  path=FullBodyOverride.AnimGraphNode_StateMachine_0.SwingStateMachine.AnimStateTransitionNode_16.Transition
  nodes [Result(bCanEnterTransition=true)]
 }
 Transition {
  path=FullBodyOverride.AnimGraphNode_StateMachine_0.SwingStateMachine.AnimStateTransitionNode_17.Transition
  nodes [Result(bCanEnterTransition=Get AllowNonRelease() -> (AllowNonRelease:bool))]
 }
 Transition {
  path=FullBodyOverride.AnimGraphNode_StateMachine_0.SwingStateMachine.AnimStateTransitionNode_18.Transition
  nodes [Result(bCanEnterTransition=Get IsAttached())]
 }
 Transition {
  path=FullBodyOverride.AnimGraphNode_StateMachine_0.SwingStateMachine.AnimStateTransitionNode_19.Transition
  nodes [Result(bCanEnterTransition=Get IsAttached())]
 }
 Transition {
  path=FullBodyOverride.AnimGraphNode_StateMachine_0.SwingStateMachine.AnimStateTransitionNode_2.Transition
  nodes [Result(bCanEnterTransition=Get IsAttached())]
 }
 Transition {
  path=FullBodyOverride.AnimGraphNode_StateMachine_0.SwingStateMachine.AnimStateTransitionNode_20.Transition
  nodes [Result(bCanEnterTransition=Get IsAttached())]
 }
 Transition {
  path=FullBodyOverride.AnimGraphNode_StateMachine_0.SwingStateMachine.AnimStateTransitionNode_4.Transition
  nodes [Result(bCanEnterTransition=true)]
 }
 PreLegIK {nodes [
   Input Pose()
   Output Pose() {layerGroup=PlayerOverride}
  ]}
 SplitVector {nodes [
   n1 = SplitVector!() -> (then, Vec:Vector=~) {is=title K2Node_FunctionEntry}
   n2 = Vector Length(A=@n1)
   > Return Node!(execute, Magnitude:double=@n2, Direction:Vector=vector / vector(A:Vector=@n1, B=@n2) -> (ReturnValue:Vector))
  ]}
 SwingReleaseSelection_-90 {nodes [
   n1 = SwingReleaseSelection_-90!() -> (then, Context:"Anim Update Context Structure (by ref)", Node:"Anim Node Reference Structure (by ref)") {is=title K2Node_FunctionEntry}
   > SetSequence!(
    execute
    SequencePlayer:"Sequence Player Reference Structure (by ref)"=Convert to Sequence Player!(Node:"Anim Node Reference Structure (by ref)"=@n1.Node) -> (SequencePlayer:Sequence Player Reference Structure, Result:bool)
    Sequence:Anim Sequence Base Object Reference="Get (a copy)"(
     Array:Array of Anim Sequence Base Object References=Get ReleaseAnim_-90() -> (ReleaseAnim_-90:Array of Anim Sequence Base Object References)
     Dimension 1=Get CurrentReleaseAnims_-90() -> (CurrentReleaseAnims_-90:int=0)
    ) -> (Output:Anim Sequence Base Object Reference)
   ) -> (then, ReturnValue:Sequence Player Reference Structure)
   > n2 = Increment Int(Value=Get CurrentReleaseAnims_-90() -> (CurrentReleaseAnims_-90=0))
   n3 = Get ReleaseAnim_-90()
   > Branch(Condition="integer >= integer"(A=@n2, B=Length(TargetArray:Array of Anim Sequence Base Object References=@n3)))
   > Set CurrentReleaseAnims_-90(CurrentReleaseAnims_-90:int=0) -> (Output_Get:int=0)
   > Return Node!(execute)
  ]}
 SwingReleaseSelection_0 {nodes [
   n1 = SwingReleaseSelection_0!() -> (then, Context:"Anim Update Context Structure (by ref)", Node:"Anim Node Reference Structure (by ref)") {is=title K2Node_FunctionEntry}
   > SetSequence(
    SequencePlayer=Convert to Sequence Player(Node=@n1.Node)
    Sequence="Get (a copy)"(
     Array=Get ReleaseAnim_0() -> (ReleaseAnim_0:Array of Anim Sequence Base Object References)
     Dimension 1=Get CurrentReleaseAnims_0() -> (CurrentReleaseAnims_0:int=0)
    )
   )
   > n2 = Increment Int(Value=Get CurrentReleaseAnims_0() -> (CurrentReleaseAnims_0=0))
   n3 = Get ReleaseAnim_0()
   > n4 = Branch(Condition="integer >= integer"(A=@n2, B=Length(TargetArray=@n3)))
   > n5 = Set CurrentReleaseAnims_0(CurrentReleaseAnims_0:int=0) -> (Output_Get:int=0)
   Return Node(execute=[@n5.then, @n4.else])
  ]}
 SwingReleaseSelection_45 {nodes [
   n1 = SwingReleaseSelection_45!() -> (then, Context:"Anim Update Context Structure (by ref)", Node:"Anim Node Reference Structure (by ref)") {is=title K2Node_FunctionEntry}
   > SetSequence(
    SequencePlayer=Convert to Sequence Player(Node=@n1.Node)
    Sequence="Get (a copy)"(
     Array=Get ReleaseAnim_45() -> (ReleaseAnim_45:Array of Anim Sequence Base Object References)
     Dimension 1=Get CurrentReleaseAnims_45() -> (CurrentReleaseAnims_45:int=0)
    )
   )
   > n2 = Increment Int(Value=Get CurrentReleaseAnims_45() -> (CurrentReleaseAnims_45=0))
   n3 = Get ReleaseAnim_45()
   > Branch(Condition="integer >= integer"(A=@n2, B=Length(TargetArray=@n3)))
   > Set CurrentReleaseAnims_45(CurrentReleaseAnims_45:int=0) -> (Output_Get:int=0)
   > Return Node()
  ]}
 SwingReleaseSelection_90 {nodes [
   n1 = SwingReleaseSelection_90!() -> (then, Context:"Anim Update Context Structure (by ref)", Node:"Anim Node Reference Structure (by ref)") {is=title K2Node_FunctionEntry}
   > SetSequence(
    SequencePlayer=Convert to Sequence Player(Node=@n1.Node)
    Sequence="Get (a copy)"(
     Array=Get ReleaseAnim_90() -> (ReleaseAnim_90:Array of Anim Sequence Base Object References)
     Dimension 1=Get CurrentReleaseAnims_90() -> (CurrentReleaseAnims_90:int=0)
    )
   )
   > n2 = Increment Int(Value=Get CurrentReleaseAnims_90() -> (CurrentReleaseAnims_90=0))
   n3 = Get ReleaseAnim_90()
   > Branch(Condition="integer >= integer"(A=@n2, B=Length(TargetArray=@n3)))
   > Set CurrentReleaseAnims_90(CurrentReleaseAnims_90:int=0) -> (Output_Get:int=0)
   > Return Node()
  ]}
 UpperBody {nodes [
   Input Pose() {name=InPoseUpperBody}
   Output Pose() {layerGroup=PlayerOverride}
  ]}
]
variables [
 AllowNonRelease {type=boolean}
 AngledRelease {type=boolean}
 Bound {type=boolean}
 CurrentReleaseAnims_-90 {type=integer}
 CurrentReleaseAnims_0 {type=integer}
 CurrentReleaseAnims_45 {type=integer}
 CurrentReleaseAnims_90 {type=integer}
 IsWallHit {type=boolean}
 ReleaseAnim_-90 {
  type="array<AnimSequenceBase>"
  default [
   &GrappleGloves_Release_-45A
   &GrappleGloves_Release_-45A
  ]
 }
 ReleaseAnim_0 {
  type="array<AnimSequenceBase>"
  default [
   &/GrappleGloves_Release_0_A
   &GrappleGloves_Release_0_B
   &GrappleGloves_Release_0_A
  ]
 }
 ReleaseAnim_45 {
  type="array<AnimSequenceBase>"
  default [
   &/GrappleGloves_Release_45_B
   &/GrappleGloves_Release_45_C
   &GrappleGloves_Release_0_A
   &/GrappleGloves_Release_90_B
   &GrappleGloves_Release_45_C
  ]
 }
 ReleaseAnim_90 {
  type="array<AnimSequenceBase>"
  default [
   &GrappleGloves_Release_0_B
   &GrappleGloves_Release_45_C
   &GrappleGloves_Release_90_B
  ]
 }
 Weapon {type=B_Weapon_GrappleParent_C, default=null}
]
