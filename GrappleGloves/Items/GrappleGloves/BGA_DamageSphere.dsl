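; BGA_DamageSphere is a damage-catching actor for the Grapple Gloves that makes the grapple fall when it is hit.
; On the first server damage event, it calls Fall on the instigator pawn's current grapple weapon, then disables collision and destroys itself.
; A separate ApplyDamage event, not called here, applies the TransferDamage gameplay effect to a target with SetByCaller damage and prints a debug string.
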(event EventBeginPlay)

(event EventTick
  (params (DeltaSeconds :type Float%20%28single-precision%29)))

(event Collision|EventActorBeginOverlap
  (params (OtherActor :type Actor%20Object%20Reference)))

(event EventOnDamageServer
  (params (Damage :type Float%20%28single-precision%29) (DamageTags :type Gameplay%20Tag%20Container%20Structure%20%28by%20ref%29) (Momentum :type Vector :default "0, 0, 0") (HitInfo :type Hit%20Result%20Structure%20%28by%20ref%29) (InstigatedBy :type Controller%20Object%20Reference) (DamageCauser :type Actor%20Object%20Reference) (EffectContext :type Gameplay%20Effect%20Context%20Handle%20Structure))
  (if (not (Variables|Default|GetDone))
    (bind _asplayer_pawn_athena (Utilities|Casting|CastToPlayerPawn_Athena (Variables|Actor|GetInstigator)))
    (bind _asb_weapon_grapple_parent (Utilities|Casting|CastToB_Weapon_GrappleParent (Weapon|GetCurrentWeapon _asplayer_pawn_athena)))
    (Class|BWeaponGrappleParent|Fall _asb_weapon_grapple_parent)
    (Variables|Default|SetDone true)
    (Collision|SetActorEnableCollision false)
    (Actor|DestroyActor)))

(event Custom|ApplyDamage
  (params (Damage :type Float%20%28double-precision%29) (DamageTarget :type Actor%20Object%20Reference) (Instigator :type Controller%20Object%20Reference) (DamageSource :type Actor%20Object%20Reference))
  (bind _returnvalue (Pawn|GetControlledPawn Instigator))
  (bind _returnvalue_2 (Ability|GetFortAbilitySystemComponent Instigator))
  (bind _returnvalue_3 (GameplayEffects|MakeFortEffectContext _returnvalue_2 0 _returnvalue DamageSource 0))
  (Development|PrintString (Utilities|GetDisplayName DamageSource) true true "(R=0.000000,G=0.660000,B=1.000000,A=1.000000)" 1e+01)
  (bind _returnvalue_4 (Ability|GameplayEffect|AssignTagSetbyCallerMagnitude (GameplayEffects|MakeOutgoingSpec _returnvalue_2 "/GrappleGloves/Items/GrappleGloves/GE_GrappleGloves_TransferDamage.GE_GrappleGloves_TransferDamage_C" 0.0 _returnvalue_3) "(TagName=\"SetByCaller.WeaponDamage\")" Damage))
  (GameplayEffects|FortApplyGameplayEffectSpecToTarget _returnvalue_2 _returnvalue_4 (Ability|GetFortAbilitySystemComponent DamageTarget) 0 _returnvalue DamageSource _returnvalue_3))

(fn ConstructionScript)
