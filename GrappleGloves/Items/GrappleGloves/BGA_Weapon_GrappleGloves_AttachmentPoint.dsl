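; Actor marking a Grapple Gloves attachment point, which plays the hit-point effect when it appears on clients.
; On BeginPlay, after the parent call, only the non-authority side runs a local burst GameplayCue using the GrappleGloves hit-point tag.
; The cue takes the actor's location and passes its Player variable as instigator, effect causer and source object.
; Overlap and tick events are empty, and the construction script only calls the parent.
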(event EventBeginPlay
  (|CallParentFunction)
  (Networking|SwitchHasAuthority
    (:Remote
      (GameplayCue|ExecuteGameplayCue%28Burst%29Local self (Variables|Default|GetGCN_GrappleGloves_HitPoint) (Ability|GameplayCue|MakeGameplayCueParameters :NormalizedMagnitude 0.0 :RawMagnitude 0.0 :Location (Transformation|GetActorLocation) :Normal "0,0,0" :Instigator (Variables|Default|GetPlayer) :EffectCauser (Variables|Default|GetPlayer) :SourceObject (Variables|Default|GetPlayer) :GameplayEffectLevel 1 :AbilityLevel 1 :bReplicateLocationWhenUsingMinimalRepProxy false)))))

(event Collision|EventActorBeginOverlap
  (params (OtherActor :type Actor%20Object%20Reference)))

(event EventTick
  (params (DeltaSeconds :type Float%20%28single-precision%29)))

(fn ConstructionScript
  (|CallParentFunction))
