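; Grapple Gloves weapon Blueprint, extending a parent grapple weapon for rope swinging with escalating momentum states and a landing-based cooldown.
; Attach and detach advance the state, which sets warm/hot gameplay effects, swing-control overrides and cooldown length.
; Server landing checks and a slow-update failsafe multicast a forgiveness timer; the client then blocks firing and requests state reset, with race-condition guards.
; It also handles timeout and dangle audio, grace UI and an optional damage-sphere collider.
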(event Custom|ActuallyApplyCooldown
  ;[1 1 1] CooldownFX
  (|CallParentFunction false)
  (CallFunction|FECooldownFX))

(event Custom|FE_CooldownFX
  ;[1 1 1] CooldownFX
  (GameplayCue|ExecuteGameplayCue%28Burst%29Local self (Variables|Default|GetStartedCooldownCue) (Ability|GameplayCue|MakeGameplayCueParameters 0.0)))

(event Rope|EventOnRopeStartFadeOut)

(event Custom|UnblockFiring_Multicast
  (Variables|Default|SetDisableFiring false)
  (Development|PrintString "FALSE3" true true "(R=0.000000,G=0.660000,B=1.000000,A=1.000000)" 5.0))

(event EventBeginPlay
  ;[1 1 1] Begin Play
  (|CallParentFunction)
  ;;[1 1 1] Overriding GE for specific glove swinging audio
  (Variables|Default|SetGE_InMovementMode (Variables|Default|GetGEMovementModeOverride))
  (Ability|Tasks|WaitGameplayTagAddToActor (Variables|Default|GetPlayer) "(TagName=\"Granted.Athena.GrappleGloves.Warm\")"
    (:then
      (Ability|Tasks|WaitGameplayTagRemoveFromActor (Variables|Default|GetPlayer) "(TagName=\"Granted.Cooldown.GrappleParent\")"
        (:then
          (Ability|Tasks|WaitGameplayTagRemoveFromActor (Variables|Default|GetPlayer) "(TagName=\"Granted.Athena.GrappleGloves.Warm\")"
            (:then
              (bind _asfort_movement_comp_character (Utilities|Casting|CastToFortMovementComp_Character (Variables|/Script/Engine.Character|GetCharacterMovement (Variables|Default|GetPlayer))))
              (Variables|Default|SetMovementComponent _asfort_movement_comp_character)
              (CallFunction|SetState "NewEnumerator0"))
            (:Removed
              (Networking|SwitchHasAuthority
                (:Remote
                  (bind _isswinging (Class|PlayerPawnAthena|IsSwinging (Variables|Default|GetPlayer)))
                  (if (Math|Boolean|ANDBoolean (not _isswinging) (Utilities|Enum|Equal(Enum) (Variables|Default|GetState) "NewEnumerator0") (!= (Variables|Default|GetCachedCooldownTime) 0))
                    (CallFunction|ActuallyApplyCooldown)
                    (Variables|Default|SetPreviousState "NewEnumerator1")
                    (if (ScalableFloat|GetValueasBool (Variables|Default|GetSF_DoReenableFiringFailsafe))
                      (Utilities|Time|SetTimerbyEvent (delegate Custom|Re-enableFiring) (ScalableFloat|GetValueatLevel (Variables|Default|GetSF_ReenableFiringFailsafeTime))))
                    (else
                      (CallFunction|CorrectCooldownRaceCondition)
                      (Variables|Default|SetDisableFiring false)
                      (bind _isswinging_2 (Class|PlayerPawnAthena|IsSwinging (Variables|Default|GetPlayer)))
                      (if _isswinging_2
                        (CallFunction|SetState (Variables|Default|GetPreviousState))))))))) ; avoid race conditions
        )
        (:Removed
          (Variables|Default|SetDisableFiring false)
          (CallFunction|HandleTimeoutAudioStopped))))
    (:Added
      (Variables|Default|SetStartedSwingingTime_ (Utilities|Time|GetGameTimeinSeconds)))))

(event Custom|Re-enableFiring
  ;[1 1 1] Begin Play
  (Variables|Default|SetDisableFiring false))

(event Custom|CorrectCooldownRaceCondition
  ;[1 1 1] Begin Play
  (Variables|Default|SetDisableFiring false)
  (bind _isswinging (Class|PlayerPawnAthena|IsSwinging (Variables|Default|GetPlayer)))
  (if _isswinging
    (CallFunction|SetState (Variables|Default|GetPreviousState))))

(event Custom|TryToCooldown
  ;[1 1 1] Delayed Cooldown
  (Variables|Default|SetStateWhenLanded (Variables|Default|GetState))
  (Networking|SwitchHasAuthority
    (:Authority
      (if (and (GameplayTags|HasMatchingGameplayTag (Variables|Default|GetPlayer) "(TagName=\"MovementMode.Sliding\")") (or (FortWeapon|IsEquipped) (not (ScalableFloat|GetValueasBool (Variables|Default|GetSF_TriggerCooldownOnUnequip)))))
        (Utilities|Time|SetTimerbyEvent (delegate Custom|StartForgivenessTimer) (ScalableFloat|GetValueatLevel (Variables|Design|GetExtraForgivenessWhenSliding)))
        (if (ScalableFloat|GetValueasBool (Variables|Default|GetSF_DecreaseSpeedIfLandedWithSliding))
          (CallFunction|DecreaseSpeedMulticast))
        (elif (ScalableFloat|GetValueasBool (Variables|Default|GetSF_DecreaseSpeedIfLandedWithoutSliding))
          (CallFunction|DecreaseSpeedMulticast)
          (CallFunction|StartForgivenessTimerMulticast)
          (else
            _))))))

(event EventApplyCooldown
  (params (Force :type Boolean))
  (if (or Force (<= (FortWeapon|GetMagazineAmmoCount) 0))
    (CallFunction|ForceApplyCooldown)
    (else
      (CallFunction|TrytoCooldown))))

(event EventOnCooldownApplied
  ;[1 1 1] Delayed Cooldown
  (|CallParentFunction))

(event FortWeapon|EventOnUnEquippedAndDetached
  ;[1 1 1] Delayed Cooldown
  (|CallParentFunction)
  (Audio|Components|Audio|FadeOut (Variables|Default|GetTimeoutAudioComponent) 0.1 0.0 "Linear" :!class /Script/Engine.AudioComponent)
  (if (ScalableFloat|GetValueasBool (Variables|Default|GetSF_TriggerCooldownOnUnequip))
    (bind _isswinging (Class|PlayerPawnAthena|IsSwinging (Variables|Default|GetPlayer)))
    (if (and (not _isswinging) (Utilities|Enum|NotEqual(Enum) (Variables|Default|GetState) "NewEnumerator0"))
      (CallFunction|ForceApplyCooldown))))

(event Custom|StartForgivenessTimer
  ;[1 1 1] Delayed Cooldown
  (CallFunction|StartForgivenessTimerMulticast))

(event Custom|StartForgivenessTimer_Multicast
  ;[1 1 1] Delayed Cooldown
  (bind _isswinging (Class|PlayerPawnAthena|IsSwinging (Variables|Default|GetPlayer)))
  (if (not _isswinging)
    (Variables|Default|SetCooldownTicks 0)
    (if (not (Utilities|Time|IsTimerActivebyHandle (Variables|Default|GetApplyCooldownTimer)))
      (CallFunction|HandleTimeoutAudio)
      (if (FortWeapon|IsEquipped)
        (bind _returnvalue (Utilities|Time|SetTimerbyEvent (delegate Custom|TickCooldown) (ScalableFloat|GetValueatLevel (Variables|Design|GetWaitTimeBeforeCooldownTrigger)) true))
        (Variables|Default|SetApplyCooldownTimer _returnvalue)
        (else
          (CallFunction|ForceApplyCooldown))))))

(event Custom|TickCooldown
  (bind _player (Variables|Default|GetPlayer))
  ;[1 1 1] Delayed Cooldown
  (bind _isswinging (Class|PlayerPawnAthena|IsSwinging (Variables|Default|GetPlayer)))
  (if _isswinging
    (Utilities|Time|ClearandInvalidateTimerbyHandle (Variables|Default|GetApplyCooldownTimer))
    (else
      (Utilities|Time|ClearandInvalidateTimerbyHandle (Variables|Default|GetApplyCooldownTimer))
      (bind _isswinging_2 (Class|PlayerPawnAthena|IsSwinging _player))
      (if _isswinging_2
        _
        (elif (GameplayTags|HasMatchingGameplayTag _player "(TagName=\"MovementMode.Falling\")")
          (Utilities|FlowControl|Delay 0.1)
          (else
            (Networking|SwitchHasAuthority
              (:Remote
                (Default|CallReticleBlockFiring)
                (Variables|Default|SetDisableFiring true)
                (CallFunction|SetStateOffServer) ; failsafe
                (Variables|Default|SetCachedCooldownTime (Math|Float|Max(Float) (Utilities|Select :Index (Variables|Default|GetStateWhenLanded) :NewEnumerator0 5.0 :NewEnumerator1 5.0 :NewEnumerator3 5.0 :NewEnumerator4 1e+01 :NewEnumerator2 15.0) 2.0))
                (Variables|Default|SetPreviousState (Variables|Default|GetState))
                (CallFunction|SetState "NewEnumerator0")
                (Class|CameraModifierGrappleParent|TurnOff (Variables|Default|GetCameraModifier)))) ; client authorititative
          ))))))

(event Custom|ApplyCooldownNow
  (bind _player (Variables|Default|GetPlayer))
  ;[1 1 1] Delayed Cooldown
  (bind _isswinging (Class|PlayerPawnAthena|IsSwinging _player))
  (if _isswinging
    _
    (elif (GameplayTags|HasMatchingGameplayTag _player "(TagName=\"MovementMode.Falling\")")
      (Utilities|FlowControl|Delay 0.1)
      (else
        (Networking|SwitchHasAuthority
          (:Remote
            (Default|CallReticleBlockFiring)
            (Variables|Default|SetDisableFiring true)
            (CallFunction|SetStateOffServer) ; failsafe
            (Variables|Default|SetCachedCooldownTime (Math|Float|Max(Float) (Utilities|Select :Index (Variables|Default|GetStateWhenLanded) :NewEnumerator0 5.0 :NewEnumerator1 5.0 :NewEnumerator3 5.0 :NewEnumerator4 1e+01 :NewEnumerator2 15.0) 2.0))
            (Variables|Default|SetPreviousState (Variables|Default|GetState))
            (CallFunction|SetState "NewEnumerator0")
            (Class|CameraModifierGrappleParent|TurnOff (Variables|Default|GetCameraModifier)))) ; client authorititative
      ))))

(event Custom|SetStateOff_Server
  ;[1 1 1] Delayed Cooldown
  (Variables|Default|SetCachedCooldownTime (Math|Float|Max(Float) (Utilities|Select :Index (Variables|Default|GetStateWhenLanded) :NewEnumerator0 5.0 :NewEnumerator1 5.0 :NewEnumerator3 5.0 :NewEnumerator4 1e+01 :NewEnumerator2 15.0) 2.0))
  (Variables|Default|SetPreviousState (Variables|Default|GetState))
  (CallFunction|SetState "NewEnumerator0")
  (Class|CameraModifierGrappleParent|TurnOff (Variables|Default|GetCameraModifier)))

(event Custom|ForceApplyCooldown
  ;[1 1 1] Delayed Cooldown
  (Networking|SwitchHasAuthority
    (:Remote
      (Default|CallReticleBlockFiring)
      (Variables|Default|SetDisableFiring true)
      (CallFunction|SetStateOffServer) ; failsafe
      (Variables|Default|SetCachedCooldownTime (Math|Float|Max(Float) (Utilities|Select :Index (Variables|Default|GetStateWhenLanded) :NewEnumerator0 5.0 :NewEnumerator1 5.0 :NewEnumerator3 5.0 :NewEnumerator4 1e+01 :NewEnumerator2 15.0) 2.0))
      (Variables|Default|SetPreviousState (Variables|Default|GetState))
      (CallFunction|SetState "NewEnumerator0")
      (Class|CameraModifierGrappleParent|TurnOff (Variables|Default|GetCameraModifier)))) ; client authorititative
)

(event Custom|DecreaseSpeedMulticast
  ;[1 1 1] Delayed Cooldown
  (CallFunction|SetState "NewEnumerator1"))

(event EventSpawnAttachment
  ;[1 1 1] Delayed Cooldown
  ;;[1 1 1] Deal with attachment race conditions
  (|CallParentFunction)
  (CallFunction|ClearCooldownTimer))

(event Custom|ClearCooldownTimer
  ;[1 1 1] Delayed Cooldown
  (Utilities|Time|ClearandInvalidateTimerbyHandle (Variables|Default|GetApplyCooldownTimer)))

(event FortWeapon|EventOnSpawnProjectile
  (params (SpawnProjectile :type Fort%20Projectile%20Base%20Object%20Reference))
  (|CallParentFunction SpawnProjectile))

(event EventFirePressed
  (|CallParentFunction)
  (Ability|SendGameplayEventtoActor (Variables|Default|GetPlayer) "(TagName=\"Event.Athena.Balloons.RemoveAll\")") ; Handle balloon removal
)

(event EventUpdateReticleFromNative
  (params (HitSomething :type Boolean) (BlockedUse :type Boolean) (IsTargetAtAimLocation :type Boolean) (TargetPosition :type Vector :default "0, 0, 0"))
  (|CallParentFunction HitSomething (or BlockedUse (Variables|Default|GetDisableFiring)) IsTargetAtAimLocation TargetPosition)
  (Default|CallUpdateGraceUI (Utilities|Time|IsTimerActivebyHandle (Variables|Default|GetApplyCooldownTimer)) (Math|Float|SelectFloat (Utilities|RemapFloat (/ (Utilities|Time|GetTimerElapsedTimebyHandle (Variables|Default|GetApplyCooldownTimer)) (ScalableFloat|GetValueatLevel (Variables|Design|GetWaitTimeBeforeCooldownTrigger))) "(X=0.000000,Y=1.000000)" "(X=0.000000,Y=0.750000)" true) 0.0 (Utilities|Time|IsTimerActivebyHandle (Variables|Default|GetApplyCooldownTimer)))))

(event EventOnEnteredSwinging(ServerOnly)
  (bind (_returnvalue_location _returnvalue_rotation) (Math|Transform|BreakTransform (Transformation|GetActorTransform (Variables|Default|GetPlayer))))
  (|CallParentFunction)
  (if (ScalableFloat|GetValueasBool (Variables|Default|GetSF_SpawnInvisibleCollider))
    (bind _returnvalue (Building|SpawnBuildingGameplayActor "/GrappleGloves/Items/GrappleGloves/BGA_DamageSphere.BGA_DamageSphere_C" (Math|Transform|MakeTransform _returnvalue_location _returnvalue_rotation (Math|Conversions|ToVector(Float) (ScalableFloat|GetValueatLevel (Variables|Default|GetSF_InvisibleColliderSizeMultiplier)))) (Variables|Default|GetPlayer) false false))
    (bind _output_get (Variables|Default|SetLargerColliderBGA _returnvalue))
    (Transformation|AttachActorToActor _output_get (Variables|Default|GetPlayer) "pelvis" "SnapToTarget" "SnapToTarget" "KeepRelative" false)))

(event EventOnAttach
  ;[1 1 1] Changing the State
  (Variables|Default|SetCurrentForgivenessFailsafeUpdates 0)
  (CallFunction|CleanUpTimeOutAudio)
  (switch Utilities|FlowControl|Switch|SwitchonGrappleGlovesState (Variables|Default|GetState)
    (:NewEnumerator0
      (CallFunction|SetState "NewEnumerator1"))
    (:NewEnumerator3
      _)
    (:NewEnumerator4
      _)
    (:NewEnumerator2
      _))
  (CallFunction|PlaySwingSound))

(event EventSlowerUpdate
  ;[0.1529 1 0.768] Slow Update
  ; Trigger cooldown forgiveness failsafe if player has been in an invald movement mode for a few updates.
  (|CallParentFunction)
  (Networking|SwitchHasAuthority
    (:Authority
      (if (Math|Boolean|ANDBoolean (or (Utilities|Enum|Equal(Enum) (Variables|/Script/Engine.CharacterMovementComponent|GetMovementMode (Variables|/Script/Engine.Character|GetCharacterMovement (Variables|Default|GetPlayer))) "MOVE_Walking") (GameplayTags|HasAnyMatchingGameplayTags (Variables|Default|GetPlayer) (GameplayTags|MakeLiteralGameplayTagContainer "(GameplayTags=((TagName=\"MovementMode.Driving\"),(TagName=\"MovementMode.Parachuting\"),(TagName=\"MovementMode.Skydiving\"),(TagName=\"MovementMode.Passenger\"),(TagName=\"MovementMode.Ziplining\"),(TagName=\"MovementMode.ZipliningOnSpline\"),(TagName=\"MovementMode.SurfaceSwimming\")))"))) (not (Utilities|Time|IsTimerActivebyHandle (Variables|Default|GetApplyCooldownTimer))) (Utilities|Enum|NotEqual(Enum) (Variables|Default|GetState) "NewEnumerator0"))
        (bind _result (Math|Integer|IncrementInt (Variables|Default|GetCurrentForgivenessFailsafeUpdates))
          (:%20%20
            (if (>= _result (ScalableFloat|GetValueasInteger (Variables|Default|GetSF_ForgivenessFailsafeUpdateCount)))
              (CallFunction|StartForgivenessTimerMulticast)
              (Variables|Default|SetCurrentForgivenessFailsafeUpdates 0))))
        (else
          (Variables|Default|SetCurrentForgivenessFailsafeUpdates 0))))
    (:Remote
      (CallFunction|HandleDangleAudio))))

(event EventOnSwingingEnded
  (params (SkipRemovingSwingingGE :type Boolean))
  (|CallParentFunction SkipRemovingSwingingGE)
  (if (ScalableFloat|GetValueasBool (Variables|Default|GetSF_SpawnInvisibleCollider))
    (Actor|DestroyActor (Variables|Default|GetLargerColliderBGA))))

(event EventOnDetach
  ;[1 1 1] Changing the State
  (Utilities|FlowControl|DoOnce false
    (:Completed
      (Utilities|FlowControl|Sequence
        (:then_0
          _)
        (:then_1
          (Utilities|FlowControl|Delay 0.25)))
      (switch Utilities|FlowControl|Switch|SwitchonGrappleGlovesState (Variables|Default|GetState)
        (:NewEnumerator1
          (CallFunction|SetState "NewEnumerator3"))
        (:NewEnumerator3
          (CallFunction|SetState "NewEnumerator4"))
        (:NewEnumerator4
          (CallFunction|SetState "NewEnumerator2"))))))

(event Custom|CleanUpTimeOutAudio
  ;[1 1 1] Timeout Audio
  (Utilities|Time|ClearandInvalidateTimerbyHandle (Variables|Audio|GetTimeoutTickAudioTimerHandle))
  (Utilities|Time|ClearandInvalidateTimerbyHandle (Variables|Audio|GetTimeoutAudioDelayTimerHandle))
  (Audio|Components|Audio|FadeOut (Variables|Default|GetTimeoutAudioComponent) 0.1 0.0 "Linear" :!class /Script/Engine.AudioComponent)
  (Variables|Audio|SetTimeoutBeepNumber 0))

(event Custom|HandleTimeoutAudio
  ;[1 1 1] Timeout Audio
  (if (and (Pawn|IsLocallyControlled (Game|GetInstigator)) (not (CallFunction|IsSwinging)))
    (bind _returnvalue (Utilities|Time|SetTimerbyEvent (delegate Custom|StartTimeoutAudio) (ScalableFloat|GetValueatLevel (Variables|Design|GetWaitTimeBeforeCooldownTrigger))))
    (Variables|Audio|SetTimeoutAudioDelayTimerHandle _returnvalue)
    (if (FortWeapon|IsEquipped)
      (Audio|PlaySound2D (Variables|Audio|GetTimeout_StartWarning)))))

(event Custom|StartTimeoutAudio
  ;[1 1 1] Timeout Audio
  (if (not (CallFunction|IsSwinging))
    (bind _returnvalue (Utilities|Time|SetTimerbyEvent (delegate Custom|TickTimeoutSound) 1.0 true false (- 1.0 (ScalableFloat|GetValueatLevel (Variables|Design|GetWaitTimeBeforeCooldownTrigger)))))
    (Variables|Audio|SetTimeoutTickAudioTimerHandle _returnvalue)
    (if (FortWeapon|IsEquipped)
      (bind _returnvalue_2 (Audio|SpawnSoundAttached :Sound (Variables|Audio|GetTimeout_Start) :AttachToComponent (Variables|Visual|GetWeaponMesh) :AttachPointName "None" :LocationType "KeepRelativeOffset" :bStopWhenAttachedToDestroyed true :VolumeMultiplier 1.0 :PitchMultiplier 1.0 :StartTime 0.0 :bAutoDestroy true))
      (Variables|Default|SetTimeoutAudioComponent _returnvalue_2)
      (bind _returnvalue_3 (Audio|SpawnSoundAttached (Variables|Audio|GetTimeout_Loop) (Variables|Visual|GetWeaponMesh)))
      (Variables|Default|SetTimeoutLoopComponent _returnvalue_3))))

(event Custom|TickTimeoutSound
  ;[1 1 1] Timeout Audio
  (if (and (GameplayTags|HasMatchingGameplayTag (Variables|Default|GetPlayer) "(TagName=\"Granted.Cooldown.GrappleParent\")") (not (CallFunction|IsSwinging)))
    (Math|Integer|IncrementInt (Variables|Audio|GetTimeoutBeepNumber)
      (:%20%20
        (if (FortWeapon|IsEquipped)
          (Audio|PlaySound2D (Variables|Audio|GetTimeout_Tick) (ScalableFloat|GetValueatLevel (Variables|Audio|GetSF_BeepMultiplier) (Math|Conversions|ToFloat(Integer) (Variables|Audio|GetTimeoutBeepNumber))))
          (if (== (Variables|Audio|GetTimeoutBeepNumber) (- (Variables|Default|GetCachedCooldownTime) 2.0))
            (Audio|PlaySound2D (Variables|Audio|GetTimeout_EndWarning))))))
    (else
      (CallFunction|CleanUpTimeOutAudio))))

(event Custom|HandleTimeoutAudioStopped
  ;[1 1 1] Timeout Audio
  (if (and (FortWeapon|IsEquipped) (Utilities|Time|IsTimerActivebyHandle (Variables|Audio|GetTimeoutTickAudioTimerHandle)))
    (CallFunction|CleanUpTimeOutAudio)
    (if (and (FortWeapon|IsEquipped) (Pawn|IsLocallyControlled (Game|GetInstigator)))
      (Audio|PlaySound2D (Variables|Audio|GetTimeout_Stop)))))

(event Custom|HandleDangleAudio
  (bind _danglingloop (Variables|Default|GetDanglingLoop))
  ;[1 1 1] Timeout Audio
  (if (Variables|Status|GetbIsAttached)
    (if (Audio|Components|Audio|IsPlaying (Variables|Default|GetDanglingLoop))
      (Audio|Parameter|SetFloatParameter _danglingloop "DangleSpeed" (Math|Vector|VectorLength (Variables|/Script/Engine.MovementComponent|GetVelocity (Pawn|GetMovementComponent (Variables|Default|GetPlayer)))) :!class /Script/AudioExtensions.AudioParameterControllerInterface)
      (else
        (Audio|Components|Audio|FadeIn _danglingloop 0.5 1.0 0.0 "Linear" :!class /Script/Engine.AudioComponent)))
    (elif (Audio|Components|Audio|IsPlaying (Variables|Default|GetDanglingLoop))
      (Audio|Components|Audio|Stop (Variables|Default|GetDanglingLoop)))))

(event FortWeapon|EventOnEquippedAndAttached
  ;[1 1 1] Timeout Audio
  (|CallParentFunction)
  (if (Utilities|Time|IsTimerActivebyHandle (Variables|Audio|GetTimeoutTickAudioTimerHandle))
    (bind _timeoutloopcomponent (Variables|Default|GetTimeoutLoopComponent)
      (:then
        (Audio|Components|Audio|FadeIn _timeoutloopcomponent 0.1 1.0 0.0 "Linear" :!class /Script/Engine.AudioComponent))
      (:else
        (bind _returnvalue (Audio|SpawnSoundAttached (Variables|Audio|GetTimeout_Loop) (Variables|Visual|GetWeaponMesh)))
        (Variables|Default|SetTimeoutLoopComponent _returnvalue)))))

(fn AllowStopSwinging
  (return (not (Variables|Default|GetCharging))))

(fn CalculateCooldownTime
  (return (Variables|Default|GetCachedCooldownTime)))

(fn InitialJumpMultiplier
  (return (Utilities|Select :Index (Variables|Default|GetState) :NewEnumerator0 1.4 :NewEnumerator1 1.4 :NewEnumerator3 1.0 :NewEnumerator4 1.0 :NewEnumerator2 1.0)))

(fn SetState
  (params (State :type GrappleGlovesState%20Enum :default "NewEnumerator0"))
  (if (Utilities|Enum|NotEqual(Enum) State (Variables|Default|GetState))
    (bind _output_get (Variables|Default|SetState State))
    (Default|CallOnStateChanged)
    (switch Utilities|FlowControl|Switch|SwitchonGrappleGlovesState _output_get
      (:NewEnumerator0
        (GameplayEffects|RemoveActiveEffectswithGrantedTags (Ability|GetFortAbilitySystemComponent (Variables|Actor|GetInstigator)) (Variables|Default|GetTag_Warm)))
      (:NewEnumerator1
        (GameplayEffects|ApplyGameplayEffectToSelf :self (Ability|GetFortAbilitySystemComponent (Variables|Actor|GetInstigator)) :GameplayEffectClass "/GrappleGloves/Items/GrappleGloves/GE_GrappleGloves_IsWarm.GE_GrappleGloves_IsWarm_C" :Level 0.0))
      (:NewEnumerator3
        (GameplayEffects|ApplyGameplayEffectToSelf :self (Ability|GetFortAbilitySystemComponent (Variables|Actor|GetInstigator)) :GameplayEffectClass "/GrappleGloves/Items/GrappleGloves/GE_GrappleGloves_IsWarm.GE_GrappleGloves_IsWarm_C" :Level 0.0)
        (SwingingControlParams|SetActiveOverrideSwingingControlParams "(TagName=\"Gameplay.SwingingControlParams.GrappleGloves.Phase1\")"))
      (:NewEnumerator4
        (GameplayEffects|ApplyGameplayEffectToSelf :self (Ability|GetFortAbilitySystemComponent (Variables|Actor|GetInstigator)) :GameplayEffectClass "/GrappleGloves/Items/GrappleGloves/GE_GrappleGloves_IsWarm.GE_GrappleGloves_IsWarm_C" :Level 0.0)
        (SwingingControlParams|SetActiveOverrideSwingingControlParams "(TagName=\"Gameplay.SwingingControlParams.GrappleGloves.Phase2\")"))
      (:NewEnumerator2
        (GameplayEffects|ApplyGameplayEffectToSelf :self (Ability|GetFortAbilitySystemComponent (Variables|Actor|GetInstigator)) :GameplayEffectClass "/GrappleGloves/Items/GrappleGloves/GE_GrappleGloves_IsWarm.GE_GrappleGloves_IsWarm_C" :Level 0.0)
        (GameplayEffects|ApplyGameplayEffectToSelf :self (Ability|GetFortAbilitySystemComponent (Variables|Actor|GetInstigator)) :GameplayEffectClass "/GrappleGloves/Items/GrappleGloves/GE_GrappleGloves_IsHot.GE_GrappleGloves_IsHot_C" :Level 0.0)
        (SwingingControlParams|SetActiveOverrideSwingingControlParams "(TagName=\"Gameplay.SwingingControlParams.GrappleGloves.Phase3\")")))
    (GameplayEffects|RemoveActiveEffectswithGrantedTags (Ability|GetFortAbilitySystemComponent (Variables|Actor|GetInstigator)) (Variables|Default|GetTag_Hot))
    (SwingingControlParams|SetActiveOverrideSwingingControlParams "(TagName=\"Gameplay.SwingingControlParams.GrappleGloves.Default\")")))

(fn ShouldAutoSlide
  (return (Variables|Default|GetTryingToCharge)))

(fn ShouldDisableFiring
  (return (Variables|Default|GetDisableFiring)))

(fn ConstructionScript
  (|CallParentFunction))
