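; Camera modifier for Grapple Gloves, built on the Fortnite swinging camera modifier, that adapts forward offset and multipliers to player state.
; On activation and every camera update, it refreshes forward offset from the max-speed tag and its upper speed bound from swinging max horizontal speed, overridden while sliding.
; CalculateDynamicMultipliers selects FOV and boost multipliers from max-speed state, with sliding overriding FOV.
; BlueprintModifyCamera itself passes the view through unchanged.
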(event EventTurnOn
  ;[0.1035 1 0] Immediately update the forward speed bounds to align with the player's current speed maximum.
  (|CallParentFunction)
  (CallFunction|UpdateForwardOffsetData))

(fn BlueprintModifyCamera
  (params (DeltaTime :type Float%20%28single-precision%29) (ViewLocation :type Vector :default "0, 0, 0") (ViewRotation :type Rotator :default "0, 0, 0") (FOV :type Float%20%28single-precision%29))
  (CallFunction|UpdateForwardOffsetData)
  (return ViewLocation ViewRotation FOV))

(fn CalculateDynamicMultipliers
  (bind _cachedplayer (Variables|FortCameraModifier_Swinging|GetCachedPlayer)
    (:then
      (LocalVariables|SetbIsAtMaxSpeed (GameplayTags|HasMatchingGameplayTag _cachedplayer (Variables|Default|GetMaxSpeedTag)))
      (return (Math|Float|SelectFloat (Variables|Default|GetFOVMultiplier_Sliding) (Math|Float|SelectFloat (Variables|Default|GetFOVMultiplier_MaxSpeed) (Variables|Default|GetFOVMultiplier_NormalSpeed) (LocalVariables|GetbIsAtMaxSpeed)) (GameplayTags|HasMatchingGameplayTag (Variables|FortCameraModifier_Swinging|GetCachedPlayer) "(TagName=\"MovementMode.Sliding\")")) (Math|Float|SelectFloat (Variables|Default|GetBoostMultiplier_MaxSpeed) (Variables|Default|GetBoostMultiplier_NormalSpeed) (LocalVariables|GetbIsAtMaxSpeed))))
    (:else)))

(fn UpdateForwardOffsetData
  (bind _speedbounds_x (Math|Vector2D|BreakVector2D (Utilities|Struct|BreakSpeedScaledAxisOffsetData (Variables|Config|Offset|GetForwardOffsetData))))
  (bind _cachedplayer (Variables|FortCameraModifier_Swinging|GetCachedPlayer)
    (:then
      (bind _asfort_movement_comp_character (Utilities|Casting|CastToFortMovementComp_Character (Variables|/Script/Engine.Character|GetCharacterMovement _cachedplayer)))
      (LocalVariables|SetbIsAtMaxSpeed (GameplayTags|HasMatchingGameplayTag (Variables|FortCameraModifier_Swinging|GetCachedPlayer) (Variables|Default|GetMaxSpeedTag)))
      (Utilities|Struct|SetmembersinSpeedScaledAxisOffsetData (Variables|Config|Offset|GetForwardOffsetData) (Math|Float|SelectFloat (Variables|Default|GetForwardScaledOffsetAtMaxSpeed) (Variables|Default|GetForwardScaledOffsetAtNormalSpeed) (LocalVariables|GetbIsAtMaxSpeed)) (Math|Vector2D|MakeVector2D _speedbounds_x (Math|Float|SelectFloat (Variables|Default|GetSpeedBoundsWhileSlidingOverride) (* (ScalableFloat|GetValueatLevel (.MaxHorizontalSpeed (Utilities|Struct|BreakSwingingControlParams (Swinging|GetSwingingControlParameters _asfort_movement_comp_character)))) (Variables|Default|GetSpeedBoundsMultiplier)) (GameplayTags|HasMatchingGameplayTag _cachedplayer "(TagName=\"MovementMode.Sliding\")")))))
    (:else)))
