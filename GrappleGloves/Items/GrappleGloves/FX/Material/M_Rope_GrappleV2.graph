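# Default Lit material for the grapple gloves rope, flagged for Niagara mesh particles and Nanite.
# A near-camera fade blends between minimum and maximum width, which is clamped and multiplied by the world vertex normal to vary rope thickness with camera distance.
# U coordinates are multiplied by object Z scale and a tiling parameter, so the diffuse and normal textures repeat along the rope as it stretches.
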Material /GrappleGloves/Items/GrappleGloves/FX/Material/M_Rope_GrappleV2
graph {
 properties {
  bUsedWithNanite=true
  bUsedWithNiagaraMeshParticles=true
  pixelDepthOffsetMode=PDOM_Legacy
  previewMesh=/GrappleGloves/Items/GrappleGloves/FX/SM_GrappleGloves_RopeV2
  usedShadingModels=Default Lit
 }
 nodes [
  n1 = ScalarParameter() {defaultValue=0.1, parameterName=MinWidth}
  n2 = ScalarParameter() {defaultValue=3, parameterName=MaxWidth}
  n3 = MaterialFunctionCall(Distance=ScalarParameter() {defaultValue=5000, parameterName=NearCameraWPO}) {materialFunction=&Near_Camera_Fade}
  Multiply(A=Clamp(None=LinearInterpolate(A=@n1, B=@n2, Alpha=@n3.Result) {constB=15}, Max=@n2) {maxDefault=10}, B=VertexNormalWS())
  n4 = MaterialFunctionCall() {materialFunction=/Engine/Functions/Engine_MaterialFunctions02/WorldPositionOffset/ObjectScale}
  n5 = MaterialFunctionCall(Float2=TextureCoordinate()) {materialFunction=&BreakOutFloat2Components}
  n6 = AppendVector(A=Multiply(A=Multiply(A=@n4.Scale Z, B=@n5.R), B=ScalarParameter() {defaultValue=1.5, parameterName=VScale}), B=@n5.G)
  TextureSample(UVs=@n6) {texture=&/T_Gadget_Grapple_Glove_D_FX}
  TextureSample(UVs=@n6) {samplerType=SAMPLERTYPE_Normal, texture=&/T_Gadget_Grapple_Glove_N_FX}
 ]
}
