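# Animation Blueprint for the Grapple Gloves gadget mesh, animating equip, fire, swing loop, cooldown and idle.
# Each update sets FireWeapon from the Fortnite pawn's bSwingingAttached and WeaponCoolDown from its grapple cooldown gameplay tag.
# These drive a conduit-routed state machine, whose additive poses layer over the attached parent's copied pose before inertialization.
# A gauntlet_mstr bone constraint to right-arm bones exists but is disconnected from the output.
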AnimBlueprint /GrappleGloves/Items/GrappleGloves/Meshes/Grapple_Gloves_Gadget_AnimBP
target_skeleton=&/Grapple_Gloves_Gadget_Skeleton
graphs [
 AnimGraph {nodes [
   n1 = "Constraint - gauntlet_mstr (PPPP)"!(
    ComponentPose=Local To Component()
    Alpha=1
    ConstraintWeights_0=0.3818
    ConstraintWeights_1:float=0.0701
    ConstraintWeights_2:float=0.3492
    ConstraintWeights_3:float=0.1989
   ) -> (Pose) {
    alphaBoolBlend {blendOption=Linear}
    boneToModify {boneName=gauntlet_mstr}
    constraintSetup [
     {targetBone {boneName=lowerarm_twist_02_r}}
     {targetBone {boneName=upperarm_r}}
     {targetBone {boneName=lowerarm_twist_01_r}}
     {targetBone {boneName=lowerarm_r}}
    ]
    constraintWeights [1,1,1,1]
   }
   Component To Local(ComponentPose=@n1)
   n2 = Copy Pose From Mesh() {bUseAttachedParent=true}
   AnimGraphNode_StateMachine_0 = GrappleStatemachine() {is=title AnimGraphNode_StateMachine}
   Output Pose(Result=Inertialization!(Source:Pose Link Structure=Apply Additive(Base=@n2, Additive=@AnimGraphNode_StateMachine_0) {alphaBoolBlend {blendOption=Linear}}) -> (Pose:Pose Link Structure) {is=title AnimGraphNode_Inertialization}) {layerGroup=None}
  ]}
 GrappleStatemachine {
  entry=Equip
  path=AnimGraph.AnimGraphNode_StateMachine_0.GrappleStatemachine
  ids=id
  nodes [
   LoopToEquip() {is=title AnimStateAliasNode, also {aliases [Loop]}}
   AnimStateConduitNode_0 = Conduit()
   AnimStateNode_0 = Equip() {
    is=title AnimStateNode
    stateEntered {guid=#1}
    stateFullyBlended {guid=#2}
    stateLeft {guid=#3}
   }
   AnimStateNode_1 = FireWeapon() {is=title AnimStateNode}
   AnimStateNode_2 = Idle()
   AnimStateNode_3 = CoolDown() {is=title AnimStateNode}
   AnimStateNode_4 = Loop() {
    is=title AnimStateNode
    stateEntered {guid=#4}
    stateFullyBlended {guid=#5}
    stateLeft {guid=#6}
   }
   AnimStateTransitionNode_0 = AnimStateTransitionNode() {crossfadeDuration=0, logicType=TLT_Inertialization}
   AnimStateTransitionNode_1 = AnimStateTransitionNode() {bAutomaticRuleBasedOnSequencePlayerInState=true}
   AnimStateTransitionNode_2 = AnimStateTransitionNode() {crossfadeDuration=0.3, logicType=TLT_Inertialization, priorityOrder=100}
   AnimStateTransitionNode_3 = AnimStateTransitionNode() {bAutomaticRuleBasedOnSequencePlayerInState=true, crossfadeDuration=0, logicType=TLT_Inertialization}
   AnimStateTransitionNode_4 = AnimStateTransitionNode() {
    logicType=TLT_Inertialization
    transitionEnd {guid=#7}
    transitionInterrupt {guid=#8}
    transitionStart {guid=#9}
   }
   AnimStateTransitionNode_5 = AnimStateTransitionNode()
   AnimStateTransitionNode_7 = AnimStateTransitionNode() {logicType=TLT_Inertialization, priorityOrder=2}
   AnimStateTransitionNode_8 = AnimStateTransitionNode()
  ]
  links [
   {
    dst=CoolDown
    rule=AnimGraph.AnimGraphNode_StateMachine_0.GrappleStatemachine.AnimStateTransitionNode_7.Transition
    src=Conduit
    via=AnimStateTransitionNode_7
   }
   {
    dst=FireWeapon
    rule=AnimGraph.AnimGraphNode_StateMachine_0.GrappleStatemachine.AnimStateTransitionNode_0.Transition
    src=Conduit
    via=AnimStateTransitionNode_0
   }
   {
    dst=Idle
    rule=AnimGraph.AnimGraphNode_StateMachine_0.GrappleStatemachine.AnimStateTransitionNode_2.Transition
    src=Conduit
    via=AnimStateTransitionNode_2
   }
   {
    dst=Conduit
    rule=AnimGraph.AnimGraphNode_StateMachine_0.GrappleStatemachine.AnimStateTransitionNode_8.Transition
    src=CoolDown
    via=AnimStateTransitionNode_8
   }
   {
    dst=Conduit
    rule=AnimGraph.AnimGraphNode_StateMachine_0.GrappleStatemachine.AnimStateTransitionNode_1.Transition
    src=Equip
    via=AnimStateTransitionNode_1
   }
   {
    dst=Loop
    rule=AnimGraph.AnimGraphNode_StateMachine_0.GrappleStatemachine.AnimStateTransitionNode_3.Transition
    src=FireWeapon
    via=AnimStateTransitionNode_3
   }
   {
    dst=Conduit
    rule=AnimGraph.AnimGraphNode_StateMachine_0.GrappleStatemachine.AnimStateTransitionNode_5.Transition
    src=Idle
    via=AnimStateTransitionNode_5
   }
   {
    dst=Equip
    rule=AnimGraph.AnimGraphNode_StateMachine_0.GrappleStatemachine.AnimStateTransitionNode_4.Transition
    src=LoopToEquip
    via=AnimStateTransitionNode_4
   }
  ]
 }
 Conduit {path=AnimGraph.AnimGraphNode_StateMachine_0.GrappleStatemachine.AnimStateConduitNode_0.Conduit, nodes [Result(bCanEnterTransition=true)]}
 Equip {
  path=AnimGraph.AnimGraphNode_StateMachine_0.GrappleStatemachine.AnimStateNode_0.Equip
  nodes [Output Animation Pose(Result="Sequence Player (additive) 'GrappleGloves_Weap_Equip'"() {bLoopAnimation=false, sequence=/GrappleGloves/Items/GrappleGloves/Animation/Weapon/GrappleGloves_Weap_Equip})]
 }
 FireWeapon {
  path=AnimGraph.AnimGraphNode_StateMachine_0.GrappleStatemachine.AnimStateNode_1.FireWeapon
  nodes [Output Animation Pose(Result="Sequence Player (additive) 'GrappleGloves_Weap_Fire'"() {bLoopAnimation=false, sequence=/GrappleGloves/Items/GrappleGloves/Animation/Weapon/GrappleGloves_Weap_Fire})]
 }
 Idle {
  path=AnimGraph.AnimGraphNode_StateMachine_0.GrappleStatemachine.AnimStateNode_2.Idle
  nodes [Output Animation Pose(Result="Sequence Player (additive) 'GrappleGloves_Weap_Equipped_Base_Pose'"() {sequence=/GrappleGloves/Items/GrappleGloves/Animation/Weapon/GrappleGloves_Weap_Equipped_Base_Pose})]
 }
 CoolDown {
  path=AnimGraph.AnimGraphNode_StateMachine_0.GrappleStatemachine.AnimStateNode_3.CoolDown
  nodes [Output Animation Pose(Result="Sequence Player (additive) 'GrappleGloves_Weap_CoolDown'"() {sequence=/GrappleGloves/Items/GrappleGloves/Animation/Weapon/GrappleGloves_Weap_CoolDown})]
 }
 Loop {
  path=AnimGraph.AnimGraphNode_StateMachine_0.GrappleStatemachine.AnimStateNode_4.Loop
  nodes [Output Animation Pose(Result="Sequence Player (additive) 'GrappleGloves_Weap_Base_Pose'"() {sequence=/GrappleGloves/Items/GrappleGloves/Animation/Weapon/GrappleGloves_Weap_Base_Pose})]
 }
 Transition {
  path=AnimGraph.AnimGraphNode_StateMachine_0.GrappleStatemachine.AnimStateTransitionNode_0.Transition
  nodes [Result(bCanEnterTransition=Get FireWeapon() -> (FireWeapon:bool))]
 }
 Transition {path=AnimGraph.AnimGraphNode_StateMachine_0.GrappleStatemachine.AnimStateTransitionNode_1.Transition, nodes [Result()]}
 Transition {
  path=AnimGraph.AnimGraphNode_StateMachine_0.GrappleStatemachine.AnimStateTransitionNode_2.Transition
  nodes [Result(bCanEnterTransition=true)]
 }
 Transition {
  path=AnimGraph.AnimGraphNode_StateMachine_0.GrappleStatemachine.AnimStateTransitionNode_3.Transition
  nodes [Result(bCanEnterTransition=true)]
 }
 Transition {
  path=AnimGraph.AnimGraphNode_StateMachine_0.GrappleStatemachine.AnimStateTransitionNode_4.Transition
  nodes [Result(bCanEnterTransition=NOT Boolean(A=Get FireWeapon()))]
 }
 Transition {
  path=AnimGraph.AnimGraphNode_StateMachine_0.GrappleStatemachine.AnimStateTransitionNode_5.Transition
  nodes [Result(bCanEnterTransition=true)]
 }
 Transition {
  path=AnimGraph.AnimGraphNode_StateMachine_0.GrappleStatemachine.AnimStateTransitionNode_7.Transition
  nodes [Result(bCanEnterTransition=Get WeaponCoolDown() -> (WeaponCoolDown:bool))]
 }
 Transition {
  path=AnimGraph.AnimGraphNode_StateMachine_0.GrappleStatemachine.AnimStateTransitionNode_8.Transition
  nodes [Result(bCanEnterTransition=true)]
 }
 EventGraph {nodes [
   Event BlueprintUpdateAnimation()
   > n1 = Cast To FortPlayerPawn(Object=TryGetPawnOwner())
   > n2 = Sequence!(execute) -> (then_0, then_1, then_2:Exec)
   > n3 = Branch(Condition=Get bSwingingAttached!(self:Fort Pawn Object Reference=@n1) -> (bSwingingAttached:bool))
   > Set FireWeapon(FireWeapon:bool=true) -> (Output_Get:bool)
   Set FireWeapon(execute=@n3.else, FireWeapon=false)
   n4 = Branch(
    execute=@n2.then_1
    Condition=HasMatchingGameplayTag!(self:Gameplay Tag Asset Interface Interface=@n1, TagToCheck:Gameplay Tag Structure=(TagName="Granted.Cooldown.GrappleParent")) -> (ReturnValue:bool)
   )
   > Set WeaponCoolDown(WeaponCoolDown:bool=true) -> (Output_Get:bool)
   Set WeaponCoolDown(execute=@n4.else, WeaponCoolDown=false)
  ]}
]
variables [
 FireWeapon {type=boolean}
 WeaponCoolDown {type=boolean}
]
