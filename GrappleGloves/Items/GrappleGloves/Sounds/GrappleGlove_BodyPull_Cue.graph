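# Sound cue for the grapple glove body-pull sound in the GrappleGloves gadget.
# A random node picks one of three equally weighted body-pull waves, which a modulator pitches down within a small random range before a short random delay starts playback.
# It plays at reduced volume in the gadgets/items sound class, with overridden logarithmic distance attenuation and a small concurrency cap on simultaneous instances.
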SoundCue /GrappleGloves/Items/GrappleGloves/Sounds/GrappleGlove_BodyPull_Cue
root=Delay_0
nodes [
 {id=Delay_0, children [Modulator_9], properties {delayMax=0.1, delayMin=0.075}}
 {id=Modulator_9, children [Random_0], properties {pitchMax=0.85, pitchMin=0.7, volumeMax=1, volumeMin=1}}
 {id=Random_0, children [WavePlayer_0,WavePlayer_1,WavePlayer_2], properties {weights [1,1,1]}}
 {id=WavePlayer_0, properties {soundWaveAssetPtr=&/GrappleGlove_BodyPull_01}}
 {id=WavePlayer_1, properties {soundWaveAssetPtr=&/GrappleGlove_BodyPull_02}}
 {id=WavePlayer_2, properties {soundWaveAssetPtr=&/GrappleGlove_BodyPull_03}}
]
properties {
 bOverrideAttenuation=true
 bOverrideConcurrency=true
 maxDistance=6500
 soundClassObject=&Fort_Gadgets_Items
 volumeMultiplier=0.25
 attenuationOverrides {distanceAlgorithm=Logarithmic, falloffDistance=6000, stereoSpread=1000, attenuationShapeExtents {x=500}}
 concurrencyOverrides {maxCount=3}
}
