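# Sound cue for the body-swing sound of the Grapple Gloves item, classed under gadget item sounds.
# A random node picks one of four equally weighted swing waves, and a modulator varies its pitch and volume.
# A short fixed delay at the root holds back playback.
# A shared third-person grappling movement attenuation asset handles distance falloff, and a concurrency override caps simultaneous instances.
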SoundCue /GrappleGloves/Items/GrappleGloves/Sounds/GrappleGlove_BodySwing_Cue
root=Delay_0
nodes [
 {id=Delay_0, children [Modulator_0], properties {delayMax=0.075, delayMin=0.075}}
 {id=Modulator_0, children [Random_0], properties {pitchMax=1.1, pitchMin=0.9, volumeMax=0.8, volumeMin=0.6}}
 {id=Random_0, children [WavePlayer_0,WavePlayer_1,WavePlayer_2,WavePlayer_5], properties {weights [1,1,1,1]}}
 {id=WavePlayer_0, properties {soundWaveAssetPtr=&/GrappleGlove_BodySwing_01}}
 {id=WavePlayer_1, properties {soundWaveAssetPtr=&/GrappleGlove_BodySwing_02}}
 {id=WavePlayer_2, properties {soundWaveAssetPtr=&/GrappleGlove_BodySwing_03}}
 {id=WavePlayer_5, properties {soundWaveAssetPtr=&/GrappleGlove_BodySwing_04}}
]
properties {
 attenuationSettings=/Game/Sounds/Attenuation/Fort_Grappling_Movement_3P_Att
 bOverrideConcurrency=true
 maxDistance=6000
 soundClassObject=&Fort_Gadgets_Items
 volumeMultiplier=0.8
 attenuationOverrides {distanceAlgorithm=Logarithmic, falloffDistance=2400, stereoSpread=2000, attenuationShapeExtents {x=800}}
 concurrencyOverrides {maxCount=4}
}
