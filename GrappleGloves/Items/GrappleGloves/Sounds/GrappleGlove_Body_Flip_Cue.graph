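# GrappleGlove_Body_Flip_Cue is the Sound Cue for the body flip sound of the Grapple Gloves item.
# A Random node picks one of three equally weighted flip waves, and a Modulator slightly varies its pitch.
# The cue plays at reduced volume in the gadget items sound class, with its own logarithmic, HRTF-spatialized attenuation and a concurrency cap on simultaneous instances.
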SoundCue /GrappleGloves/Items/GrappleGloves/Sounds/GrappleGlove_Body_Flip_Cue
root=Modulator_9
nodes [
 {id=Modulator_9, children [Random_1], properties {pitchMax=1.1, pitchMin=0.9, volumeMax=1, volumeMin=1}}
 {id=Random_1, children [WavePlayer_5,WavePlayer_6,WavePlayer_7], properties {weights [1,1,1]}}
 {id=WavePlayer_5, properties {soundWaveAssetPtr=&/GrappleGlove_Body_Flip_01}}
 {id=WavePlayer_6, properties {soundWaveAssetPtr=&/GrappleGlove_Body_Flip_02}}
 {id=WavePlayer_7, properties {soundWaveAssetPtr=&/GrappleGlove_Body_Flip_03}}
]
properties {
 bOverrideAttenuation=true
 bOverrideConcurrency=true
 maxDistance=6300
 soundClassObject=&Fort_Gadgets_Items
 volumeMultiplier=0.2
 attenuationOverrides {
  distanceAlgorithm=Logarithmic
  falloffDistance=6000
  reverbWetLevelMax=0.75
  spatializationAlgorithm=SPATIALIZATION_HRTF
  stereoSpread=0
  attenuationShapeExtents {x=300}
 }
 concurrencyOverrides {maxCount=3}
}
