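# Sound cue for the Grapple Glove, evidently the dry-fire feedback heard when the glove cannot fire.
# A single wave player feeds the dry-fire sample straight to the output, with slightly lowered pitch and reduced volume.
# It routes through the third-person ranged weapon fire sound class and overrides concurrency to cap simultaneous instances.
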SoundCue /GrappleGloves/Items/GrappleGloves/Sounds/GrappleGlove_CantFire2_Cue
root=WavePlayer_1
nodes [{id=WavePlayer_1, properties {soundWaveAssetPtr=&/GrappleGlove_DryFire_01}}]
properties {
 bOverrideConcurrency=true
 maxDistance=2097152
 pitchMultiplier=0.95
 soundClassObject=&Fort_WeaponRanged_Fire_3P_PSM
 volumeMultiplier=0.6
 concurrencyOverrides {maxCount=4}
}
