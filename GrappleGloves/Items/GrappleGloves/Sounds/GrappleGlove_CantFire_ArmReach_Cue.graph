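# Sound cue for the Grapple Glove's arm-reach foley, named for the case where the glove can't fire.
# A single wave player feeds a Modulator that slightly randomizes pitch and volume on each play, under a reduced overall volume.
# It routes to the player foley sound class with basic foley attenuation.
# Separate global and local concurrency assets govern simultaneous instances.
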SoundCue /GrappleGloves/Items/GrappleGloves/Sounds/GrappleGlove_CantFire_ArmReach_Cue
root=Modulator_2
nodes [
 {id=Modulator_2, children [WavePlayer_0], properties {pitchMax=1.1, pitchMin=0.9, volumeMax=1, volumeMin=0.8}}
 {id=WavePlayer_0, properties {soundWaveAssetPtr=&/GrappleGlove_CantFire_ArmReach_01}}
]
properties {
 attenuationSettings=&Foley_Attenuation_Basic_Att
 maxDistance=1536
 soundClassObject=&Fort_Player_Foley
 volumeMultiplier=0.5
 attenuationOverrides {bAttenuate=false}
 concurrencyOverrides {bLimitToOwner=true, maxCount=1}
 concurrencySet [
  /GrappleGloves/Items/GrappleGloves/Sounds/Settings/CONC_GrappleGlove_CantFire_ArmReach_Global
  /GrappleGloves/Items/GrappleGloves/Sounds/Settings/CONC_GrappleGlove_CantFire_ArmReach_Local
 ]
}
