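# Looping sound cue for the Grapple Gloves dangle, using a third-person grappling movement attenuation and a concurrency override that caps simultaneous instances.
# A Random node picks one of two equally weighted looping waves and feeds a continuous modulator at the root.
# The DangleSpeed parameter drives that modulator, raising pitch slightly and fading volume from full to silent as the value increases.
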SoundCue /GrappleGloves/Items/GrappleGloves/Sounds/GrappleGlove_Dangle_Loop_Cue
root=ModulatorContinuous_0
nodes [
 {
  id=ModulatorContinuous_0
  children [Random_1]
  properties {
   pitchModulationParams {default=0, maxInput=1000, maxOutput=1.1, minOutput=1, parameterName=DangleSpeed}
   volumeModulationParams {default=0, maxInput=1000, maxOutput=0, minOutput=1, parameterName=DangleSpeed}
  }
 }
 {id=Random_1, children [WavePlayer_0,WavePlayer_1], properties {bShouldExcludeFromBranchCulling=true, weights [1,1]}}
 {id=WavePlayer_0, properties {bLooping=true, soundWaveAssetPtr=&/GrappleGlove_Dangle_Loop_01}}
 {id=WavePlayer_1, properties {bLooping=true, soundWaveAssetPtr=&/GrappleGlove_Dangle_Loop_02}}
]
properties {
 attenuationSettings=/Game/Sounds/Attenuation/Fort_Grappling_Movement_3P_Att
 bOverrideConcurrency=true
 maxDistance=6000
 soundClassObject=&Fort_Gadgets_Items
 volumeMultiplier=0.8
 concurrencyOverrides {maxCount=4}
}
