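# Sound cue for the Grapple Gloves detach, mixing a random detach sound with a random body swing.
# An IsLocal branch separates the local mix from other players' sound, which uses the third-person Retract attenuation.
# Non-local playback branches on IsEnemy, playing enemies at full level and others quieter through a shared mix.
# Modulators randomize pitch throughout and swing volume, and a concurrency override caps simultaneous instances.
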SoundCue /GrappleGloves/Items/GrappleGloves/Sounds/GrappleGlove_Detach_Cue
root=Branch_0
nodes [
 {id=Attenuation_1, children [Mixer_0]}
 {id=Attenuation_2, children [Branch_1], properties {attenuationSettings=/Game/Sounds/Attenuation/Fort_GrappleGloves_Retract_3P_Att}}
 {id=Branch_0, children [Attenuation_1, Attenuation_2, Attenuation_2], properties {boolParameterName=IsLocal}}
 {id=Branch_1, children [Mixer_3,Mixer_4,Mixer_4], properties {boolParameterName=IsEnemy}}
 {id=Mixer_0, children [Modulator_2,Modulator_5], properties {inputVolume [1.3,0.6]}}
 {id=Mixer_2, children [Modulator_6,Modulator_7], properties {inputVolume [1.7,1]}}
 {id=Mixer_3, children [Mixer_2], properties {inputVolume [1]}}
 {id=Mixer_4, children [Mixer_2], properties {inputVolume [0.4]}}
 {id=Modulator_2, children [Random_3], properties {pitchMax=1.2, pitchMin=0.9}}
 {id=Modulator_5, children [Random_5], properties {pitchMax=1.2, pitchMin=0.9, volumeMax=1.1, volumeMin=0.9}}
 {id=Modulator_6, children [Random_1], properties {pitchMax=1.2, pitchMin=0.9}}
 {id=Modulator_7, children [Random_2], properties {pitchMax=1.2, pitchMin=0.9, volumeMax=1.1, volumeMin=0.9}}
 {id=Random_1, children [WavePlayer_5,WavePlayer_6,WavePlayer_7], properties {bShouldExcludeFromBranchCulling=true, weights [1,1,1]}}
 {
  id=Random_2
  children [WavePlayer_8, WavePlayer_9, WavePlayer_10, WavePlayer_11]
  properties {bShouldExcludeFromBranchCulling=true, weights [1,1,1,1]}
 }
 {id=Random_3, children [WavePlayer_16, WavePlayer_17, WavePlayer_18], properties {bShouldExcludeFromBranchCulling=true, weights [1,1,1]}}
 {
  id=Random_5
  children [WavePlayer_22, WavePlayer_23, WavePlayer_24, WavePlayer_25]
  properties {bShouldExcludeFromBranchCulling=true, weights [1,1,1,1]}
 }
 {id=WavePlayer_10, properties {soundWaveAssetPtr=&/GrappleGlove_BodySwing_02}}
 {id=WavePlayer_11, properties {soundWaveAssetPtr=&/GrappleGlove_BodySwing_03}}
 {id=WavePlayer_16, properties {soundWaveAssetPtr=&/GrappleGlove_Detach_01}}
 {id=WavePlayer_17, properties {soundWaveAssetPtr=&/GrappleGlove_Detach_02}}
 {id=WavePlayer_18, properties {soundWaveAssetPtr=&/GrappleGlove_Detach_03}}
 {id=WavePlayer_22, properties {soundWaveAssetPtr=&/GrappleGlove_BodySwing_04}}
 {id=WavePlayer_23, properties {soundWaveAssetPtr=&/GrappleGlove_BodySwing_01}}
 {id=WavePlayer_24, properties {soundWaveAssetPtr=&GrappleGlove_BodySwing_02}}
 {id=WavePlayer_25, properties {soundWaveAssetPtr=&GrappleGlove_BodySwing_03}}
 {id=WavePlayer_5, properties {soundWaveAssetPtr=&GrappleGlove_Detach_01}}
 {id=WavePlayer_6, properties {soundWaveAssetPtr=&GrappleGlove_Detach_02}}
 {id=WavePlayer_7, properties {soundWaveAssetPtr=&GrappleGlove_Detach_03}}
 {id=WavePlayer_8, properties {soundWaveAssetPtr=&GrappleGlove_BodySwing_04}}
 {id=WavePlayer_9, properties {soundWaveAssetPtr=&GrappleGlove_BodySwing_01}}
]
properties {
 bOverrideConcurrency=true
 maxDistance=2097152
 soundClassObject=&Fort_Gadgets_Items
 volumeMultiplier=1
 concurrencyOverrides {maxCount=3}
}
