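# Sound cue for the Grapple Gloves detach whoosh, in the gadget/item sound class with a three-voice concurrency override.
# An IsLocal branch splits local playback from remote playback, which uses third-person grappling movement attenuation.
# The remote path checks IsEnemy and is silent when it is false, but plays when true or unset.
# Both paths randomly pick from eight short or long body-swing waves, with layered pitch and volume modulation.
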SoundCue /GrappleGloves/Items/GrappleGloves/Sounds/GrappleGlove_Detach_Whoosh_Cue
root=Branch_0
nodes [
 {id=Attenuation_0, children [Mixer_0]}
 {id=Attenuation_1, children [Branch_1], properties {attenuationSettings=/Game/Sounds/Attenuation/Fort_Grappling_Movement_3P_Att}}
 {id=Branch_0, children [Attenuation_0, Attenuation_1, Attenuation_1], properties {boolParameterName=IsLocal}}
 {id=Branch_1, children [Mixer_1,null,Mixer_1], properties {boolParameterName=IsEnemy}}
 {id=Mixer_0, children [Modulator_9], properties {inputVolume [1]}}
 {id=Mixer_1, children [Modulator_17], properties {inputVolume [1]}}
 {id=Modulator_0, children [WavePlayer_16], properties {pitchMax=1.55, pitchMin=1.15, volumeMax=0.8, volumeMin=0.7}}
 {id=Modulator_1, children [WavePlayer_12], properties {pitchMax=1.55, pitchMin=1.15, volumeMax=0.8, volumeMin=0.7}}
 {id=Modulator_10, children [WavePlayer_13], properties {pitchMax=1.25, pitchMin=1.15, volumeMax=0.8, volumeMin=0.7}}
 {id=Modulator_11, children [WavePlayer_14], properties {pitchMax=1.25, pitchMin=1.15, volumeMax=0.8, volumeMin=0.7}}
 {id=Modulator_12, children [WavePlayer_15], properties {pitchMax=1.25, pitchMin=1.15, volumeMax=0.8, volumeMin=0.7}}
 {id=Modulator_13, children [WavePlayer_6], properties {pitchMax=0.8, pitchMin=0.9, volumeMax=1.1, volumeMin=1}}
 {id=Modulator_14, children [WavePlayer_7], properties {pitchMax=0.8, pitchMin=0.9, volumeMax=1.1, volumeMin=1}}
 {id=Modulator_15, children [WavePlayer_8], properties {pitchMax=0.8, pitchMin=0.9, volumeMax=1.1, volumeMin=1}}
 {id=Modulator_16, children [WavePlayer_9], properties {pitchMax=0.8, pitchMin=0.9, volumeMax=1.1, volumeMin=1}}
 {id=Modulator_17, children [Random_1], properties {pitchMax=1.25, volumeMax=1, volumeMin=1}}
 {id=Modulator_2, children [WavePlayer_4], properties {pitchMax=1.25, pitchMin=1.15, volumeMax=0.8, volumeMin=0.7}}
 {id=Modulator_3, children [WavePlayer_10], properties {pitchMax=1.25, pitchMin=1.15, volumeMax=0.8, volumeMin=0.7}}
 {id=Modulator_4, children [WavePlayer_11], properties {pitchMax=1.25, pitchMin=1.15, volumeMax=0.8, volumeMin=0.7}}
 {id=Modulator_5, children [WavePlayer_0], properties {pitchMax=0.8, pitchMin=0.9, volumeMax=1.1, volumeMin=1}}
 {id=Modulator_6, children [WavePlayer_1], properties {pitchMax=0.8, pitchMin=0.9, volumeMax=1.1, volumeMin=1}}
 {id=Modulator_7, children [WavePlayer_2], properties {pitchMax=0.8, pitchMin=0.9, volumeMax=1.1, volumeMin=1}}
 {id=Modulator_8, children [WavePlayer_3], properties {pitchMax=0.8, pitchMin=0.9, volumeMax=1.1, volumeMin=1}}
 {id=Modulator_9, children [Random_0], properties {pitchMax=1.25, volumeMax=1, volumeMin=1}}
 {
  id=Random_0
  children [Modulator_5,Modulator_6,Modulator_7,Modulator_8,Modulator_2,Modulator_3,Modulator_4,Modulator_1]
  properties {weights [1,1,1,1,1,1,1,1]}
 }
 {
  id=Random_1
  children [Modulator_13,Modulator_14,Modulator_15,Modulator_16,Modulator_10,Modulator_11,Modulator_12,Modulator_0]
  properties {weights [1,1,1,1,1,1,1,1]}
 }
 {id=WavePlayer_0, properties {soundWaveAssetPtr=&/GrappleGlove_BodySwing_01}}
 {id=WavePlayer_1, properties {soundWaveAssetPtr=&/GrappleGlove_BodySwing_02}}
 {id=WavePlayer_10, properties {soundWaveAssetPtr=&/GrappleGlove_BodySwing_Long_02}}
 {id=WavePlayer_11, properties {soundWaveAssetPtr=&/GrappleGlove_BodySwing_Long_03}}
 {id=WavePlayer_12, properties {soundWaveAssetPtr=&/GrappleGlove_BodySwing_Long_04}}
 {id=WavePlayer_13, properties {soundWaveAssetPtr=&/GrappleGlove_BodySwing_Long_01}}
 {id=WavePlayer_14, properties {soundWaveAssetPtr=&GrappleGlove_BodySwing_Long_02}}
 {id=WavePlayer_15, properties {soundWaveAssetPtr=&GrappleGlove_BodySwing_Long_03}}
 {id=WavePlayer_16, properties {soundWaveAssetPtr=&GrappleGlove_BodySwing_Long_04}}
 {id=WavePlayer_2, properties {soundWaveAssetPtr=&/GrappleGlove_BodySwing_03}}
 {id=WavePlayer_3, properties {soundWaveAssetPtr=&/GrappleGlove_BodySwing_04}}
 {id=WavePlayer_4, properties {soundWaveAssetPtr=&GrappleGlove_BodySwing_Long_01}}
 {id=WavePlayer_6, properties {soundWaveAssetPtr=&GrappleGlove_BodySwing_01}}
 {id=WavePlayer_7, properties {soundWaveAssetPtr=&GrappleGlove_BodySwing_02}}
 {id=WavePlayer_8, properties {soundWaveAssetPtr=&GrappleGlove_BodySwing_03}}
 {id=WavePlayer_9, properties {soundWaveAssetPtr=&GrappleGlove_BodySwing_04}}
]
properties {
 bOverrideConcurrency=true
 maxDistance=2097152
 soundClassObject=&Fort_Gadgets_Items
 volumeMultiplier=1
 concurrencyOverrides {maxCount=3}
}
