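# Sound cue for the impact sound when the Grapple Glove item is dropped.
# A Random node picks between two equally weighted impact waves, feeding a Modulator that slightly varies pitch and volume on each play.
# The cue plays at reduced volume in the Fort_Loot_Collect sound class, with overridden attenuation for short-range falloff and a concurrency cap on simultaneous instances.
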SoundCue /GrappleGloves/Items/GrappleGloves/Sounds/GrappleGlove_DropImpact_Cue
root=Modulator_0
nodes [
 {id=Modulator_0, children [Random_1], properties {pitchMax=1.1, pitchMin=0.9, volumeMax=1.1, volumeMin=0.9}}
 {id=Random_1, children [WavePlayer_2,WavePlayer_3], properties {bShouldExcludeFromBranchCulling=true, weights [1,1]}}
 {id=WavePlayer_2, properties {soundWaveAssetPtr=&/GrappleGlove_DropImpact_01}}
 {id=WavePlayer_3, properties {soundWaveAssetPtr=&/GrappleGlove_DropImpact_02}}
]
properties {
 bOverrideAttenuation=true
 bOverrideConcurrency=true
 maxDistance=1000
 soundClassObject=&Fort_Loot_Collect
 volumeMultiplier=0.4
 attenuationOverrides {falloffDistance=1000, attenuationShapeExtents {x=0}}
 concurrencyOverrides {maxCount=3}
}
