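# SoundCue that plays the drop sound for the Grapple Glove item.
# A random node picks one of three drop variations, mixed with a quieter constant layer sound and passed through a Modulator at the root.
# It uses weapon-drop attenuation and the loot-collect sound class, with an overridden concurrency limit to cap simultaneous instances.
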SoundCue /GrappleGloves/Items/GrappleGloves/Sounds/GrappleGlove_Drop_Cue
root=Modulator_0
nodes [
 {id=Mixer_0, children [Random_1,WavePlayer_5], properties {inputVolume [1,0.6]}}
 {id=Modulator_0, children [Mixer_0]}
 {id=Random_1, children [WavePlayer_2,WavePlayer_3,WavePlayer_4], properties {weights [1,1,1]}}
 {id=WavePlayer_2, properties {soundWaveAssetPtr=&/GrappleGlove_Drop_01}}
 {id=WavePlayer_3, properties {soundWaveAssetPtr=&/GrappleGlove_Drop_02}}
 {id=WavePlayer_4, properties {soundWaveAssetPtr=&/GrappleGlove_Drop_03}}
 {id=WavePlayer_5, properties {soundWaveAssetPtr=&/GrappleGlove_Drop_Layer_01}}
]
properties {
 attenuationSettings=&Fort_Weapon_Drop_Att
 bOverrideConcurrency=true
 maxDistance=3000
 soundClassObject=&Fort_Loot_Collect
 volumeMultiplier=0.4
 concurrencyOverrides {maxCount=3}
}
