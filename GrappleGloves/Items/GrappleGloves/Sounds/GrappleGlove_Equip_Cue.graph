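# Sound cue for the Grapple Gloves equip foley, on the weapon foley sound class with weapon reload attenuation.
# A Random node picks one of three equally weighted equip waves, and a Modulator above it slightly varies the pitch on each play.
# Concurrency is overridden to allow a single instance per owner, stopping the oldest when a new one starts.
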SoundCue /GrappleGloves/Items/GrappleGloves/Sounds/GrappleGlove_Equip_Cue
root=Modulator_0
nodes [
 {id=Modulator_0, children [Random_1], properties {pitchMax=1.1, pitchMin=0.9}}
 {id=Random_1, children [WavePlayer_2,WavePlayer_3,WavePlayer_4], properties {weights [1,1,1]}}
 {id=WavePlayer_2, properties {soundWaveAssetPtr=&/GrappleGlove_Equip_01}}
 {id=WavePlayer_3, properties {soundWaveAssetPtr=&/GrappleGlove_Equip_02}}
 {id=WavePlayer_4, properties {soundWaveAssetPtr=&/GrappleGlove_Equip_03}}
]
properties {
 attenuationSettings=&Fort_Weapon_Reload_Att
 bOverrideConcurrency=true
 maxDistance=2250
 soundClassObject=&Fort_WeaponFoley
 volumeMultiplier=0.301
 concurrencyOverrides {bLimitToOwner=true, maxCount=1, resolutionRule=StopOldest}
}
