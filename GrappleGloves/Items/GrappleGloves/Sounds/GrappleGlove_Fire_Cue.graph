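# Sound cue for firing the Grapple Gloves, with separate local and non-local versions, limited by gunfire concurrency sets.
# An IsLocal branch picks a first-person close layer, or otherwise a quality-level switch mixing third-person close and distant medium layers, each with its own attenuation.
# The first quality level draws on ten variations per layer, while the others share smaller pools.
# Every layer gets random wave selection, a sound class override and modulation.
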SoundCue /GrappleGloves/Items/GrappleGloves/Sounds/GrappleGlove_Fire_Cue
root=Branch_0
nodes [
 {id=Attenuation_0, children [Random_7], properties {attenuationSettings=&Fort_Weapon_Projectile_Close_Athena_Att}}
 {
  id=Attenuation_1
  children [Random_8]
  properties {attenuationSettings=&Fort_Weapon_Projectile_Medium_Mono_Athena_Att}
 }
 {id=Attenuation_2, children [Random_1]}
 {id=Attenuation_3, children [Random_0], properties {attenuationSettings=&Fort_Weapon_Projectile_Close_Athena_Att}}
 {
  id=Attenuation_4
  children [Random_3]
  properties {attenuationSettings=&Fort_Weapon_Projectile_Medium_Mono_Athena_Att}
 }
 {id=Branch_0, children [Mixer_2, QualityLevel_0, QualityLevel_0], properties {boolParameterName=IsLocal}}
 {id=Mixer_0, children [Modulator_2,Modulator_3], properties {inputVolume [0.65,0.4]}}
 {id=Mixer_1, children [Modulator_0,Modulator_1], properties {inputVolume [0.6,0.3]}}
 {id=Mixer_2, children [Modulator_4], properties {inputVolume [1.15]}}
 {id=Modulator_0, children [SoundClass_3], properties {pitchMax=1.1, pitchMin=0.9, volumeMax=0.85, volumeMin=0.7}}
 {id=Modulator_1, children [SoundClass_2], properties {volumeMax=1.1, volumeMin=0.9}}
 {id=Modulator_2, children [SoundClass_1], properties {volumeMax=1.1, volumeMin=0.9}}
 {id=Modulator_3, children [SoundClass_0], properties {volumeMax=1.1, volumeMin=0.9}}
 {id=Modulator_4, children [SoundClass_4], properties {pitchMax=1.2, pitchMin=0.9, volumeMax=1, volumeMin=1}}
 {id=QualityLevel_0, children [Mixer_1,Mixer_0,Mixer_0,Mixer_0]}
 {
  id=Random_0
  children [
   WavePlayer_1
   WavePlayer_7
   WavePlayer_10
   WavePlayer_11
   WavePlayer_12
   WavePlayer_13
   WavePlayer_18
   WavePlayer_21
   WavePlayer_22
   WavePlayer_23
  ]
  properties {weights [1,1,1,1,1,1,1,1,1,1]}
 }
 {
  id=Random_1
  children [
   WavePlayer_6
   WavePlayer_8
   WavePlayer_9
   WavePlayer_14
   WavePlayer_15
   WavePlayer_16
   WavePlayer_17
   WavePlayer_19
   WavePlayer_20
   WavePlayer_26
  ]
  properties {weights [1,1,1,1,1,1,1,1,1,1]}
 }
 {
  id=Random_3
  children [
   WavePlayer_24
   WavePlayer_25
   WavePlayer_27
   WavePlayer_28
   WavePlayer_29
   WavePlayer_30
   WavePlayer_31
   WavePlayer_32
   WavePlayer_33
   WavePlayer_34
  ]
  properties {weights [1,1,1,1,1,1,1,1,1,1]}
 }
 {id=Random_7, children [WavePlayer_0,WavePlayer_2,WavePlayer_3,WavePlayer_4], properties {weights [1,1,1,1]}}
 {id=Random_8, children [WavePlayer_50, WavePlayer_51, WavePlayer_52], properties {weights [1,1,1]}}
 {
  id=SoundClass_0
  children [Attenuation_1]
  properties {soundClassOverride=&Fort_WeaponRanged_Fire_Distant}
 }
 {
  id=SoundClass_1
  children [Attenuation_0]
  properties {soundClassOverride=&Fort_WeaponRanged_Fire_3P_PSM}
 }
 {
  id=SoundClass_2
  children [Attenuation_4]
  properties {soundClassOverride=&Fort_WeaponRanged_Fire_Distant}
 }
 {
  id=SoundClass_3
  children [Attenuation_3]
  properties {soundClassOverride=&Fort_WeaponRanged_Fire_3P_PSM}
 }
 {
  id=SoundClass_4
  children [Attenuation_2]
  properties {soundClassOverride=&Fort_WeaponRanged_Fire_1P_PSM}
 }
 {id=WavePlayer_0, properties {soundWaveAssetPtr=&/GrappleGlove_Fire_Close_01}}
 {id=WavePlayer_1, properties {soundWaveAssetPtr=&GrappleGlove_Fire_Close_01}}
 {id=WavePlayer_10, properties {soundWaveAssetPtr=&/GrappleGlove_Fire_Close_03}}
 {id=WavePlayer_11, properties {soundWaveAssetPtr=&/GrappleGlove_Fire_Close_04}}
 {id=WavePlayer_12, properties {soundWaveAssetPtr=&/GrappleGlove_Fire_Close_05}}
 {id=WavePlayer_13, properties {soundWaveAssetPtr=&/GrappleGlove_Fire_Close_06}}
 {id=WavePlayer_14, properties {soundWaveAssetPtr=&GrappleGlove_Fire_Close_04}}
 {id=WavePlayer_15, properties {soundWaveAssetPtr=&GrappleGlove_Fire_Close_05}}
 {id=WavePlayer_16, properties {soundWaveAssetPtr=&GrappleGlove_Fire_Close_06}}
 {id=WavePlayer_17, properties {soundWaveAssetPtr=&/GrappleGlove_Fire_Close_07}}
 {id=WavePlayer_18, properties {soundWaveAssetPtr=&GrappleGlove_Fire_Close_07}}
 {id=WavePlayer_19, properties {soundWaveAssetPtr=&/GrappleGlove_Fire_Close_08}}
 {id=WavePlayer_2, properties {soundWaveAssetPtr=&/GrappleGlove_Fire_Close_02}}
 {id=WavePlayer_20, properties {soundWaveAssetPtr=&/GrappleGlove_Fire_Close_09}}
 {id=WavePlayer_21, properties {soundWaveAssetPtr=&GrappleGlove_Fire_Close_08}}
 {id=WavePlayer_22, properties {soundWaveAssetPtr=&GrappleGlove_Fire_Close_09}}
 {id=WavePlayer_23, properties {soundWaveAssetPtr=&/GrappleGlove_Fire_Close_10}}
 {id=WavePlayer_24, properties {soundWaveAssetPtr=&/GrappleGlove_Fire_Medium_01}}
 {id=WavePlayer_25, properties {soundWaveAssetPtr=&/GrappleGlove_Fire_Medium_02}}
 {id=WavePlayer_26, properties {soundWaveAssetPtr=&GrappleGlove_Fire_Close_10}}
 {id=WavePlayer_27, properties {soundWaveAssetPtr=&/GrappleGlove_Fire_Medium_03}}
 {id=WavePlayer_28, properties {soundWaveAssetPtr=&/GrappleGlove_Fire_Medium_04}}
 {id=WavePlayer_29, properties {soundWaveAssetPtr=&/GrappleGlove_Fire_Medium_05}}
 {id=WavePlayer_3, properties {soundWaveAssetPtr=&GrappleGlove_Fire_Close_03}}
 {id=WavePlayer_30, properties {soundWaveAssetPtr=&/GrappleGlove_Fire_Medium_06}}
 {id=WavePlayer_31, properties {soundWaveAssetPtr=&/GrappleGlove_Fire_Medium_07}}
 {id=WavePlayer_32, properties {soundWaveAssetPtr=&/GrappleGlove_Fire_Medium_08}}
 {id=WavePlayer_33, properties {soundWaveAssetPtr=&/GrappleGlove_Fire_Medium_09}}
 {id=WavePlayer_34, properties {soundWaveAssetPtr=&/GrappleGlove_Fire_Medium_10}}
 {id=WavePlayer_4, properties {soundWaveAssetPtr=&GrappleGlove_Fire_Close_04}}
 {id=WavePlayer_50, properties {soundWaveAssetPtr=&GrappleGlove_Fire_Medium_01}}
 {id=WavePlayer_51, properties {soundWaveAssetPtr=&GrappleGlove_Fire_Medium_02}}
 {id=WavePlayer_52, properties {soundWaveAssetPtr=&GrappleGlove_Fire_Medium_03}}
 {id=WavePlayer_6, properties {soundWaveAssetPtr=&GrappleGlove_Fire_Close_01}}
 {id=WavePlayer_7, properties {soundWaveAssetPtr=&GrappleGlove_Fire_Close_02}}
 {id=WavePlayer_8, properties {soundWaveAssetPtr=&GrappleGlove_Fire_Close_02}}
 {id=WavePlayer_9, properties {soundWaveAssetPtr=&GrappleGlove_Fire_Close_03}}
]
properties {
 maxDistance=2097152
 volumeMultiplier=1
 attenuationOverrides {falloffDistance=5000, attenuationShapeExtents {x=1000}}
 concurrencyOverrides {resolutionRule=PreventNew}
 concurrencySet [&Weapon_Gunfire_Local_Conc, &Weapon_Gunfire_Global_Conc]
}
