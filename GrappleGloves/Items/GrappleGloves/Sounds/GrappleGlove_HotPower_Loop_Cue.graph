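# Looping sound cue for the grapple glove "HotPower" effect, with intensity that tracks a Velocity parameter.
# A mixer layers three looping glider thrust sounds, with the TNTina layer at reduced volume and a fixed raised pitch.
# The mix passes through a random pitch modulator, then a continuous modulator that maps Velocity to pitch and volume.
# It uses third-person grappling movement attenuation and the gadget items sound class.
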SoundCue /GrappleGloves/Items/GrappleGloves/Sounds/GrappleGlove_HotPower_Loop_Cue
root=ModulatorContinuous_0
nodes [
 {id=Mixer_0, children [Modulator_1,WavePlayer_2,WavePlayer_1], properties {inputVolume [0.5,1,1]}}
 {
  id=ModulatorContinuous_0
  children [Modulator_0]
  properties {
   pitchModulationParams {default=0, maxOutput=1.5, minOutput=1, parameterName=Velocity}
   volumeModulationParams {default=0, minOutput=0.6, parameterName=Velocity}
  }
 }
 {id=Modulator_0, children [Mixer_0], properties {pitchMax=1.2, pitchMin=1}}
 {id=Modulator_1, children [WavePlayer_5], properties {pitchMax=1.2, pitchMin=1.2}}
 {
  id=WavePlayer_1
  properties {bLooping=true, soundWaveAssetPtr=/Game/Sounds/Fort_Gadgets/Athena/Bistro_Astronaut/Glider_Bistro_Astronaut_Thruster_Forwards_Loop}
 }
 {id=WavePlayer_2, properties {bLooping=true, soundWaveAssetPtr=/Game/Sounds/Fort_Gadgets/Athena/Glider_DarkNinja_ThrustLoop}}
 {id=WavePlayer_5, properties {bLooping=true, soundWaveAssetPtr=/Game/Sounds/Fort_Gadgets/Athena/TNTina/Glider_TNTina_ThrustLoop_01}}
]
properties {
 attenuationSettings=/Game/Sounds/Attenuation/Fort_Grappling_Movement_3P_Att
 maxDistance=6000
 soundClassObject=&Fort_Gadgets_Items
 volumeMultiplier=0.6
}
