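# Sound cue for the Grapple Glove's player-impact sound.
# A single Wave Player serves as the root and plays one impact wave, with no randomization or modulation.
# It routes through the Fort_Impacts sound class and borrows the long bullet opponent-impact attenuation asset for spatial falloff.
# A concurrency override caps how many instances play at once.
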SoundCue /GrappleGloves/Items/GrappleGloves/Sounds/GrappleGlove_ImpactPlayer_Cue
root=WavePlayer_0
nodes [{id=WavePlayer_0, properties {soundWaveAssetPtr=&/GrappleGlove_ImpactPlayer_01}}]
properties {
 attenuationSettings=/Game/Sounds/Attenuation/Fort_Bullet_ImpactOpponent_Lng_Att
 bOverrideConcurrency=true
 maxDistance=5500
 soundClassObject=&Fort_Impacts
 volumeMultiplier=1
 concurrencyOverrides {maxCount=3}
}
