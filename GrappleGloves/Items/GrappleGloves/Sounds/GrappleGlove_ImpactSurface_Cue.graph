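# Sound cue for the Grapple Gloves surface impact, in the gadget items sound class with a concurrency override.
# A mixer layers a random impact sound with a random yank sound that is quieter, pitched down and starts after a short fixed delay.
# Modulators add slight pitch and volume variation, and the mix plays through the gloves' third-person retract attenuation asset.
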SoundCue /GrappleGloves/Items/GrappleGloves/Sounds/GrappleGlove_ImpactSurface_Cue
root=Attenuation_4
nodes [
 {id=Attenuation_4, children [Modulator_2], properties {attenuationSettings=/Game/Sounds/Attenuation/Fort_GrappleGloves_Retract_3P_Att}}
 {id=Delay_0, children [Random_0], properties {delayMax=0.3, delayMin=0.3}}
 {id=Mixer_0, children [Modulator_4,Modulator_5], properties {inputVolume [1,0.5]}}
 {id=Modulator_0, children [Mixer_0], properties {pitchMax=1, pitchMin=1, volumeMax=1, volumeMin=0.9}}
 {id=Modulator_2, children [Modulator_0], properties {volumeMax=1, volumeMin=1}}
 {id=Modulator_4, children [Random_1], properties {pitchMax=1, pitchMin=0.9, volumeMax=1, volumeMin=1}}
 {id=Modulator_5, children [Delay_0], properties {pitchMax=0.9, pitchMin=0.8}}
 {id=Random_0, children [WavePlayer_0,WavePlayer_1,WavePlayer_2], properties {bShouldExcludeFromBranchCulling=true, weights [1,1,1]}}
 {
  id=Random_1
  children [WavePlayer_10, WavePlayer_11, WavePlayer_12, WavePlayer_14]
  properties {bShouldExcludeFromBranchCulling=true, weights [1,1,1,1]}
 }
 {id=WavePlayer_0, properties {soundWaveAssetPtr=&/GrappleGlove_Yank_01}}
 {id=WavePlayer_1, properties {soundWaveAssetPtr=&/GrappleGlove_Yank_02}}
 {id=WavePlayer_10, properties {soundWaveAssetPtr=&/GrappleGlove_ImpactSurface_01}}
 {id=WavePlayer_11, properties {soundWaveAssetPtr=&/GrappleGlove_ImpactSurface_02}}
 {id=WavePlayer_12, properties {soundWaveAssetPtr=&/GrappleGlove_ImpactSurface_03}}
 {id=WavePlayer_14, properties {soundWaveAssetPtr=&/GrappleGlove_ImpactSurface_05}}
 {id=WavePlayer_2, properties {soundWaveAssetPtr=&/GrappleGlove_Yank_03}}
]
properties {
 bOverrideConcurrency=true
 maxDistance=9000
 soundClassObject=&Fort_Gadgets_Items
 volumeMultiplier=0.8
 concurrencyOverrides {maxCount=4}
}
