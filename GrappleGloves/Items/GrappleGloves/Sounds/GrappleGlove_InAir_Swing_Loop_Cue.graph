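# Looping cue for Grapple Glove in-air swinging, with separate first- and third-person branches.
# IsLocal picks the 1P mix or a similar 3P mix without whoosh-speed and dirt layers, which IsTeam makes quieter for teammates, louder and higher-pitched otherwise.
# An Attached crossfade blends speed loops, plus IsMaxVelocity thrust, into swing, wire-stress, whoosh, water-carve and skid layers.
# Speed, Angle, Length, WaterImmersion and Skidding continuously drive layer volume or pitch.
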SoundCue /GrappleGloves/Items/GrappleGloves/Sounds/GrappleGlove_InAir_Swing_Loop_Cue
root=Branch_4
nodes [
 {id=Attenuation_4, children [Modulator_12], properties {attenuationSettings=/Game/Sounds/Attenuation/Fort_GrappleGloves_Swinging_Loop_1P_Att}}
 {
  id=Attenuation_5
  children [ModulatorContinuous_27]
  properties {attenuationSettings=/Game/Sounds/Attenuation/Fort_Grappling_Movement_3P_Att}
 }
 {id=Branch_0, children [Modulator_9,null,null], properties {boolParameterName=IsMaxVelocity}}
 {id=Branch_1, children [Modulator_19,null,null], properties {boolParameterName=IsMaxVelocity}}
 {id=Branch_4, children [Attenuation_4, Branch_5, Branch_5], properties {boolParameterName=IsLocal}}
 {id=Branch_5, children [Modulator_0,Modulator_1,Modulator_1], properties {boolParameterName=IsTeam}}
 {id=Mixer_0, children [Modulator_10,WavePlayer_2,null], properties {inputVolume [0.5,1,1]}}
 {
  id=Mixer_1
  children [ModulatorContinuous_5, ModulatorContinuous_7, ModulatorContinuous_9, ModulatorContinuous_2, ModulatorContinuous_10, ModulatorContinuous_11]
  properties {inputVolume [1,1,0.5,0.9,1,1.8]}
 }
 {id=Mixer_2, children [Random_10,Random_11,Random_9], properties {inputVolume [0.7,0.7,0.7]}}
 {id=Mixer_3, children [Branch_0,Modulator_2], properties {inputVolume [0.5,0.9]}}
 {
  id=Mixer_4
  children [
   ModulatorContinuous_16
   ModulatorContinuous_18
   ModulatorContinuous_22
   ModulatorContinuous_20
   ModulatorContinuous_24
   ModulatorContinuous_25
  ]
  properties {inputVolume [1,1,0.5,0.9,1,1.8]}
 }
 {id=Mixer_5, children [Random_14,Random_15,null], properties {inputVolume [0.7,0.7,0.7]}}
 {id=Mixer_6, children [Modulator_20, WavePlayer_57, null], properties {inputVolume [0.5,1,1]}}
 {id=Mixer_7, children [Branch_1,Modulator_11], properties {inputVolume [0.5,0.6]}}
 {
  id=ModulatorContinuous_0
  children [WavePlayer_3]
  properties {
   pitchModulationParams {default=0, minOutput=1, parameterName=Speed}
   volumeModulationParams {default=0, minInput=0.3, minOutput=0.01, parameterName=Speed}
  }
 }
 {
  id=ModulatorContinuous_1
  children [Modulator_7]
  properties {
   pitchModulationParams {default=0, maxOutput=1.1, minOutput=1, parameterName=Speed}
   volumeModulationParams {minOutput=0.1, parameterName=Speed}
  }
 }
 {
  id=ModulatorContinuous_10
  children [ModulatorContinuous_1]
  properties {
   pitchModulationParams {default=0, maxOutput=0.9, minOutput=1, parameterName=WaterImmersion}
   volumeModulationParams {default=0, parameterName=WaterImmersion}
  }
 }
 {
  id=ModulatorContinuous_11
  children [ModulatorContinuous_12]
  properties {pitchModulationParams {default=0, minOutput=1, parameterName=Speed}, volumeModulationParams {default=0, parameterName=Speed}}
 }
 {
  id=ModulatorContinuous_12
  children [Modulator_8]
  properties {pitchModulationParams {minOutput=1}, volumeModulationParams {default=0, parameterName=Skidding}}
 }
 {
  id=ModulatorContinuous_13
  children [ModulatorContinuous_3]
  properties {
   pitchModulationParams {default=0, minOutput=1, parameterName=SpeedMultiplier}
   volumeModulationParams {maxInput=1.5, maxOutput=1.2, minInput=1, minOutput=0.8, paramMode=MPM_Direct, parameterName=Speed}
  }
 }
 {
  id=ModulatorContinuous_14
  children [ParamCrossFade_3]
  properties {pitchModulationParams {default=0, minOutput=1, parameterName=Speed}, volumeModulationParams {minOutput=0.2, parameterName=Speed}}
 }
 {
  id=ModulatorContinuous_15
  children [ParamCrossFade_2]
  properties {
   pitchModulationParams {default=0, minOutput=0.6, parameterName=Angle}
   volumeModulationParams {default=0, maxOutput=0.5, minOutput=1, parameterName=Angle}
  }
 }
 {
  id=ModulatorContinuous_16
  children [ModulatorContinuous_15]
  properties {
   pitchModulationParams {default=0, minOutput=0.6, parameterName=Length}
   volumeModulationParams {default=0, maxOutput=0.7, minOutput=1, parameterName=Length}
  }
 }
 {
  id=ModulatorContinuous_17
  children [Modulator_15]
  properties {
   pitchModulationParams {default=0, minOutput=0.8, parameterName=Angle}
   volumeModulationParams {default=0, maxOutput=0.5, minOutput=1, parameterName=Angle}
  }
 }
 {
  id=ModulatorContinuous_18
  children [ModulatorContinuous_17]
  properties {
   pitchModulationParams {default=0, maxOutput=0.5, minOutput=1, parameterName=Length}
   volumeModulationParams {default=0, maxOutput=1.5, minOutput=0.75, parameterName=Length}
  }
 }
 {
  id=ModulatorContinuous_19
  children [WavePlayer_6]
  properties {
   pitchModulationParams {default=0, minOutput=1, parameterName=Speed}
   volumeModulationParams {default=0, minInput=0.3, minOutput=0.01, parameterName=Speed}
  }
 }
 {
  id=ModulatorContinuous_2
  children [ModulatorContinuous_0]
  properties {
   pitchModulationParams {default=0, minOutput=1, parameterName=Angle}
   volumeModulationParams {default=0, maxOutput=0.5, minOutput=1, parameterName=Angle}
  }
 }
 {
  id=ModulatorContinuous_20
  children [ModulatorContinuous_19]
  properties {
   pitchModulationParams {default=0, minOutput=1, parameterName=Angle}
   volumeModulationParams {default=0, maxOutput=0.5, minOutput=1, parameterName=Angle}
  }
 }
 {
  id=ModulatorContinuous_21
  children [Mixer_7]
  properties {
   pitchModulationParams {default=0, maxOutput=1.25, minOutput=0.9, parameterName=Speed}
   volumeModulationParams {minOutput=0.5, parameterName=Speed}
  }
 }
 {
  id=ModulatorContinuous_22
  children [Modulator_16]
  properties {
   pitchModulationParams {default=0, maxInput=0.9, minOutput=1, parameterName=Angle}
   volumeModulationParams {default=0, minOutput=0.5, parameterName=Angle}
  }
 }
 {
  id=ModulatorContinuous_23
  children [Modulator_17]
  properties {
   pitchModulationParams {default=0, maxOutput=1.1, minOutput=1, parameterName=Speed}
   volumeModulationParams {minOutput=0.1, parameterName=Speed}
  }
 }
 {
  id=ModulatorContinuous_24
  children [ModulatorContinuous_23]
  properties {
   pitchModulationParams {default=0, maxOutput=0.9, minOutput=1, parameterName=WaterImmersion}
   volumeModulationParams {default=0, parameterName=WaterImmersion}
  }
 }
 {
  id=ModulatorContinuous_25
  children [ModulatorContinuous_26]
  properties {pitchModulationParams {default=0, minOutput=1, parameterName=Speed}, volumeModulationParams {default=0, parameterName=Speed}}
 }
 {
  id=ModulatorContinuous_26
  children [Modulator_18]
  properties {pitchModulationParams {minOutput=1}, volumeModulationParams {default=0, parameterName=Skidding}}
 }
 {
  id=ModulatorContinuous_27
  children [ModulatorContinuous_14]
  properties {
   pitchModulationParams {default=0, minOutput=1, parameterName=SpeedMultiplier}
   volumeModulationParams {maxInput=1.5, maxOutput=1.2, minInput=1, minOutput=0.8, paramMode=MPM_Direct, parameterName=Speed}
  }
 }
 {
  id=ModulatorContinuous_3
  children [ParamCrossFade_1]
  properties {pitchModulationParams {default=0, minOutput=1, parameterName=Speed}, volumeModulationParams {minOutput=0.2, parameterName=Speed}}
 }
 {
  id=ModulatorContinuous_4
  children [ParamCrossFade_0]
  properties {
   pitchModulationParams {default=0, minOutput=0.6, parameterName=Angle}
   volumeModulationParams {default=0, maxOutput=0.5, minOutput=1, parameterName=Angle}
  }
 }
 {
  id=ModulatorContinuous_5
  children [ModulatorContinuous_4]
  properties {
   pitchModulationParams {default=0, minOutput=0.6, parameterName=Length}
   volumeModulationParams {default=0, maxOutput=0.7, minOutput=1, parameterName=Length}
  }
 }
 {
  id=ModulatorContinuous_6
  children [Modulator_5]
  properties {
   pitchModulationParams {default=0, minOutput=0.8, parameterName=Angle}
   volumeModulationParams {default=0, maxOutput=0.5, minOutput=1, parameterName=Angle}
  }
 }
 {
  id=ModulatorContinuous_7
  children [ModulatorContinuous_6]
  properties {
   pitchModulationParams {default=0, maxOutput=0.5, minOutput=1, parameterName=Length}
   volumeModulationParams {default=0, maxOutput=1.5, minOutput=0.75, parameterName=Length}
  }
 }
 {
  id=ModulatorContinuous_8
  children [Mixer_3]
  properties {
   pitchModulationParams {default=0, maxOutput=1.25, minOutput=0.9, parameterName=Speed}
   volumeModulationParams {minOutput=0.5, parameterName=Speed}
  }
 }
 {
  id=ModulatorContinuous_9
  children [Modulator_6]
  properties {
   pitchModulationParams {default=0, maxInput=0.9, minOutput=1, parameterName=Angle}
   volumeModulationParams {default=0, minOutput=0.5, parameterName=Angle}
  }
 }
 {id=Modulator_0, children [Attenuation_5], properties {pitchMax=1, pitchMin=1, volumeMax=0.6, volumeMin=0.6}}
 {id=Modulator_1, children [Attenuation_5], properties {pitchMax=1.6, pitchMin=1.2, volumeMax=1.5, volumeMin=1.5}}
 {id=Modulator_10, children [WavePlayer_5], properties {pitchMax=1.2, pitchMin=1.2}}
 {id=Modulator_11, children [Random_0], properties {pitchMax=1.2, pitchMin=1, volumeMax=0.8, volumeMin=0.7}}
 {id=Modulator_12, children [ModulatorContinuous_13], properties {volumeMax=1.25, volumeMin=1.25}}
 {id=Modulator_13, children [Random_1], properties {pitchMax=1.1, pitchMin=0.9, volumeMax=0.8, volumeMin=0.7}}
 {id=Modulator_14, children [Random_2], properties {pitchMax=1.1, pitchMin=0.9, volumeMax=0.9, volumeMin=0.8}}
 {id=Modulator_15, children [Random_4], properties {pitchMax=1.1, pitchMin=0.9, volumeMax=1.4, volumeMin=1.2}}
 {id=Modulator_16, children [null], properties {pitchMax=1.1, pitchMin=0.9, volumeMax=0.9, volumeMin=0.8}}
 {id=Modulator_17, children [Random_12], properties {pitchMax=1.2, pitchMin=1, volumeMax=1.3, volumeMin=1.1}}
 {id=Modulator_18, children [Mixer_5], properties {pitchMax=1.2, pitchMin=1, volumeMax=1.2, volumeMin=1}}
 {id=Modulator_19, children [Mixer_6], properties {pitchMax=1.2, pitchMin=1, volumeMax=1.2, volumeMin=1}}
 {id=Modulator_2, children [Random_3], properties {pitchMax=1.2, pitchMin=1, volumeMax=0.8, volumeMin=0.7}}
 {id=Modulator_20, children [WavePlayer_58], properties {pitchMax=1.2, pitchMin=1.2}}
 {id=Modulator_3, children [Random_5], properties {pitchMax=1.1, pitchMin=0.9, volumeMax=0.8, volumeMin=0.7}}
 {id=Modulator_4, children [Random_6], properties {pitchMax=1.1, pitchMin=0.9, volumeMax=0.9, volumeMin=0.8}}
 {id=Modulator_5, children [Random_7], properties {pitchMax=1.1, pitchMin=0.9, volumeMax=1.4, volumeMin=1.2}}
 {id=Modulator_6, children [WavePlayer_10], properties {pitchMax=1.1, pitchMin=0.9, volumeMax=0.9, volumeMin=0.8}}
 {id=Modulator_7, children [Random_8], properties {pitchMax=1.2, pitchMin=1, volumeMax=1.3, volumeMin=1.1}}
 {id=Modulator_8, children [Mixer_2], properties {pitchMax=1.2, pitchMin=1, volumeMax=1.2, volumeMin=1}}
 {id=Modulator_9, children [Mixer_0], properties {pitchMax=1.2, pitchMin=1, volumeMax=1.2, volumeMin=1}}
 {
  id=ParamCrossFade_0
  children [Modulator_3,Modulator_4]
  properties {
   paramName=Angle
   crossFadeInput [
    {fadeInDistanceEnd=0, fadeInDistanceStart=0, fadeOutDistanceEnd=1, fadeOutDistanceStart=0, volume=1}
    {fadeInDistanceEnd=1, fadeInDistanceStart=0, fadeOutDistanceEnd=0, fadeOutDistanceStart=1, volume=1}
   ]
  }
 }
 {
  id=ParamCrossFade_1
  children [ModulatorContinuous_8, Mixer_1]
  properties {
   paramName=Attached
   crossFadeInput [
    {fadeInDistanceEnd=0, fadeInDistanceStart=0, fadeOutDistanceEnd=1, fadeOutDistanceStart=0, volume=1}
    {fadeInDistanceEnd=1, fadeInDistanceStart=0, fadeOutDistanceEnd=2, fadeOutDistanceStart=1, volume=1}
   ]
  }
 }
 {
  id=ParamCrossFade_2
  children [Modulator_13,Modulator_14]
  properties {
   paramName=Angle
   crossFadeInput [
    {fadeInDistanceEnd=0, fadeInDistanceStart=0, fadeOutDistanceEnd=1, fadeOutDistanceStart=0, volume=1}
    {fadeInDistanceEnd=1, fadeInDistanceStart=0, fadeOutDistanceEnd=0, fadeOutDistanceStart=1, volume=1}
   ]
  }
 }
 {
  id=ParamCrossFade_3
  children [ModulatorContinuous_21, Mixer_4]
  properties {
   paramName=Attached
   crossFadeInput [
    {fadeInDistanceEnd=0, fadeInDistanceStart=0, fadeOutDistanceEnd=1, fadeOutDistanceStart=0, volume=1}
    {fadeInDistanceEnd=1, fadeInDistanceStart=0, fadeOutDistanceEnd=2, fadeOutDistanceStart=1, volume=1}
   ]
  }
 }
 {id=Random_0, children [WavePlayer_7,WavePlayer_8,WavePlayer_9], properties {weights [1,1,1]}}
 {id=Random_1, children [WavePlayer_11, WavePlayer_12], properties {weights [1,1]}}
 {id=Random_10, children [WavePlayer_37, WavePlayer_38, WavePlayer_39], properties {bShouldExcludeFromBranchCulling=true, weights [1,1,1]}}
 {id=Random_11, children [WavePlayer_40, WavePlayer_41, WavePlayer_42], properties {bShouldExcludeFromBranchCulling=true, weights [1,1,1]}}
 {id=Random_12, children [WavePlayer_44, WavePlayer_45], properties {bShouldExcludeFromBranchCulling=true, weights [1,1]}}
 {id=Random_14, children [WavePlayer_50, WavePlayer_51, WavePlayer_52], properties {bShouldExcludeFromBranchCulling=true, weights [1,1,1]}}
 {id=Random_15, children [WavePlayer_53, WavePlayer_54, WavePlayer_55], properties {bShouldExcludeFromBranchCulling=true, weights [1,1,1]}}
 {id=Random_2, children [WavePlayer_13, WavePlayer_14, WavePlayer_15], properties {weights [1,1,1]}}
 {id=Random_3, children [WavePlayer_18, WavePlayer_19, WavePlayer_20], properties {weights [1,1,1]}}
 {id=Random_4, children [WavePlayer_16, WavePlayer_17, WavePlayer_30, WavePlayer_43], properties {weights [1,1,1,1]}}
 {id=Random_5, children [WavePlayer_21, WavePlayer_22], properties {weights [1,1]}}
 {id=Random_6, children [WavePlayer_23, WavePlayer_24, WavePlayer_25], properties {weights [1,1,1]}}
 {id=Random_7, children [WavePlayer_26, WavePlayer_27, WavePlayer_28, WavePlayer_29], properties {weights [1,1,1,1]}}
 {id=Random_8, children [WavePlayer_31, WavePlayer_32], properties {bShouldExcludeFromBranchCulling=true, weights [1,1]}}
 {id=Random_9, children [WavePlayer_33, WavePlayer_34, WavePlayer_35, WavePlayer_36], properties {weights [1,1,1,1]}}
 {
  id=WavePlayer_10
  properties {bLooping=true, soundWaveAssetPtr=/Game/Sounds/Fort_GamePlay_Sounds/Player/GrapplingSwings/Player_Grappling_SwingSpeed_Whoosh_Loop}
 }
 {id=WavePlayer_11, properties {bLooping=true, soundWaveAssetPtr=&/GrappleGlove_SwingLoop_Vertical_01}}
 {id=WavePlayer_12, properties {bLooping=true, soundWaveAssetPtr=&/GrappleGlove_SwingLoop_Vertical_02}}
 {
  id=WavePlayer_13
  properties {bLooping=true, soundWaveAssetPtr=&/GrappleGlove_SwingLoop_Horizontal_01}
 }
 {
  id=WavePlayer_14
  properties {bLooping=true, soundWaveAssetPtr=&/GrappleGlove_SwingLoop_Horizontal_02}
 }
 {
  id=WavePlayer_15
  properties {bLooping=true, soundWaveAssetPtr=&/GrappleGlove_SwingLoop_Horizontal_03}
 }
 {
  id=WavePlayer_16
  properties {bLooping=true, soundWaveAssetPtr=&/GrappleGlove_SwingLoop_WireStress_04}
 }
 {
  id=WavePlayer_17
  properties {bLooping=true, soundWaveAssetPtr=&/GrappleGlove_SwingLoop_WireStress_01}
 }
 {
  id=WavePlayer_18
  properties {bLooping=true, soundWaveAssetPtr=&/GrappleGlove_SwingLoop_SwingSpeed_01}
 }
 {
  id=WavePlayer_19
  properties {bLooping=true, soundWaveAssetPtr=&/GrappleGlove_SwingLoop_SwingSpeed_02}
 }
 {id=WavePlayer_2, properties {bLooping=true, soundWaveAssetPtr=/Game/Sounds/Fort_Gadgets/Athena/Glider_DarkNinja_ThrustLoop}}
 {
  id=WavePlayer_20
  properties {bLooping=true, soundWaveAssetPtr=&/GrappleGlove_SwingLoop_SwingSpeed_03}
 }
 {id=WavePlayer_21, properties {bLooping=true, soundWaveAssetPtr=&GrappleGlove_SwingLoop_Vertical_01}}
 {id=WavePlayer_22, properties {bLooping=true, soundWaveAssetPtr=&GrappleGlove_SwingLoop_Vertical_02}}
 {
  id=WavePlayer_23
  properties {bLooping=true, soundWaveAssetPtr=&GrappleGlove_SwingLoop_Horizontal_01}
 }
 {
  id=WavePlayer_24
  properties {bLooping=true, soundWaveAssetPtr=&GrappleGlove_SwingLoop_Horizontal_02}
 }
 {
  id=WavePlayer_25
  properties {bLooping=true, soundWaveAssetPtr=&GrappleGlove_SwingLoop_Horizontal_03}
 }
 {
  id=WavePlayer_26
  properties {bLooping=true, soundWaveAssetPtr=&GrappleGlove_SwingLoop_WireStress_04}
 }
 {
  id=WavePlayer_27
  properties {bLooping=true, soundWaveAssetPtr=&GrappleGlove_SwingLoop_WireStress_01}
 }
 {
  id=WavePlayer_28
  properties {bLooping=true, soundWaveAssetPtr=&/GrappleGlove_SwingLoop_WireStress_02}
 }
 {
  id=WavePlayer_29
  properties {bLooping=true, soundWaveAssetPtr=&/GrappleGlove_SwingLoop_WireStress_03}
 }
 {
  id=WavePlayer_3
  properties {bLooping=true, soundWaveAssetPtr=/Game/Sounds/Fort_GamePlay_Sounds/Player/GrapplingSwings/Player_Grappling_LowestPoint_Whoosh_Loop}
 }
 {
  id=WavePlayer_30
  properties {bLooping=true, soundWaveAssetPtr=&GrappleGlove_SwingLoop_WireStress_02}
 }
 {id=WavePlayer_31, properties {bLooping=true, soundWaveAssetPtr=&/GrappleGlove_SwingLoop_Carve_01}}
 {id=WavePlayer_32, properties {bLooping=true, soundWaveAssetPtr=&/GrappleGlove_SwingLoop_Carve_02}}
 {id=WavePlayer_33, properties {bLooping=true, soundWaveAssetPtr=/Game/Athena/Sounds/Character/Sliding/Waves/Dirt/Sliding_Dirt_Debris_01}}
 {id=WavePlayer_34, properties {bLooping=true, soundWaveAssetPtr=/Game/Athena/Sounds/Character/Sliding/Waves/Dirt/Sliding_Dirt_Debris_02}}
 {id=WavePlayer_35, properties {bLooping=true, soundWaveAssetPtr=/Game/Athena/Sounds/Character/Sliding/Waves/Dirt/Sliding_Dirt_Debris_03}}
 {id=WavePlayer_36, properties {bLooping=true, soundWaveAssetPtr=/Game/Athena/Sounds/Character/Sliding/Waves/Dirt/Sliding_Dirt_Debris_04}}
 {id=WavePlayer_37, properties {bLooping=true, soundWaveAssetPtr=&/GrappleGlove_SlideLoopA_01}}
 {id=WavePlayer_38, properties {bLooping=true, soundWaveAssetPtr=&/GrappleGlove_SlideLoopA_02}}
 {id=WavePlayer_39, properties {bLooping=true, soundWaveAssetPtr=&/GrappleGlove_SlideLoopA_03}}
 {id=WavePlayer_40, properties {bLooping=true, soundWaveAssetPtr=&/GrappleGlove_SlideLoopB_01}}
 {id=WavePlayer_41, properties {bLooping=true, soundWaveAssetPtr=&/GrappleGlove_SlideLoopB_02}}
 {id=WavePlayer_42, properties {bLooping=true, soundWaveAssetPtr=&/GrappleGlove_SlideLoopB_03}}
 {
  id=WavePlayer_43
  properties {bLooping=true, soundWaveAssetPtr=&GrappleGlove_SwingLoop_WireStress_03}
 }
 {id=WavePlayer_44, properties {bLooping=true, soundWaveAssetPtr=&GrappleGlove_SwingLoop_Carve_01}}
 {id=WavePlayer_45, properties {bLooping=true, soundWaveAssetPtr=&GrappleGlove_SwingLoop_Carve_02}}
 {id=WavePlayer_5, properties {bLooping=true, soundWaveAssetPtr=/Game/Sounds/Fort_Gadgets/Athena/TNTina/Glider_TNTina_ThrustLoop_01}}
 {id=WavePlayer_50, properties {bLooping=true, soundWaveAssetPtr=&GrappleGlove_SlideLoopA_01}}
 {id=WavePlayer_51, properties {bLooping=true, soundWaveAssetPtr=&GrappleGlove_SlideLoopA_02}}
 {id=WavePlayer_52, properties {bLooping=true, soundWaveAssetPtr=&GrappleGlove_SlideLoopA_03}}
 {id=WavePlayer_53, properties {bLooping=true, soundWaveAssetPtr=&GrappleGlove_SlideLoopB_01}}
 {id=WavePlayer_54, properties {bLooping=true, soundWaveAssetPtr=&GrappleGlove_SlideLoopB_02}}
 {id=WavePlayer_55, properties {bLooping=true, soundWaveAssetPtr=&GrappleGlove_SlideLoopB_03}}
 {id=WavePlayer_57, properties {bLooping=true, soundWaveAssetPtr=&Glider_DarkNinja_ThrustLoop}}
 {id=WavePlayer_58, properties {bLooping=true, soundWaveAssetPtr=&Glider_TNTina_ThrustLoop_01}}
 {
  id=WavePlayer_6
  properties {bLooping=true, soundWaveAssetPtr=&Player_Grappling_LowestPoint_Whoosh_Loop}
 }
 {
  id=WavePlayer_7
  properties {bLooping=true, soundWaveAssetPtr=&GrappleGlove_SwingLoop_SwingSpeed_01}
 }
 {
  id=WavePlayer_8
  properties {bLooping=true, soundWaveAssetPtr=&GrappleGlove_SwingLoop_SwingSpeed_02}
 }
 {
  id=WavePlayer_9
  properties {bLooping=true, soundWaveAssetPtr=&GrappleGlove_SwingLoop_SwingSpeed_03}
 }
]
properties {maxDistance=6000, soundClassObject=&Fort_Gadgets_Items, volumeMultiplier=1.1}
