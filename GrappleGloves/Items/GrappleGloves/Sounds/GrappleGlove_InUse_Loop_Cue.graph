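# Sound cue named as the in-use loop for the Grapple Gloves item, but it has no nodes and no root, so it has nothing to play.
# It carries only playback settings: a third-person grappling movement attenuation asset, the gadget items sound class, a boosted volume multiplier and a max distance.
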SoundCue /GrappleGloves/Items/GrappleGloves/Sounds/GrappleGlove_InUse_Loop_Cue
root=null
nodes []
properties {
 attenuationSettings=/Game/Sounds/Attenuation/Fort_Grappling_Movement_3P_Att
 maxDistance=6000
 soundClassObject=&Fort_Gadgets_Items
 volumeMultiplier=4
}
