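# Sound cue for the grapple gloves landing sound, routed to the gadget items sound class with a dedicated landing attenuation.
# A mixer layers a heavy landing footstep with a glove land sound, the latter mixed slightly louder.
# Each layer randomly picks one of two waves and passes it through a modulator.
# A concurrency override caps simultaneous instances.
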SoundCue /GrappleGloves/Items/GrappleGloves/Sounds/GrappleGlove_Land_Cue
root=Mixer_1
nodes [
 {id=Mixer_1, children [Modulator_3,Modulator_4], properties {inputVolume [1,1.3]}}
 {id=Modulator_3, children [Random_1]}
 {id=Modulator_4, children [Random_0]}
 {id=Random_0, children [WavePlayer_0,WavePlayer_1], properties {bShouldExcludeFromBranchCulling=true, weights [1,1]}}
 {id=Random_1, children [WavePlayer_2,WavePlayer_3], properties {bShouldExcludeFromBranchCulling=true, weights [1,1]}}
 {id=WavePlayer_0, properties {soundWaveAssetPtr=&/GrappleGlove_Land_01}}
 {id=WavePlayer_1, properties {soundWaveAssetPtr=&/GrappleGlove_Land_02}}
 {id=WavePlayer_2, properties {soundWaveAssetPtr=&/GrappleGlove_Footstep_LandHeavy_01}}
 {id=WavePlayer_3, properties {soundWaveAssetPtr=&/GrappleGlove_Footstep_LandHeavy_02}}
]
properties {
 attenuationSettings=/Game/Sounds/Attenuation/Fort_GrappleGloves_Landing
 bOverrideConcurrency=true
 maxDistance=6000
 soundClassObject=&Fort_Gadgets_Items
 volumeMultiplier=0.6
 concurrencyOverrides {maxCount=3}
}
