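# Sound cue for picking up the Grapple Gloves item.
# A Random node picks one of three equally weighted pickup wave variations, which passes through a Modulator at the root.
# The cue plays in the loot-collect sound class at reduced volume, using the shared weapon pickup attenuation settings.
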SoundCue /GrappleGloves/Items/GrappleGloves/Sounds/GrappleGlove_Pickup_Cue
root=Modulator_0
nodes [
 {id=Modulator_0, children [Random_1]}
 {id=Random_1, children [WavePlayer_2,WavePlayer_3,WavePlayer_4], properties {weights [1,1,1]}}
 {id=WavePlayer_2, properties {soundWaveAssetPtr=&/GrappleGlove_Pickup_01}}
 {id=WavePlayer_3, properties {soundWaveAssetPtr=&/GrappleGlove_Pickup_02}}
 {id=WavePlayer_4, properties {soundWaveAssetPtr=&/GrappleGlove_Pickup_03}}
]
properties {
 attenuationSettings=&Fort_Weapon_Pickup_Att
 maxDistance=3500
 soundClassObject=&Fort_Loot_Collect
 volumeMultiplier=0.259
}
