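# Sound cue for the Grapple Glove player collision, playing a hard object impact in the gadget items sound class.
# A random node picks one of four equally weighted impact waves, then a modulator slightly varies pitch and volume.
# A continuous modulator at the root drives pitch and volume from a Speed parameter.
# Overridden concurrency allows one instance per owner, blocking new plays with a short retrigger time.
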SoundCue /GrappleGloves/Items/GrappleGloves/Sounds/GrappleGlove_Player_Collision_Cue
root=ModulatorContinuous_0
nodes [
 {
  id=ModulatorContinuous_0
  children [Modulator_0]
  properties {pitchModulationParams {minOutput=1, parameterName=Speed}, volumeModulationParams {parameterName=Speed}}
 }
 {id=Modulator_0, children [Random_0], properties {pitchMax=1.1, pitchMin=0.9, volumeMax=1, volumeMin=0.8}}
 {id=Random_0, children [WavePlayer_4, WavePlayer_12, WavePlayer_13, WavePlayer_14], properties {weights [1,1,1,1]}}
 {
  id=WavePlayer_12
  properties {soundWaveAssetPtr=/Game/Sounds/Fort_GamePlay_Sounds/Player/GrapplingSwings/Player_Grappling_Collision_Object_Hard_02}
 }
 {
  id=WavePlayer_13
  properties {soundWaveAssetPtr=/Game/Sounds/Fort_GamePlay_Sounds/Player/GrapplingSwings/Player_Grappling_Collision_Object_Hard_03}
 }
 {
  id=WavePlayer_14
  properties {soundWaveAssetPtr=/Game/Sounds/Fort_GamePlay_Sounds/Player/GrapplingSwings/Player_Grappling_Collision_Object_Hard_04}
 }
 {
  id=WavePlayer_4
  properties {soundWaveAssetPtr=/Game/Sounds/Fort_GamePlay_Sounds/Player/GrapplingSwings/Player_Grappling_Collision_Object_Hard_01}
 }
]
properties {
 bOverrideConcurrency=true
 maxDistance=2097152
 soundClassObject=&Fort_Gadgets_Items
 volumeMultiplier=0.5
 concurrencyOverrides {bLimitToOwner=true, maxCount=1, resolutionRule=PreventNew, retriggerTime=0.2}
}
