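# Sound cue for the grapple glove timeout end warning, classed under gadget item sounds.
# A single warning wave player feeds a Modulator at default settings, with the cue lowering pitch slightly and volume heavily.
# Inline attenuation overrides give HRTF spatialization with zero stereo spread, taking precedence over the referenced prop-search attenuation.
# Concurrency override allows only one instance and blocks new plays while it sounds.
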SoundCue /GrappleGloves/Items/GrappleGloves/Sounds/GrappleGlove_Timeout_EndWarning_Cue
root=Modulator_0
nodes [
 {id=Modulator_0, children [WavePlayer_14]}
 {id=WavePlayer_14, properties {soundWaveAssetPtr=&/GrappleGlove_Timeout_EndWarning_A02B}}
]
properties {
 attenuationSettings=/Game/Sounds/Attenuation/Fort_Prop_Search_Att
 bOverrideAttenuation=true
 bOverrideConcurrency=true
 maxDistance=1300
 pitchMultiplier=0.9
 soundClassObject=&Fort_Gadgets_Items
 volumeMultiplier=0.1
 attenuationOverrides {falloffDistance=1000, spatializationAlgorithm=SPATIALIZATION_HRTF, stereoSpread=0, attenuationShapeExtents {x=300}}
 concurrencyOverrides {maxCount=1, resolutionRule=PreventNew}
}
