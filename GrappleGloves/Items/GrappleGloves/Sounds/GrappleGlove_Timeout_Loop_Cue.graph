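# Sound cue named for the looping timeout sound of the Grapple Gloves item, routed to the gadgets and items sound class at reduced volume.
# The node graph is empty with no root, so the cue references no audio here and carries only playback settings.
# It overrides attenuation with short-range HRTF spatialization and no stereo spread, and limits concurrency to one instance by blocking new plays.
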SoundCue /GrappleGloves/Items/GrappleGloves/Sounds/GrappleGlove_Timeout_Loop_Cue
root=null
nodes []
properties {
 attenuationSettings=/Game/Sounds/Attenuation/Fort_Prop_Search_Att
 bOverrideAttenuation=true
 bOverrideConcurrency=true
 maxDistance=1300
 soundClassObject=&Fort_Gadgets_Items
 volumeMultiplier=0.2
 attenuationOverrides {falloffDistance=1000, spatializationAlgorithm=SPATIALIZATION_HRTF, stereoSpread=0, attenuationShapeExtents {x=300}}
 concurrencyOverrides {maxCount=1, resolutionRule=PreventNew}
}
