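# Sound cue for the Grapple Glove timeout start warning, classed as an item gadget sound.
# A mixer plays three wave layers together (a main warning, beeps and a stall layer), each at its own input volume, under a reduced overall volume.
# It references a prop-search attenuation asset but overrides it with HRTF spatialization and custom falloff.
# Its concurrency override prevents new instances once the limit is reached.
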SoundCue /GrappleGloves/Items/GrappleGloves/Sounds/GrappleGlove_Timeout_StartWarning_Cue
root=Mixer_1
nodes [
 {id=Mixer_1, children [WavePlayer_1, WavePlayer_11, WavePlayer_13], properties {inputVolume [1,1.8,0.8]}}
 {id=WavePlayer_1, properties {soundWaveAssetPtr=&/GrappleGlove_Timeout_StartWarning_A01}}
 {id=WavePlayer_11, properties {soundWaveAssetPtr=&/GrappleGlove_Timeout_StartWarning_LayerBeeps_01}}
 {id=WavePlayer_13, properties {soundWaveAssetPtr=&/GrappleGlove_Timeout_StartWarning_LayerStall_01}}
]
properties {
 attenuationSettings=/Game/Sounds/Attenuation/Fort_Prop_Search_Att
 bOverrideAttenuation=true
 bOverrideConcurrency=true
 maxDistance=1300
 soundClassObject=&Fort_Gadgets_Items
 volumeMultiplier=0.325
 attenuationOverrides {falloffDistance=1000, spatializationAlgorithm=SPATIALIZATION_HRTF, stereoSpread=0, attenuationShapeExtents {x=300}}
 concurrencyOverrides {resolutionRule=PreventNew}
}
