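# Sound cue for the start of the grapple glove timeout, in the gadgets and items sound class.
# A mixer layers one of three random steam waves with one of three random timeout-start waves, the latter passed through a modulator.
# Overridden attenuation gives short-range HRTF spatialization with no stereo spread.
# Overridden concurrency rejects new instances once the limit is reached.
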SoundCue /GrappleGloves/Items/GrappleGloves/Sounds/GrappleGlove_Timeout_Start_Cue
root=Mixer_0
nodes [
 {id=Mixer_0, children [Modulator_1,Random_0], properties {inputVolume [1.2,1.3]}}
 {id=Modulator_1, children [Random_1]}
 {id=Random_0, children [WavePlayer_0,WavePlayer_1,WavePlayer_2], properties {weights [1,1,1]}}
 {id=Random_1, children [WavePlayer_3,WavePlayer_4,WavePlayer_5], properties {weights [1,1,1]}}
 {id=WavePlayer_0, properties {soundWaveAssetPtr=&/GrappleGlove_Steam_01}}
 {id=WavePlayer_1, properties {soundWaveAssetPtr=&/GrappleGlove_Steam_02}}
 {id=WavePlayer_2, properties {soundWaveAssetPtr=&/GrappleGlove_Steam_03}}
 {id=WavePlayer_3, properties {soundWaveAssetPtr=&/GrappleGlove_Timeout_Start_01}}
 {id=WavePlayer_4, properties {soundWaveAssetPtr=&/GrappleGlove_Timeout_Start_02}}
 {id=WavePlayer_5, properties {soundWaveAssetPtr=&/GrappleGlove_Timeout_Start_03}}
]
properties {
 attenuationSettings=/Game/Sounds/Attenuation/Fort_Prop_Search_Att
 bOverrideAttenuation=true
 bOverrideConcurrency=true
 maxDistance=1300
 soundClassObject=&Fort_Gadgets_Items
 volumeMultiplier=0.325
 attenuationOverrides {falloffDistance=1000, spatializationAlgorithm=SPATIALIZATION_HRTF, stereoSpread=0, attenuationShapeExtents {x=300}}
 concurrencyOverrides {resolutionRule=PreventNew}
}
