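# Sound cue for the Grapple Glove timeout stop sound, played as a gadget item effect.
# A Random node picks equally between two stop wave variations and feeds a Modulator at the root.
# The cue lowers volume and pitch, overrides attenuation with HRTF spatialization and a custom falloff, and limits concurrency to one instance, blocking new plays while one is active.
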SoundCue /GrappleGloves/Items/GrappleGloves/Sounds/GrappleGlove_Timeout_Stop_Cue
root=Modulator_0
nodes [
 {id=Modulator_0, children [Random_0]}
 {id=Random_0, children [WavePlayer_2,WavePlayer_3], properties {weights [1,1]}}
 {id=WavePlayer_2, properties {soundWaveAssetPtr=&/GrappleGlove_Timeout_Stop_02}}
 {id=WavePlayer_3, properties {soundWaveAssetPtr=&/GrappleGlove_Timeout_Stop_03}}
]
properties {
 attenuationSettings=/Game/Sounds/Attenuation/Fort_Prop_Search_Att
 bOverrideAttenuation=true
 bOverrideConcurrency=true
 maxDistance=1500
 pitchMultiplier=0.9
 soundClassObject=&Fort_Gadgets_Items
 volumeMultiplier=0.1625
 attenuationOverrides {falloffDistance=1000, spatializationAlgorithm=SPATIALIZATION_HRTF, stereoSpread=0, attenuationShapeExtents {x=500}}
 concurrencyOverrides {maxCount=1, resolutionRule=PreventNew}
}
